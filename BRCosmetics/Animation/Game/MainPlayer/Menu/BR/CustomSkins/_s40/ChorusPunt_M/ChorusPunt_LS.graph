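# Level Sequence that animates a FortPlayerMannequin for the ChorusPunt_M custom skin's Battle Royale main menu animation.
# The spawned mannequin sits at a fixed origin transform with an empty skeletal animation track, so motion comes from a keyed Fortnite_Biped_CtrlRig Control Rig track.
# FK/IK switches are constant and differ between legs and arms, with small foot IK steps and toe pivots under busier upper-arm FK motion.
# Many visible curves return to their starting values near the end, suggesting a loop.
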LevelSequence /BRCosmetics/Animation/Game/MainPlayer/Menu/BR/CustomSkins/_s40/ChorusPunt_M/ChorusPunt_LS
bindings [
 FortPlayerMannequin {
  id=#1
  possessed_class=FortPlayerMannequin
  tracks [
   {type=MovieSceneSpawnTrack, sections [{type=MovieSceneSpawnSection, channels [None {type=MovieSceneScriptingBoolChannel, default=true}]}]}
   {type=MovieSceneSkeletalAnimationTrack, sections []}
   {
    type=MovieScene3DTransformTrack
    sections [{
      type=MovieScene3DTransformSection
      channels [
       Location.X {type=MovieSceneScriptingDoubleChannel, default=0, const [0,2]}
       Location.Y {type=MovieSceneScriptingDoubleChannel, default=0, const [0,2]}
       Location.Z {type=MovieSceneScriptingDoubleChannel, default=0, const [0,2]}
       Rotation.X {type=MovieSceneScriptingDoubleChannel, default=0, const [0,2]}
       Rotation.Y {type=MovieSceneScriptingDoubleChannel, default=0, const [0,2]}
       Rotation.Z {type=MovieSceneScriptingDoubleChannel, default=0, const [0,2]}
       Scale.X {type=MovieSceneScriptingDoubleChannel, default=1, const [1,2]}
       Scale.Y {type=MovieSceneScriptingDoubleChannel, default=1, const [1,2]}
       Scale.Z {type=MovieSceneScriptingDoubleChannel, default=1, const [1,2]}
      ]
     }]
   }
   {
    type=MovieSceneControlRigParameterTrack
    display_name=Fortnite_Biped_CtrlRig
    sections [{
      type=MovieSceneControlRigParameterSection
      channels [
       leg_l_stretch_switch {type=MovieSceneScriptingBoolChannel, default=false, const [true,21]}
       leg_r_stretch_switch {type=MovieSceneScriptingBoolChannel, default=false, const [true,19]}
       arm_l_fk_local {type=MovieSceneScriptingBoolChannel, default=true, const [true,32]}
       Auto Interpolate_3 {type=MovieSceneScriptingBoolChannel, default=false, const [false,21]}
       Auto Interpolate {type=MovieSceneScriptingBoolChannel, default=false, const [false,20]}
       arm_l_stretch_switch {type=MovieSceneScriptingBoolChannel, default=false, const [false,12]}
       Keep Aligned {type=MovieSceneScriptingBoolChannel, default=false, const [false,12]}
       arm_r_fk_local {type=MovieSceneScriptingBoolChannel, default=true, const [true,29]}
       Auto Interpolate_4 {type=MovieSceneScriptingBoolChannel, default=false, const [false,21]}
       Auto Interpolate_2 {type=MovieSceneScriptingBoolChannel, default=false, const [false,21]}
       arm_r_stretch_switch {type=MovieSceneScriptingBoolChannel, default=false, const [false,12]}
       Keep Aligned_2 {type=MovieSceneScriptingBoolChannel, default=false, const [false,12]}
       head_local {type=MovieSceneScriptingBoolChannel, default=true, const [true,14]}
       Stretch {type=MovieSceneScriptingBoolChannel, default=false, const [false,14]}
       head_stretch {type=MovieSceneScriptingBoolChannel, default=false, const [false,14]}
       spine_stretch_switch {type=MovieSceneScriptingBoolChannel, default=false, const [false,14]}
       arm_l_fk_ik_switch {type=MovieSceneScriptingBoolChannel, default=false, const [false,15]}
       leg_l_fk_ik_switch {type=MovieSceneScriptingBoolChannel, default=true, const [true,15]}
       arm_r_fk_ik_switch {type=MovieSceneScriptingBoolChannel, default=false, const [false,15]}
       leg_r_fk_ik_switch {type=MovieSceneScriptingBoolChannel, default=true, const [true,15]}
       ShowBodyControls {type=MovieSceneScriptingBoolChannel, default=true, const [true,15]}
       spine_fk_ik_switch {type=MovieSceneScriptingBoolChannel, default=false, const [false,15]}
       neck_fk_ik_switch {type=MovieSceneScriptingBoolChannel, default=false, const [false,15]}
       space {type=MovieSceneScriptingByteChannel, default=0, const [0,5]}
       global_ctrl.Location.X {type=MovieSceneScriptingFloatChannel, default=0, thinned_from=16, keys [[0,0],[393,0]]}
       global_ctrl.Location.Y {type=MovieSceneScriptingFloatChannel, default=0, thinned_from=16, keys [[0,0],[393,0]]}
       global_ctrl.Location.Z {type=MovieSceneScriptingFloatChannel, default=0, thinned_from=16, keys [[0,0],[393,0]]}
       global_ctrl.Rotation.X {type=MovieSceneScriptingFloatChannel, default=0, thinned_from=16, keys [[0,0],[393,0]]}
       global_ctrl.Rotation.Y {type=MovieSceneScriptingFloatChannel, default=0, thinned_from=16, keys [[0,0],[393,0]]}
       global_ctrl.Rotation.Z {type=MovieSceneScriptingFloatChannel, default=0, const [6.4503,1]}
       root_ctrl.Location.X {type=MovieSceneScriptingFloatChannel, default=0, thinned_from=15, keys [[0,0],[393,0]]}
       root_ctrl.Location.Y {type=MovieSceneScriptingFloatChannel, default=0, thinned_from=15, keys [[0,0],[393,0]]}
       root_ctrl.Location.Z {type=MovieSceneScriptingFloatChannel, default=0, thinned_from=15, keys [[0,0],[393,0]]}
       root_ctrl.Rotation.X {type=MovieSceneScriptingFloatChannel, default=0, thinned_from=15, keys [[0,0],[393,0]]}
       root_ctrl.Rotation.Y {type=MovieSceneScriptingFloatChannel, default=0, thinned_from=15, keys [[0,0],[393,0]]}
       root_ctrl.Rotation.Z {type=MovieSceneScriptingFloatChannel, default=0, thinned_from=15, keys [[0,0],[393,0]]}
       root_ctrl.Scale.X {type=MovieSceneScriptingFloatChannel, default=1, const [1,14]}
       root_ctrl.Scale.Y {type=MovieSceneScriptingFloatChannel, default=1, const [1,14]}
       root_ctrl.Scale.Z {type=MovieSceneScriptingFloatChannel, default=1, const [1,14]}
       body_offset_ctrl.Location.X {type=MovieSceneScriptingFloatChannel, default=0, thinned_from=15, keys [[0,0],[393,0]]}
       body_offset_ctrl.Location.Y {type=MovieSceneScriptingFloatChannel, default=0, thinned_from=15, keys [[0,0],[393,0]]}
       body_offset_ctrl.Location.Z {type=MovieSceneScriptingFloatChannel, default=0, thinned_from=15, keys [[0,0],[393,0]]}
       body_offset_ctrl.Rotation.X {type=MovieSceneScriptingFloatChannel, default=0, thinned_from=15, keys [[0,0],[393,0]]}
       body_offset_ctrl.Rotation.Y {type=MovieSceneScriptingFloatChannel, default=0, thinned_from=15, keys [[0,0],[393,0]]}
       body_offset_ctrl.Rotation.Z {type=MovieSceneScriptingFloatChannel, default=0, thinned_from=15, keys [[0,0],[393,0]]}
       body_offset_ctrl.Scale.X {type=MovieSceneScriptingFloatChannel, default=1, const [1,14]}
       body_offset_ctrl.Scale.Y {type=MovieSceneScriptingFloatChannel, default=1, const [1,14]}
       body_offset_ctrl.Scale.Z {type=MovieSceneScriptingFloatChannel, default=1, const [1,14]}
       thigh_l_fk_ctrl.Location.X {type=MovieSceneScriptingFloatChannel, default=0, thinned_from=14, keys [[0,0],[357,0]]}
       thigh_l_fk_ctrl.Location.Y {type=MovieSceneScriptingFloatChannel, default=0, thinned_from=14, keys [[0,0],[357,0]]}
       thigh_l_fk_ctrl.Location.Z {type=MovieSceneScriptingFloatChannel, default=0, thinned_from=14, keys [[0,0],[357,0]]}
       thigh_l_fk_ctrl.Rotation.X {
        type=MovieSceneScriptingFloatChannel
        default=1.216
        keys [[0,8.4788],[30,38.9608],[37,19.7949],[52,1.8024],[104,7.3072],[109,7.3072],[170,38.9608],[177,19.7949],[192,1.8024],[230,3.3103],[331,38.9608],[347,36.9048],[350,19.7949],[357,38.9608]]
       }
       thigh_l_fk_ctrl.Rotation.Y {
        type=MovieSceneScriptingFloatChannel
        default=-1.2128
        keys [[0,1.3548],[30,4.8327],[37,-5.1446],[52,-1.2292],[104,-1.3834],[109,-1.3834],[170,4.8327],[177,-5.1446],[192,-1.2292],[230,-1.2714],[331,4.8327],[347,2.6345],[350,-5.1446],[357,4.8327]]
       }
       thigh_l_fk_ctrl.Rotation.Z {
        type=MovieSceneScriptingFloatChannel
        default=-0.1664
        keys [[0,-12.3111],[30,-2.8658],[37,-1.9858],[52,-0.0394],[104,1.1529],[109,1.1529],[170,-2.8658],[177,-1.9858],[192,-0.0394],[230,0.2872],[331,-2.8658],[347,-4.4973],[350,-1.9858],[357,-2.8658]]
       }
       thigh_l_fk_ctrl.Scale.X {type=MovieSceneScriptingFloatChannel, default=1, const [1,14]}
       thigh_l_fk_ctrl.Scale.Y {type=MovieSceneScriptingFloatChannel, default=1, const [1,14]}
       thigh_l_fk_ctrl.Scale.Z {type=MovieSceneScriptingFloatChannel, default=1, const [1,14]}
       calf_l_fk_ctrl.Location.X {
        type=MovieSceneScriptingFloatChannel
        default=0
        thinned_from=14
        keys [[0,0],[30,-0.0097],[37,-0.1715],[52,-0.1715],[104,0],[109,0],[170,-0.0097],[177,-0.1715],[331,-0.1715],[347,0],[350,-0.1715],[357,-0.0097]]
       }
       calf_l_fk_ctrl.Location.Y {
        type=MovieSceneScriptingFloatChannel
        default=0
        thinned_from=14
        keys [[0,0],[30,-0.025],[37,0.0128],[52,0.0128],[104,0],[109,0],[170,-0.025],[177,0.0128],[331,0.0128],[347,0],[350,0.0128],[357,-0.025]]
       }
       calf_l_fk_ctrl.Location.Z {
        type=MovieSceneScriptingFloatChannel
        default=0
        thinned_from=14
        keys [[0,0],[30,0.7199],[37,-0.0037],[52,-0.0037],[104,0],[109,0],[170,0.7199],[177,-0.0037],[331,-0.0037],[347,0],[350,-0.0037],[357,0.7199]]
       }
       calf_l_fk_ctrl.Rotation.X {
        type=MovieSceneScriptingFloatChannel
        default=0.0543
        thinned_from=14
        keys [[0,0.5038],[30,-2.9248],[37,0.1537],[52,0.1537],[104,0.0219],[109,0.0219],[170,-2.9248],[177,0.1537],[331,0.1537],[347,0.4433],[350,0.1537],[357,-2.9248]]
       }
       calf_l_fk_ctrl.Rotation.Y {
        type=MovieSceneScriptingFloatChannel
        default=1.2238
        thinned_from=14
        keys [[0,1.1167],[30,2.034],[37,1.2153],[52,1.2153],[109,1.2239],[170,2.034],[177,1.2153],[331,1.2153],[347,1.142],[350,1.2153],[357,2.034]]
       }
       calf_l_fk_ctrl.Rotation.Z {
        type=MovieSceneScriptingFloatChannel
        default=0.3732
        thinned_from=14
        keys [[0,22.1173],[30,27.124],[37,5.0434],[52,5.0434],[104,-1.1433],[109,-1.1433],[170,27.124],[177,5.0434],[331,5.0434],[347,19.0495],[350,5.0434],[357,27.124]]
       }
       calf_l_fk_ctrl.Scale.X {type=MovieSceneScriptingFloatChannel, default=1, const [1,14]}
       calf_l_fk_ctrl.Scale.Y {type=MovieSceneScriptingFloatChannel, default=1, const [1,14]}
       calf_l_fk_ctrl.Scale.Z {type=MovieSceneScriptingFloatChannel, default=1, const [1,14]}
       foot_l_fk_ctrl.Location.X {
        type=MovieSceneScriptingFloatChannel
        default=0
        thinned_from=14
        keys [[0,0],[30,0],[37,-0.1874],[52,-0.1874],[104,0],[170,0],[177,-0.1874],[331,-0.1874],[347,0],[350,-0.1874],[357,0]]
       }
       foot_l_fk_ctrl.Location.Y {
        type=MovieSceneScriptingFloatChannel
        default=0.0398
        thinned_from=14
        keys [[0,0.0398],[30,0.0389],[37,0.0398],[109,0.0398],[170,0.0389],[177,0.0398],[350,0.0398],[357,0.0389]]
       }
       foot_l_fk_ctrl.Location.Z {
        type=MovieSceneScriptingFloatChannel
        default=-0.8885
        thinned_from=14
        keys [[0,-0.8885],[30,-0.6669],[37,-0.8885],[109,-0.8885],[170,-0.6669],[177,-0.8885],[350,-0.8885],[357,-0.6669]]
       }
       foot_l_fk_ctrl.Rotation.X {
        type=MovieSceneScriptingFloatChannel
        default=-1.1647
        thinned_from=14
        keys [[0,2.4078],[30,-5.3719],[37,-0.355],[52,-0.355],[104,-1.1768],[109,-1.1768],[170,-5.3719],[177,-0.355],[331,-0.355],[347,-0.3811],[350,-0.355],[357,-5.3719]]
       }
       foot_l_fk_ctrl.Rotation.Y {
        type=MovieSceneScriptingFloatChannel
        default=0.0792
        thinned_from=14
        keys [[0,-1.8745],[30,1.8476],[37,4.6189],[52,4.6189],[104,-0.1166],[109,-0.1166],[170,1.8476],[177,4.6189],[331,4.6189],[347,1.316],[350,4.6189],[357,1.8476]]
       }
       foot_l_fk_ctrl.Rotation.Z {
        type=MovieSceneScriptingFloatChannel
        default=-0.2074
        thinned_from=14
        keys [[0,-8.4844],[30,-13.4921],[37,-2.7464],[52,-2.7464],[104,0.5735],[109,0.5735],[170,-13.4921],[177,-2.7464],[331,-2.7464],[347,-10.4511],[350,-2.7464],[357,-13.4921]]
       }
       foot_l_fk_ctrl.Scale.X {type=MovieSceneScriptingFloatChannel, default=1, const [1,14]}
       foot_l_fk_ctrl.Scale.Y {type=MovieSceneScriptingFloatChannel, default=1, const [1,14]}
       foot_l_fk_ctrl.Scale.Z {type=MovieSceneScriptingFloatChannel, default=1, const [1,14]}
       ball_l_fk_ctrl.Location.X {type=MovieSceneScriptingFloatChannel, default=0, thinned_from=15, keys [[0,0],[393,0]]}
       ball_l_fk_ctrl.Location.Y {type=MovieSceneScriptingFloatChannel, default=0, thinned_from=15, keys [[0,0],[393,0]]}
       ball_l_fk_ctrl.Location.Z {type=MovieSceneScriptingFloatChannel, default=0, thinned_from=15, keys [[0,0],[393,0]]}
       ball_l_fk_ctrl.Rotation.X {type=MovieSceneScriptingFloatChannel, default=0, thinned_from=14, keys [[0,0],[368,0]]}
       ball_l_fk_ctrl.Rotation.Y {type=MovieSceneScriptingFloatChannel, default=0, thinned_from=14, keys [[0,0],[368,0]]}
       ball_l_fk_ctrl.Rotation.Z {type=MovieSceneScriptingFloatChannel, default=0, thinned_from=14, keys [[0,0],[368,0]]}
       ball_l_fk_ctrl.Scale.X {type=MovieSceneScriptingFloatChannel, default=1, const [1,14]}
       ball_l_fk_ctrl.Scale.Y {type=MovieSceneScriptingFloatChannel, default=1, const [1,14]}
       ball_l_fk_ctrl.Scale.Z {type=MovieSceneScriptingFloatChannel, default=1, const [1,14]}
       foot_l_ik_ctrl.Location.X {
        type=MovieSceneScriptingFloatChannel
        default=0
        thinned_from=26
        keys [[0,0],[14,0],[17,1.9717],[20,5.6013],[25,4.1905],[58,4.1905],[63,-1.9571],[69,-3.8005],[87,-3.8005],[91,-1.0836],[94,0.6798],[99,0],[393,0]]
       }
       foot_l_ik_ctrl.Location.Y {type=MovieSceneScriptingFloatChannel, default=0, thinned_from=26, keys [[0,0],[87,0],[91,-0.0619],[94,-0.1021],[99,0],[393,0]]}
       foot_l_ik_ctrl.Location.Z {
        type=MovieSceneScriptingFloatChannel
        default=0
        thinned_from=26
        keys [[0,0],[14,0],[17,1.4085],[20,0],[58,0],[63,2.539],[69,0],[87,0],[91,3.2851],[94,2.5148],[99,0],[393,0]]
       }
       foot_l_ik_ctrl.Rotation.X {
        type=MovieSceneScriptingFloatChannel
        default=0
        thinned_from=26
        keys [[0,0],[14,0],[17,2.5247],[20,0],[58,0],[63,6.7633],[69,0],[87,0],[91,4.9874],[94,6.4621],[99,0],[393,0]]
       }
       foot_l_ik_ctrl.Rotation.Y {
        type=MovieSceneScriptingFloatChannel
        default=0
        thinned_from=26
        keys [[0,0],[14,0],[17,-2.7459],[20,0],[58,0],[63,0.2935],[69,0],[87,0],[91,0.9529],[94,-1.527],[99,0],[393,0]]
       }
       foot_l_ik_ctrl.Rotation.Z {
        type=MovieSceneScriptingFloatChannel
        default=0
        thinned_from=25
        keys [[0,-10.0576],[14,-10.0576],[17,-5.8764],[20,2.1646],[25,-2.1376],[58,-2.1376],[63,-13.4925],[69,-10.0283],[87,-10.0283],[91,-3.3774],[94,-7.9449],[99,-10.0283],[170,-10.0576],[177,-10.0283],[350,-10.0283],[357,-10.0576]]
       }
       foot_l_ik_ctrl.Scale.X {type=MovieSceneScriptingFloatChannel, default=1, const [1,25]}
       foot_l_ik_ctrl.Scale.Y {type=MovieSceneScriptingFloatChannel, default=1, const [1,25]}
       foot_l_ik_ctrl.Scale.Z {type=MovieSceneScriptingFloatChannel, default=1, const [1,25]}
       ball_l_ik_ctrl.Location.X {type=MovieSceneScriptingFloatChannel, default=0, thinned_from=15, keys [[0,0],[393,0]]}
       ball_l_ik_ctrl.Location.Y {type=MovieSceneScriptingFloatChannel, default=0, thinned_from=15, keys [[0,0],[393,0]]}
       ball_l_ik_ctrl.Location.Z {type=MovieSceneScriptingFloatChannel, default=0, thinned_from=15, keys [[0,0],[393,0]]}
       ball_l_ik_ctrl.Rotation.X {type=MovieSceneScriptingFloatChannel, default=0, thinned_from=15, keys [[0,0],[393,0]]}
       ball_l_ik_ctrl.Rotation.Y {type=MovieSceneScriptingFloatChannel, default=0, thinned_from=15, keys [[0,0],[393,0]]}
       ball_l_ik_ctrl.Rotation.Z {type=MovieSceneScriptingFloatChannel, default=0, thinned_from=15, keys [[0,0],[393,0]]}
       ball_l_ik_ctrl.Scale.X {type=MovieSceneScriptingFloatChannel, default=1, const [1,14]}
       ball_l_ik_ctrl.Scale.Y {type=MovieSceneScriptingFloatChannel, default=1, const [1,14]}
       ball_l_ik_ctrl.Scale.Z {type=MovieSceneScriptingFloatChannel, default=1, const [1,14]}
       foot_bk1_l_ctrl.Location.X {type=MovieSceneScriptingFloatChannel, default=0, thinned_from=46, keys [[0,0],[399,0]]}
       foot_bk1_l_ctrl.Location.Y {type=MovieSceneScriptingFloatChannel, default=0, thinned_from=46, keys [[0,0],[399,0]]}
       foot_bk1_l_ctrl.Location.Z {type=MovieSceneScriptingFloatChannel, default=0, thinned_from=46, keys [[0,0],[399,0]]}
       foot_bk1_l_ctrl.Rotation.X {type=MovieSceneScriptingFloatChannel, default=0, thinned_from=46, keys [[0,0],[399,0]]}
       foot_bk1_l_ctrl.Rotation.Y {type=MovieSceneScriptingFloatChannel, default=0, thinned_from=46, keys [[0,0],[399,0]]}
       foot_bk1_l_ctrl.Rotation.Z {type=MovieSceneScriptingFloatChannel, default=0, thinned_from=45, keys [[0,0],[399,0]]}
       foot_bk1_l_ctrl.Scale.X {type=MovieSceneScriptingFloatChannel, default=1, const [1,45]}
       foot_bk1_l_ctrl.Scale.Y {type=MovieSceneScriptingFloatChannel, default=1, const [1,45]}
       foot_bk1_l_ctrl.Scale.Z {type=MovieSceneScriptingFloatChannel, default=1, const [1,45]}
       foot_roll_l_ctrl.Location.X {type=MovieSceneScriptingFloatChannel, default=0, thinned_from=15, keys [[0,0],[393,0]]}
       foot_roll_l_ctrl.Location.Y {type=MovieSceneScriptingFloatChannel, default=0, thinned_from=15, keys [[0,0],[393,0]]}
       foot_roll_l_ctrl.Location.Z {type=MovieSceneScriptingFloatChannel, default=0, thinned_from=15, keys [[0,0],[393,0]]}
       foot_roll_l_ctrl.Rotation.X {type=MovieSceneScriptingFloatChannel, default=0, thinned_from=15, keys [[0,0],[393,0]]}
       foot_roll_l_ctrl.Rotation.Y {type=MovieSceneScriptingFloatChannel, default=0, thinned_from=15, keys [[0,0],[393,0]]}
       foot_roll_l_ctrl.Rotation.Z {type=MovieSceneScriptingFloatChannel, default=0, thinned_from=15, keys [[0,0],[393,0]]}
       foot_roll_l_ctrl.Scale.X {type=MovieSceneScriptingFloatChannel, default=1, const [1,14]}
       foot_roll_l_ctrl.Scale.Y {type=MovieSceneScriptingFloatChannel, default=1, const [1,14]}
       foot_roll_l_ctrl.Scale.Z {type=MovieSceneScriptingFloatChannel, default=1, const [1,14]}
       roll_blend_l_ctrl {type=MovieSceneScriptingFloatChannel, default=1, const [1,14]}
       tip_l_ctrl.Location.X {type=MovieSceneScriptingFloatChannel, default=0, thinned_from=27, keys [[0,0],[393,0]]}
       tip_l_ctrl.Location.Y {type=MovieSceneScriptingFloatChannel, default=0, thinned_from=27, keys [[0,0],[393,0]]}
       tip_l_ctrl.Location.Z {type=MovieSceneScriptingFloatChannel, default=0, thinned_from=27, keys [[0,0],[393,0]]}
       tip_l_ctrl.Rotation.X {type=MovieSceneScriptingFloatChannel, default=0, thinned_from=27, keys [[0,0],[393,0]]}
       tip_l_ctrl.Rotation.Y {
        type=MovieSceneScriptingFloatChannel
        default=0
        thinned_from=29
        keys [[0,-2.088],[13,-2.088],[30,0],[58,0],[62,-2.9397],[67,-3.2662],[95,-3.259],[104,1.9593],[109,1.9593],[123,-2.3071],[132,-2.2844],[154,-2.2844],[163,4.1942],[170,4.2743],[198,4.2743],[205,0],[315,0],[331,5.1553],[357,5.1553],[373,-2.088]]
       }
       tip_l_ctrl.Rotation.Z {type=MovieSceneScriptingFloatChannel, default=0, thinned_from=27, keys [[0,0],[393,0]]}
       tip_l_ctrl.Scale.X {type=MovieSceneScriptingFloatChannel, default=1, const [1,26]}
       tip_l_ctrl.Scale.Y {type=MovieSceneScriptingFloatChannel, default=1, const [1,26]}
       tip_l_ctrl.Scale.Z {type=MovieSceneScriptingFloatChannel, default=1, const [1,26]}
       heel_l_ctrl.Location.X {type=MovieSceneScriptingFloatChannel, default=0, thinned_from=15, keys [[0,0],[393,0]]}
       heel_l_ctrl.Location.Y {type=MovieSceneScriptingFloatChannel, default=0, thinned_from=15, keys [[0,0],[393,0]]}
       heel_l_ctrl.Location.Z {type=MovieSceneScriptingFloatChannel, default=0, thinned_from=15, keys [[0,0],[393,0]]}
       heel_l_ctrl.Rotation.X {type=MovieSceneScriptingFloatChannel, default=0, thinned_from=15, keys [[0,0],[393,0]]}
       heel_l_ctrl.Rotation.Y {type=MovieSceneScriptingFloatChannel, default=0, thinned_from=15, keys [[0,0],[393,0]]}
       heel_l_ctrl.Rotation.Z {type=MovieSceneScriptingFloatChannel, default=0, thinned_from=15, keys [[0,0],[393,0]]}
       heel_l_ctrl.Scale.X {type=MovieSceneScriptingFloatChannel, default=1, const [1,14]}
       heel_l_ctrl.Scale.Y {type=MovieSceneScriptingFloatChannel, default=1, const [1,14]}
       heel_l_ctrl.Scale.Z {type=MovieSceneScriptingFloatChannel, default=1, const [1,14]}
       leg_l_softness {type=MovieSceneScriptingFloatChannel, default=0, const [0.5,21]}
       leg_l_pv_ik_ctrl.Location.X {
        type=MovieSceneScriptingFloatChannel
        default=0
        keys [[0,-1.6977],[30,-1.6977],[37,-0.4772],[52,-1.312],[170,-1.6977],[177,-0.4772],[192,-1.312],[331,-1.6977],[347,-0.4772],[350,-0.4772],[357,-1.6977]]
       }
       leg_l_pv_ik_ctrl.Location.Y {
        type=MovieSceneScriptingFloatChannel
        default=0
        keys [[0,-0.4677],[30,-0.4677],[37,-0.0881],[52,-0.2423],[170,-0.4677],[177,-0.0881],[192,-0.2423],[331,-0.4677],[347,-0.0881],[350,-0.0881],[357,-0.4677]]
       }
       leg_l_pv_ik_ctrl.Location.Z {
        type=MovieSceneScriptingFloatChannel
        default=0
        keys [[0,-0.1666],[30,-0.1666],[37,-0.0473],[52,-0.13],[170,-0.1666],[177,-0.0473],[192,-0.13],[331,-0.1666],[347,-0.0473],[350,-0.0473],[357,-0.1666]]
       }
       leg_l_pv_ik_ctrl.Rotation.X {type=MovieSceneScriptingFloatChannel, default=0, thinned_from=11, keys [[0,0],[368,0]]}
       leg_l_pv_ik_ctrl.Rotation.Y {type=MovieSceneScriptingFloatChannel, default=0, thinned_from=11, keys [[0,0],[368,0]]}
       leg_l_pv_ik_ctrl.Rotation.Z {type=MovieSceneScriptingFloatChannel, default=0, thinned_from=11, keys [[0,0],[368,0]]}
       leg_l_pv_ik_ctrl.Scale.X {type=MovieSceneScriptingFloatChannel, default=1, const [1,11]}
       leg_l_pv_ik_ctrl.Scale.Y {type=MovieSceneScriptingFloatChannel, default=1, const [1,11]}
       leg_l_pv_ik_ctrl.Scale.Z {type=MovieSceneScriptingFloatChannel, default=1, const [1,11]}
       thigh_r_fk_ctrl.Location.X {type=MovieSceneScriptingFloatChannel, default=0, thinned_from=14, keys [[0,0],[368,0]]}
       thigh_r_fk_ctrl.Location.Y {type=MovieSceneScriptingFloatChannel, default=0, thinned_from=14, keys [[0,0],[368,0]]}
       thigh_r_fk_ctrl.Location.Z {type=MovieSceneScriptingFloatChannel, default=0, thinned_from=14, keys [[0,0],[368,0]]}
       thigh_r_fk_ctrl.Rotation.X {
        type=MovieSceneScriptingFloatChannel
        default=1.2139
        keys [[0,20.7082],[30,16.3967],[37,12.3362],[52,1.9686],[104,9.0534],[109,9.0534],[170,16.3967],[177,12.3362],[192,1.9686],[230,3.9093],[331,16.3967],[347,-4.9243],[350,12.3362],[357,16.3967]]
       }
       thigh_r_fk_ctrl.Rotation.Y {
        type=MovieSceneScriptingFloatChannel
        default=-1.213
        keys [[0,4.0381],[30,3.926],[37,4.3309],[52,-1.1665],[104,-0.7304],[109,-0.7304],[170,3.926],[177,4.3309],[192,-1.1665],[230,-1.0471],[331,3.926],[347,-0.7643],[350,4.3309],[357,3.926]]
       }
       thigh_r_fk_ctrl.Rotation.Z {
        type=MovieSceneScriptingFloatChannel
        default=-0.1592
        keys [[0,-12.243],[30,-12.3522],[37,-2.8816],[52,-0.223],[104,-0.8222],[109,-0.8222],[170,-12.3522],[177,-2.8816],[192,-0.223],[230,-0.3871],[331,-12.3522],[347,-13.3662],[350,-2.8816],[357,-12.3522]]
       }
       thigh_r_fk_ctrl.Scale.X {type=MovieSceneScriptingFloatChannel, default=1, const [1,14]}
       thigh_r_fk_ctrl.Scale.Y {type=MovieSceneScriptingFloatChannel, default=1, const [1,14]}
       thigh_r_fk_ctrl.Scale.Z {type=MovieSceneScriptingFloatChannel, default=1, const [1,14]}
       calf_r_fk_ctrl.Location.X {
        type=MovieSceneScriptingFloatChannel
        default=0
        thinned_from=14
        keys [[0,0],[30,0.0003],[37,-0.6727],[52,-0.6727],[104,0],[170,0.0003],[177,-0.6727],[331,-0.6727],[347,0],[350,-0.6727],[357,0.0003]]
       }
       calf_r_fk_ctrl.Location.Y {
        type=MovieSceneScriptingFloatChannel
        default=0
        thinned_from=14
        keys [[0,0],[30,0],[37,0.0515],[52,0.0515],[104,0],[170,0],[177,0.0515],[331,0.0515],[347,0],[350,0.0515],[357,0]]
       }
       calf_r_fk_ctrl.Location.Z {
        type=MovieSceneScriptingFloatChannel
        default=0
        thinned_from=14
        keys [[0,0],[30,-0.0135],[37,-0.0141],[52,-0.0141],[104,0],[109,0],[170,-0.0135],[177,-0.0141],[331,-0.0141],[347,0],[350,-0.0141],[357,-0.0135]]
       }
       calf_r_fk_ctrl.Rotation.X {
        type=MovieSceneScriptingFloatChannel
        default=0.0561
        thinned_from=14
        keys [[0,0.5593],[30,0.5898],[37,0.049],[52,0.049],[104,0.0646],[109,0.0646],[170,0.5898],[177,0.049],[331,0.049],[347,0.3115],[350,0.049],[357,0.5898]]
       }
       calf_r_fk_ctrl.Rotation.Y {
        type=MovieSceneScriptingFloatChannel
        default=1.1982
        thinned_from=14
        keys [[0,1.0611],[30,1.0716],[37,1.1985],[109,1.1977],[170,1.0716],[177,1.1985],[331,1.1985],[347,1.1583],[350,1.1985],[357,1.0716]]
       }
       calf_r_fk_ctrl.Rotation.Z {
        type=MovieSceneScriptingFloatChannel
        default=0.2616
        thinned_from=14
        keys [[0,25.3754],[30,23.2386],[37,-0.0767],[52,-0.0767],[104,0.669],[109,0.669],[170,23.2386],[177,-0.0767],[331,-0.0767],[347,12.6364],[350,-0.0767],[357,23.2386]]
       }
       calf_r_fk_ctrl.Scale.X {type=MovieSceneScriptingFloatChannel, default=1, const [1,14]}
       calf_r_fk_ctrl.Scale.Y {type=MovieSceneScriptingFloatChannel, default=1, const [1,14]}
       calf_r_fk_ctrl.Scale.Z {type=MovieSceneScriptingFloatChannel, default=1, const [1,14]}
       foot_r_fk_ctrl.Location.X {
        type=MovieSceneScriptingFloatChannel
        default=0
        thinned_from=14
        keys [[0,0],[30,0],[37,-0.7354],[52,-0.7354],[104,0],[170,0],[177,-0.7354],[331,-0.7354],[347,0],[350,-0.7354],[357,0]]
       }
       foot_r_fk_ctrl.Location.Y {
        type=MovieSceneScriptingFloatChannel
        default=-0.0329
        thinned_from=14
        keys [[0,-0.0329],[30,-0.0338],[37,-0.0329],[109,-0.0329],[170,-0.0337],[177,-0.0329],[350,-0.0329],[357,-0.0338]]
       }
       foot_r_fk_ctrl.Location.Z {
        type=MovieSceneScriptingFloatChannel
        default=-0.9075
        thinned_from=14
        keys [[0,-0.9075],[30,-0.686],[37,-0.9075],[109,-0.9075],[170,-0.686],[177,-0.9075],[350,-0.9075],[357,-0.686]]
       }
       foot_r_fk_ctrl.Rotation.X {
        type=MovieSceneScriptingFloatChannel
        default=-1.1648
        thinned_from=14
        keys [[0,9.9224],[30,-13.1649],[37,10.5474],[52,10.5474],[104,-1.2855],[109,-1.2855],[170,-13.1649],[177,10.5474],[331,10.5474],[347,10.8896],[350,10.5474],[357,-13.1649]]
       }
       foot_r_fk_ctrl.Rotation.Y {
        type=MovieSceneScriptingFloatChannel
        default=0.1067
        thinned_from=14
        keys [[0,-0.2789],[30,0.7308],[37,-5.6758],[52,-5.6758],[104,0.1136],[109,0.1136],[170,0.7308],[177,-5.6758],[331,-5.6758],[347,-3.4585],[350,-5.6758],[357,0.7308]]
       }
       foot_r_fk_ctrl.Rotation.Z {
        type=MovieSceneScriptingFloatChannel
        default=-0.1036
        thinned_from=14
        keys [[0,-9.7625],[30,-2.8578],[37,6.2399],[52,6.2399],[104,0.9952],[109,0.9952],[170,-2.8578],[177,6.2399],[331,6.2399],[347,1.0807],[350,6.2399],[357,-2.8578]]
       }
       foot_r_fk_ctrl.Scale.X {type=MovieSceneScriptingFloatChannel, default=1, const [1,14]}
       foot_r_fk_ctrl.Scale.Y {type=MovieSceneScriptingFloatChannel, default=1, const [1,14]}
       foot_r_fk_ctrl.Scale.Z {type=MovieSceneScriptingFloatChannel, default=1, const [1,14]}
       ball_r_fk_ctrl.Location.X {type=MovieSceneScriptingFloatChannel, default=0, thinned_from=14, keys [[0,0],[368,0]]}
       ball_r_fk_ctrl.Location.Y {type=MovieSceneScriptingFloatChannel, default=0, thinned_from=14, keys [[0,0],[368,0]]}
       ball_r_fk_ctrl.Location.Z {type=MovieSceneScriptingFloatChannel, default=0, thinned_from=14, keys [[0,0],[368,0]]}
       ball_r_fk_ctrl.Rotation.X {type=MovieSceneScriptingFloatChannel, default=0, thinned_from=14, keys [[0,-5.4152],[30,0],[368,0]]}
       ball_r_fk_ctrl.Rotation.Y {type=MovieSceneScriptingFloatChannel, default=0, thinned_from=14, keys [[0,0.0446],[30,0],[368,0]]}
       ball_r_fk_ctrl.Rotation.Z {type=MovieSceneScriptingFloatChannel, default=0, thinned_from=14, keys [[0,-0.0023],[30,0],[368,0]]}
       ball_r_fk_ctrl.Scale.X {type=MovieSceneScriptingFloatChannel, default=1, const [1,14]}
       ball_r_fk_ctrl.Scale.Y {type=MovieSceneScriptingFloatChannel, default=1, const [1,14]}
       ball_r_fk_ctrl.Scale.Z {type=MovieSceneScriptingFloatChannel, default=1, const [1,14]}
       foot_r_ik_ctrl.Location.X {
        type=MovieSceneScriptingFloatChannel
        default=0
        thinned_from=21
        keys [[0,2.415],[22,2.415],[28,-0.9521],[98,-0.9521],[101,-1.3852],[113,-1.3852],[118,0.2995],[123,0.5854],[192,0.5854],[202,0.6079],[315,0.6079],[322,0.4785],[328,-1.38],[331,-2.1752],[355,-2.1752],[360,1.4627],[366,2.415],[385,2.415]]
       }
       foot_r_ik_ctrl.Location.Y {
        type=MovieSceneScriptingFloatChannel
        default=0
        thinned_from=22
        keys [[0,0],[322,0],[324,-0.0126],[328,-0.0486],[331,0],[355,0],[360,-0.1307],[366,0],[385,0]]
       }
       foot_r_ik_ctrl.Location.Z {
        type=MovieSceneScriptingFloatChannel
        default=0
        thinned_from=22
        keys [[0,0],[93,0],[98,0.9817],[101,0],[113,0],[118,1.6566],[123,0],[322,0],[324,0.3735],[328,1.4405],[331,0],[355,0],[360,3.9299],[366,0],[385,0]]
       }
       foot_r_ik_ctrl.Rotation.X {
        type=MovieSceneScriptingFloatChannel
        default=0
        thinned_from=22
        keys [[0,0],[93,0],[98,2.8515],[101,0],[113,0],[118,2.4197],[123,0],[322,0],[324,0.9929],[328,2.599],[331,0],[355,0],[360,9.3933],[366,0],[385,0]]
       }
       foot_r_ik_ctrl.Rotation.Y {
        type=MovieSceneScriptingFloatChannel
        default=0
        thinned_from=22
        keys [[0,0],[93,0],[98,2.7311],[101,3.452],[113,3.452],[118,5.4647],[123,0],[322,0],[324,0.9129],[328,3.5211],[331,0],[355,0],[360,5.3472],[366,0],[385,0]]
       }
       foot_r_ik_ctrl.Rotation.Z {
        type=MovieSceneScriptingFloatChannel
        default=0
        thinned_from=21
        keys [[0,-15.5422],[22,-15.5422],[28,-32.5791],[93,-32.5791],[98,-32.0861],[101,-31.6606],[113,-31.6606],[118,-25.9598],[123,-23.5491],[322,-23.5491],[324,-24.1383],[328,-32.0009],[331,-32.5791],[355,-32.5791],[360,-15.7944],[366,-15.5422],[385,-15.5422]]
       }
       foot_r_ik_ctrl.Scale.X {type=MovieSceneScriptingFloatChannel, default=1, const [1,22]}
       foot_r_ik_ctrl.Scale.Y {type=MovieSceneScriptingFloatChannel, default=1, const [1,22]}
       foot_r_ik_ctrl.Scale.Z {type=MovieSceneScriptingFloatChannel, default=1, const [1,22]}
       ball_r_ik_ctrl.Location.X {type=MovieSceneScriptingFloatChannel, default=0, thinned_from=24, keys [[0,0],[395,0]]}
       ball_r_ik_ctrl.Location.Y {type=MovieSceneScriptingFloatChannel, default=0, thinned_from=24, keys [[0,0],[395,0]]}
       ball_r_ik_ctrl.Location.Z {type=MovieSceneScriptingFloatChannel, default=0, thinned_from=24, keys [[0,0],[395,0]]}
       ball_r_ik_ctrl.Rotation.X {type=MovieSceneScriptingFloatChannel, default=0, thinned_from=24, keys [[0,0],[395,0]]}
       ball_r_ik_ctrl.Rotation.Y {type=MovieSceneScriptingFloatChannel, default=0, thinned_from=24, keys [[0,0],[395,0]]}
       ball_r_ik_ctrl.Rotation.Z {
        type=MovieSceneScriptingFloatChannel
        default=0
        thinned_from=23
        keys [[0,5.4153],[7,4.9251],[27,0],[113,0],[123,5.6842],[153,5.6842],[160,0],[357,0],[380,5.4153]]
       }
       ball_r_ik_ctrl.Scale.X {type=MovieSceneScriptingFloatChannel, default=1, const [1,23]}
       ball_r_ik_ctrl.Scale.Y {type=MovieSceneScriptingFloatChannel, default=1, const [1,23]}
       ball_r_ik_ctrl.Scale.Z {type=MovieSceneScriptingFloatChannel, default=1, const [1,23]}
       foot_bk1_r_ctrl.Location.X {type=MovieSceneScriptingFloatChannel, default=0, thinned_from=47, keys [[0,0],[400,0]]}
       foot_bk1_r_ctrl.Location.Y {type=MovieSceneScriptingFloatChannel, default=0, thinned_from=47, keys [[0,0],[400,0]]}
       foot_bk1_r_ctrl.Location.Z {type=MovieSceneScriptingFloatChannel, default=0, thinned_from=47, keys [[0,0],[400,0]]}
       foot_bk1_r_ctrl.Rotation.X {type=MovieSceneScriptingFloatChannel, default=0, thinned_from=47, keys [[0,0],[400,0]]}
       foot_bk1_r_ctrl.Rotation.Y {type=MovieSceneScriptingFloatChannel, default=0, thinned_from=47, keys [[0,0],[400,0]]}
       foot_bk1_r_ctrl.Rotation.Z {type=MovieSceneScriptingFloatChannel, default=0, thinned_from=47, keys [[0,0],[400,0]]}
       foot_bk1_r_ctrl.Scale.X {type=MovieSceneScriptingFloatChannel, default=1, const [1,44]}
       foot_bk1_r_ctrl.Scale.Y {type=MovieSceneScriptingFloatChannel, default=1, const [1,44]}
       foot_bk1_r_ctrl.Scale.Z {type=MovieSceneScriptingFloatChannel, default=1, const [1,44]}
       foot_roll_r_ctrl.Location.X {type=MovieSceneScriptingFloatChannel, default=0, thinned_from=19, keys [[0,0],[395,0]]}
       foot_roll_r_ctrl.Location.Y {type=MovieSceneScriptingFloatChannel, default=0, thinned_from=19, keys [[0,0],[395,0]]}
       foot_roll_r_ctrl.Location.Z {type=MovieSceneScriptingFloatChannel, default=0, thinned_from=19, keys [[0,0],[395,0]]}
       foot_roll_r_ctrl.Rotation.X {type=MovieSceneScriptingFloatChannel, default=0, thinned_from=19, keys [[0,0],[395,0]]}
       foot_roll_r_ctrl.Rotation.Y {type=MovieSceneScriptingFloatChannel, default=0, thinned_from=19, keys [[0,0],[395,0]]}
       foot_roll_r_ctrl.Rotation.Z {type=MovieSceneScriptingFloatChannel, default=0, thinned_from=19, keys [[0,0],[395,0]]}
       foot_roll_r_ctrl.Scale.X {type=MovieSceneScriptingFloatChannel, default=1, const [1,18]}
       foot_roll_r_ctrl.Scale.Y {type=MovieSceneScriptingFloatChannel, default=1, const [1,18]}
       foot_roll_r_ctrl.Scale.Z {type=MovieSceneScriptingFloatChannel, default=1, const [1,18]}
       roll_blend_r_ctrl {type=MovieSceneScriptingFloatChannel, default=1, const [1,14]}
       tip_r_ctrl.Location.X {type=MovieSceneScriptingFloatChannel, default=0, thinned_from=28, keys [[0,0],[393,0]]}
       tip_r_ctrl.Location.Y {type=MovieSceneScriptingFloatChannel, default=0, thinned_from=28, keys [[0,0],[393,0]]}
       tip_r_ctrl.Location.Z {type=MovieSceneScriptingFloatChannel, default=0, thinned_from=28, keys [[0,0],[393,0]]}
       tip_r_ctrl.Rotation.X {type=MovieSceneScriptingFloatChannel, default=0, thinned_from=28, keys [[0,0],[393,0]]}
       tip_r_ctrl.Rotation.Y {
        type=MovieSceneScriptingFloatChannel
        default=0
        thinned_from=27
        keys [[0,2.5524],[14,2.0501],[27,-0.0994],[37,-0.0619],[43,-0.0994],[62,-0.0994],[72,5.7439],[86,5.7439],[97,2.5856],[113,2.5856],[123,16.5579],[153,16.5579],[160,0],[192,0],[202,5.3103],[319,5.3103],[331,-0.0994],[347,-0.0817],[350,-0.0619],[357,-0.0994],[373,2.5524]]
       }
       tip_r_ctrl.Rotation.Z {type=MovieSceneScriptingFloatChannel, default=0, thinned_from=28, keys [[0,0],[393,0]]}
       tip_r_ctrl.Scale.X {type=MovieSceneScriptingFloatChannel, default=1, const [1,26]}
       tip_r_ctrl.Scale.Y {type=MovieSceneScriptingFloatChannel, default=1, const [1,26]}
       tip_r_ctrl.Scale.Z {type=MovieSceneScriptingFloatChannel, default=1, const [1,26]}
       heel_r_ctrl.Location.X {type=MovieSceneScriptingFloatChannel, default=0, thinned_from=14, keys [[0,0],[357,0]]}
       heel_r_ctrl.Location.Y {type=MovieSceneScriptingFloatChannel, default=0, thinned_from=14, keys [[0,0],[357,0]]}
       heel_r_ctrl.Location.Z {type=MovieSceneScriptingFloatChannel, default=0, thinned_from=14, keys [[0,0],[357,0]]}
       heel_r_ctrl.Rotation.X {type=MovieSceneScriptingFloatChannel, default=0, thinned_from=15, keys [[0,0],[393,0]]}
       heel_r_ctrl.Rotation.Y {type=MovieSceneScriptingFloatChannel, default=0, thinned_from=15, keys [[0,0],[393,0]]}
       heel_r_ctrl.Rotation.Z {type=MovieSceneScriptingFloatChannel, default=0, thinned_from=15, keys [[0,0],[393,0]]}
       heel_r_ctrl.Scale.X {type=MovieSceneScriptingFloatChannel, default=1, const [1,14]}
       heel_r_ctrl.Scale.Y {type=MovieSceneScriptingFloatChannel, default=1, const [1,14]}
       heel_r_ctrl.Scale.Z {type=MovieSceneScriptingFloatChannel, default=1, const [1,14]}
       leg_r_softness {type=MovieSceneScriptingFloatChannel, default=0, const [0.5,19]}
       leg_r_pv_ik_ctrl.Location.X {
        type=MovieSceneScriptingFloatChannel
        default=0
        thinned_from=12
        keys [[0,-8.8519],[30,-8.8519],[37,-8.8039],[52,-8.8268],[170,-8.8519],[177,-8.8039],[192,-8.8268],[331,-8.8519],[352,-8.8039],[357,-8.8519]]
       }
       leg_r_pv_ik_ctrl.Location.Y {
        type=MovieSceneScriptingFloatChannel
        default=0
        thinned_from=12
        keys [[0,-5.6184],[30,-5.6184],[37,-5.5873],[52,-5.6018],[170,-5.6184],[177,-5.5873],[192,-5.6018],[331,-5.6184],[352,-5.5873],[357,-5.6184]]
       }
       leg_r_pv_ik_ctrl.Location.Z {
        type=MovieSceneScriptingFloatChannel
        default=0
        thinned_from=12
        keys [[0,1.7099],[30,1.7099],[37,1.7006],[52,1.705],[170,1.7099],[177,1.7006],[192,1.705],[331,1.7099],[352,1.7005],[357,1.71]]
       }
       leg_r_pv_ik_ctrl.Rotation.X {type=MovieSceneScriptingFloatChannel, default=0, thinned_from=12, keys [[0,0.0011],[357,0.0011]]}
       leg_r_pv_ik_ctrl.Rotation.Y {type=MovieSceneScriptingFloatChannel, default=0, thinned_from=12, keys [[0,-0.0033],[170,-0.0033],[352,-0.0032],[357,-0.0033]]}
       leg_r_pv_ik_ctrl.Rotation.Z {type=MovieSceneScriptingFloatChannel, default=0, thinned_from=12, keys [[0,0.0007],[357,0.0007]]}
       leg_r_pv_ik_ctrl.Scale.X {type=MovieSceneScriptingFloatChannel, default=1, const [1,12]}
       leg_r_pv_ik_ctrl.Scale.Y {type=MovieSceneScriptingFloatChannel, default=1, const [1,12]}
       leg_r_pv_ik_ctrl.Scale.Z {type=MovieSceneScriptingFloatChannel, default=1, const [1,12]}
       upperarm_l_fk_ctrl.Location.X {type=MovieSceneScriptingFloatChannel, default=0, thinned_from=67, keys [[0,0],[385,0]]}
       upperarm_l_fk_ctrl.Location.Y {type=MovieSceneScriptingFloatChannel, default=0, thinned_from=67, keys [[0,0],[385,0]]}
       upperarm_l_fk_ctrl.Location.Z {type=MovieSceneScriptingFloatChannel, default=0, thinned_from=67, keys [[0,0],[385,0]]}
       upperarm_l_fk_ctrl.Rotation.X {
        type=MovieSceneScriptingFloatChannel
        default=24.4912
        keys [[0,37.1413],[4,42.1719],[9,42.0109],[15,25.6026],[21,5.5847],[30,36.4024],[34,42.0087],[38,45.0109],[42,39.9971],[47,46.7648],[52,12.5312],[56,-10.7934],[61,-12.4639],[67,-8.139],[75,23.5128],[84,38.3761],[95,-2.2619],[107,50.4699],[116,37.4235],[122,10.5804],[128,-15.1481],[133,0.2993],[137,16.0648],[142,33.8973],[146,39.1087],[153,40.2931],[160,0.7444],[166,11.0128],[170,22.0103],[175,29.0096],[182,23.6065],[189,31.0627],[192,29.6039],[196,7.0908],[205,-15.7813],[212,7.4309],[216,13.9245],[221,17.556],[225,19.7918],[230,21.1027],[237,8.6627],[243,-1.0481],[250,26.4985],[258,26.9414],[263,31.5414],[268,34.9063],[273,20.74],[280,-5.4416],[288,-21.3436],[293,-9.2541],[298,-2.617],[302,-1.0459],[307,-0.0982],[312,0.8568],[317,1.0676],[321,10.296],[324,26.842],[329,34.6016],[340,38.2006],[345,30.2761],[351,35.4234],[357,28.5903],[361,12.4946],[365,-13.7328],[370,-5.2158],[378,24.6711],[385,37.1413]]
       }
       upperarm_l_fk_ctrl.Rotation.Y {
        type=MovieSceneScriptingFloatChannel
        default=-10.9057
        keys [[0,9.31],[4,11.8699],[9,13.7589],[15,5.239],[21,-6.3529],[30,2.5414],[34,7.1441],[38,9.828],[42,9.6961],[47,11.6852],[52,4.7671],[56,-1.8403],[61,-8.8782],[67,-13.9704],[75,2.7357],[84,16.2312],[98,-2.0436],[107,-3.5153],[116,-7.2206],[122,-9.359],[126,-10.6433],[133,5.2933],[137,10.1279],[142,12.0754],[146,12.0892],[153,12.8784],[160,2.0957],[166,-0.4357],[170,-3.2151],[175,-5.825],[182,-8.5842],[189,-8.0476],[192,-8.3249],[196,-10.3825],[205,-9.1577],[212,-6.024],[216,-4.2548],[221,-3.5263],[225,-2.3548],[230,-1.0701],[237,-3.8225],[243,-5.3662],[250,-8.8281],[258,-10.1142],[263,-8.2535],[268,-5.7294],[273,-6.1935],[282,-9.843],[288,-15.4398],[293,-13.6824],[298,-11.4948],[302,-10.5007],[307,-8.984],[312,-7.8209],[317,-6.2402],[321,-3.8128],[324,-1.2746],[329,5.0012],[340,17.7725],[345,15.9582],[351,17.3808],[357,15.8302],[361,5.1701],[365,-15.9806],[370,-11.7479],[378,0.0858],[385,9.31]]
       }
       upperarm_l_fk_ctrl.Rotation.Z {
        type=MovieSceneScriptingFloatChannel
        default=5.0959
        keys [[0,-37.5589],[4,-37.3835],[9,-35.3885],[15,-13.1267],[21,8.7235],[30,21.1645],[34,23.5571],[38,25.6603],[42,24.3451],[47,25.8017],[52,13.6631],[56,4.1392],[61,-10.3567],[67,-23.7442],[75,-30.0971],[84,-36.6948],[98,-1.136],[107,18.8364],[116,27.1753],[122,4.681],[126,-8.0333],[133,-22.8744],[137,-25.0592],[142,-24.0754],[146,-24.1229],[153,-24.8357],[160,-12.7264],[166,9.628],[170,14.2685],[175,17.8324],[182,20.0484],[189,21.0557],[192,20.1488],[196,13.4384],[205,-11.4917],[212,-17.4013],[216,-18.9359],[221,-20.4511],[225,-20.9771],[230,-21.3639],[237,-12.8195],[243,5.6344],[250,15.3051],[258,15.8136],[263,17.3334],[268,17.9982],[273,16.1254],[282,3.2744],[288,-9.495],[293,-18.8529],[298,-22.024],[302,-23.1584],[307,-23.4433],[312,-22.2693],[317,-13.9105],[321,-0.2624],[324,13.8155],[329,20.1297],[340,27.661],[345,25.8727],[351,25.9197],[357,24.2919],[361,19.7665],[365,2.0332],[370,-14.8531],[378,-34.2493],[385,-37.5589]]
       }
       upperarm_l_fk_ctrl.Scale.X {type=MovieSceneScriptingFloatChannel, default=1, const [1.0005,68]}
       upperarm_l_fk_ctrl.Scale.Y {type=MovieSceneScriptingFloatChannel, default=1, const [1.0005,68]}
       upperarm_l_fk_ctrl.Scale.Z {type=MovieSceneScriptingFloatChannel, default=1, const [1.0005,68]}
       lowerarm_l_fk_ctrl.Location.X {type=MovieSceneScriptingFloatChannel, default=0, thinned_from=42, keys [[0,0],[385,0]]}
       lowerarm_l_fk_ctrl.Location.Y {type=MovieSceneScriptingFloatChannel, default=0, thinned_from=42, keys [[0,0],[385,0]]}
       lowerarm_l_fk_ctrl.Location.Z {type=MovieSceneScriptingFloatChannel, default=0, thinned_from=42, keys [[0,0],[385,0]]}
       lowerarm_l_fk_ctrl.Rotation.X {type=MovieSceneScriptingFloatChannel, default=3.7112, thinned_from=42, keys [[0,0],[385,0]]}
       lowerarm_l_fk_ctrl.Rotation.Y {type=MovieSceneScriptingFloatChannel, default=-1.8544, thinned_from=42, keys [[0,0],[385,0]]}
       lowerarm_l_fk_ctrl.Rotation.Z {
        type=MovieSceneScriptingFloatChannel
        default=1.4608
        thinned_from=41
        keys [[0,18.9303],[8,23.2377],[20,3.2442],[31,16.6969],[35,19.1978],[40,19.3652],[49,19.8415],[67,-1.4246],[84,33.6188],[98,3.2522],[107,10.9575],[116,25.3691],[122,3.0873],[127,-4.5219],[138,16.8827],[146,12.515],[153,18.3713],[160,-3.484],[171,3.779],[179,8.8413],[187,13.8812],[194,21.8421],[205,-4.5219],[215,4.6518],[225,9.5619],[230,7.381],[243,-5.036],[250,4.508],[256,10.8058],[271,9.79],[282,-5.2878],[292,-0.3743],[302,1.3012],[313,0.9553],[321,-1.3873],[332,17.2073],[352,18.7292],[358,26.6248],[369,-5.036],[385,18.9303]]
       }
       lowerarm_l_fk_ctrl.Scale.X {type=MovieSceneScriptingFloatChannel, default=1, const [1.0005,42]}
       lowerarm_l_fk_ctrl.Scale.Y {type=MovieSceneScriptingFloatChannel, default=1, const [1.0005,42]}
       lowerarm_l_fk_ctrl.Scale.Z {type=MovieSceneScriptingFloatChannel, default=1, const [1.0005,42]}
       hand_l_fk_ctrl.Location.X {type=MovieSceneScriptingFloatChannel, default=0, thinned_from=53, keys [[0,0],[385,0]]}
       hand_l_fk_ctrl.Location.Y {
        type=MovieSceneScriptingFloatChannel
        default=0
        thinned_from=52
        keys [[0,-0.2003],[9,-0.1636],[14,-0.0616],[18,-0.059],[26,0.262],[30,0.376],[37,0.2801],[46,0.2801],[52,0.2379],[56,0.224],[61,0.1966],[69,0.1034],[80,-0.0441],[87,-0.1239],[92,-0.1698],[98,-0.2003],[107,-0.1636],[116,0.0541],[122,0.1595],[126,0.262],[131,0.3709],[136,0.4312],[141,0.3913],[146,0.2778],[149,0.1918],[153,0.1278],[161,0.0477],[171,-0.0441],[181,-0.2003],[188,-0.082],[194,-0.0249],[199,0.1164],[205,0.262],[215,0.4312],[225,0.224],[230,0.1278],[235,0.0742],[241,0.0184],[248,-0.0441],[266,-0.2003],[282,0.262],[294,0.243],[301,0.2262],[307,0.224],[311,0.2437],[319,0.2764],[325,0.3085],[329,0.3286],[352,0.4193],[369,-0.0441],[385,-0.2003]]
       }
       hand_l_fk_ctrl.Location.Z {
        type=MovieSceneScriptingFloatChannel
        default=0
        thinned_from=53
        keys [[0,0.031],[9,0.0254],[14,0.0095],[18,0.0091],[26,-0.0406],[30,-0.0583],[37,-0.0434],[46,-0.0434],[52,-0.0369],[56,-0.0347],[61,-0.0305],[69,-0.016],[80,0.0068],[87,0.0192],[92,0.0263],[98,0.031],[107,0.0254],[116,-0.0084],[122,-0.0247],[126,-0.0406],[131,-0.0575],[136,-0.0668],[141,-0.0606],[146,-0.0431],[149,-0.0297],[153,-0.0198],[161,-0.0074],[171,0.0068],[181,0.031],[188,0.0127],[194,0.0039],[199,-0.018],[205,-0.0406],[215,-0.0668],[225,-0.0347],[230,-0.0198],[235,-0.0115],[241,-0.0029],[248,0.0068],[266,0.031],[282,-0.0406],[294,-0.0377],[301,-0.0351],[307,-0.0347],[311,-0.0389],[319,-0.0454],[325,-0.0515],[329,-0.0552],[346,-0.0668],[352,-0.0668],[369,0.0068],[385,0.031]]
       }
       hand_l_fk_ctrl.Rotation.X {
        type=MovieSceneScriptingFloatChannel
        default=0
        thinned_from=52
        keys [[0,20.7771],[9,24.7205],[14,13.462],[18,21.5273],[26,17.6434],[30,-3.5808],[37,-13.9638],[46,-13.9638],[52,-20.1057],[56,-47.4576],[61,-4.4056],[69,12.0766],[80,32.0841],[87,11.5599],[92,19.5641],[98,20.7771],[107,1.6579],[116,16.5477],[122,19.462],[126,16.7203],[131,32.6596],[136,14.5351],[141,13.7094],[146,4.9613],[149,7.598],[153,1.9128],[161,17.0083],[171,19.2169],[181,20.7771],[188,19.9748],[194,46.9089],[199,13.9295],[205,16.7203],[215,9.6819],[225,13.1],[230,15.1258],[235,21.6114],[241,7.3475],[248,25.97],[266,20.7771],[282,16.7203],[294,10.5751],[301,5.3756],[307,4.7811],[311,-0.0187],[319,6.0481],[325,4.5612],[329,-24.7636],[352,-1.2565],[369,32.0841],[385,20.7771]]
       }
       hand_l_fk_ctrl.Rotation.Y {
        type=MovieSceneScriptingFloatChannel
        default=0
        thinned_from=52
        keys [[0,26.0032],[9,3.5647],[14,0.4519],[18,10.8264],[26,-0.4118],[30,19.5639],[37,25.6055],[46,25.6055],[52,41.7102],[56,53.7544],[61,52.2219],[69,-8.0332],[80,18.7663],[87,1.8707],[92,-8.3546],[98,26.0032],[107,18.1555],[116,17.8943],[122,21.0931],[126,16.8209],[131,16.8624],[136,10.6102],[141,14.8281],[146,4.1604],[149,-1.3823],[153,3.3531],[161,-20.0654],[171,-0.2235],[181,26.0032],[188,23.8047],[194,33.286],[199,34.1559],[205,16.8209],[215,31.4691],[225,32.1801],[230,31.6888],[235,-15.0744],[241,-7.2315],[248,6.5836],[266,26.0032],[282,16.8209],[294,15.5888],[301,15.4384],[307,14.6088],[311,11.9819],[319,-17.5014],[325,-17.9195],[329,17.1201],[352,9.7015],[369,18.7663],[385,26.0032]]
       }
       hand_l_fk_ctrl.Rotation.Z {
        type=MovieSceneScriptingFloatChannel
        default=0
        thinned_from=52
        keys [[0,-8.0939],[9,-14.1522],[14,-18.6819],[18,-22.6362],[26,3.5959],[30,-9.4389],[37,-28.4674],[46,-28.4674],[52,-49.3655],[56,-78.7963],[61,-31.6129],[69,-11.5108],[80,-5.0558],[87,-4.936],[92,-9.371],[98,-8.0939],[107,-32.2772],[116,-2.999],[122,-11.993],[126,-11.98],[131,3.4054],[136,-22.707],[141,-28.4158],[146,-17.0095],[149,-7.4334],[153,-9.17],[161,1.732],[171,-10.2367],[181,-8.0939],[188,-9.3235],[194,17.5226],[199,-5.1913],[205,-11.98],[215,-20.9402],[225,-16.5317],[230,-15.5106],[235,-0.156],[241,-5.358],[248,-4.6111],[266,-8.0939],[282,-11.98],[294,-17.2458],[301,-23.9765],[307,-20.7533],[311,-23.5467],[319,-2.5516],[325,-4.4478],[329,-25.236],[352,-3.6589],[369,-5.0558],[385,-8.0939]]
       }
       hand_l_fk_ctrl.Scale.X {type=MovieSceneScriptingFloatChannel, default=1, const [1.0005,52]}
       hand_l_fk_ctrl.Scale.Y {type=MovieSceneScriptingFloatChannel, default=1, const [1.0005,52]}
       hand_l_fk_ctrl.Scale.Z {type=MovieSceneScriptingFloatChannel, default=1, const [1.0005,52]}
       thumb_01_l_ctrl.Location.X {type=MovieSceneScriptingFloatChannel, default=0, thinned_from=26, keys [[0,0],[385,0]]}
       thumb_01_l_ctrl.Location.Y {type=MovieSceneScriptingFloatChannel, default=0, thinned_from=26, keys [[0,0],[385,0]]}
       thumb_01_l_ctrl.Location.Z {type=MovieSceneScriptingFloatChannel, default=0, thinned_from=26, keys [[0,0],[385,0]]}
       thumb_01_l_ctrl.Rotation.X {
        type=MovieSceneScriptingFloatChannel
        default=21.4449
        keys [[0,-15.4693],[20,15.1169],[26,11.9127],[34,-7.2406],[47,-11.6065],[80,14.9584],[93,18.8064],[104,-11.6065],[118,-5.2772],[128,-4.781],[141,14.3462],[152,-1.146],[167,5.3313],[182,-17.6558],[194,12.1737],[215,-18.9017],[244,24.1333],[255,0.3508],[275,-2.467],[292,3.2157],[313,4.1065],[322,5.5538],[330,-15.4693],[352,-18.9017],[369,14.9584],[385,-15.4693]]
       }
       thumb_01_l_ctrl.Rotation.Y {
        type=MovieSceneScriptingFloatChannel
        default=1.3313
        thinned_from=26
        keys [[0,11.2361],[20,11.8861],[26,18.8937],[34,5.5077],[47,2.6084],[80,12.1082],[93,6.7156],[104,2.6084],[118,4.8058],[128,30.3867],[141,12.9726],[152,6.5251],[167,7.587],[182,-2.6663],[194,4.8154],[215,10.2928],[244,17.7728],[255,13.6391],[275,10.2443],[313,10.2249],[322,17.0139],[330,11.2361],[352,10.2928],[369,12.1082],[385,11.2361]]
       }
       thumb_01_l_ctrl.Rotation.Z {
        type=MovieSceneScriptingFloatChannel
        default=-18.1762
        keys [[0,7.2999],[20,-1.2895],[26,2.1624],[34,-1.7889],[47,-2.0435],[80,-1.2676],[93,-1.7982],[104,-2.0435],[118,-1.8649],[128,0.5623],[141,-1.1614],[152,-1.7168],[167,-1.6268],[182,9.0633],[194,2.2716],[215,7.1711],[244,-6.6619],[255,0.9826],[275,0.3449],[292,-2.0154],[313,-2.3854],[322,8.0891],[330,7.2999],[352,7.1711],[369,-1.2676],[385,7.2999]]
       }
       thumb_01_l_ctrl.Scale.X {type=MovieSceneScriptingFloatChannel, default=1, const [1,26]}
       thumb_01_l_ctrl.Scale.Y {type=MovieSceneScriptingFloatChannel, default=1, const [1,26]}
       thumb_01_l_ctrl.Scale.Z {type=MovieSceneScriptingFloatChannel, default=1, const [1,26]}
       thumb_02_l_ctrl.Location.X {type=MovieSceneScriptingFloatChannel, default=0, const [0,25]}
       thumb_02_l_ctrl.Location.Y {type=MovieSceneScriptingFloatChannel, default=0, const [0,25]}
       thumb_02_l_ctrl.Location.Z {type=MovieSceneScriptingFloatChannel, default=0, const [0,25]}
       thumb_02_l_ctrl.Rotation.X {type=MovieSceneScriptingFloatChannel, default=3.0705, thinned_from=25, keys [[0,0],[385,0]]}
       thumb_02_l_ctrl.Rotation.Y {type=MovieSceneScriptingFloatChannel, default=-2.0172, thinned_from=25, keys [[0,0],[385,0]]}
       thumb_02_l_ctrl.Rotation.Z {
        type=MovieSceneScriptingFloatChannel
        default=20.9383
        thinned_from=25
        keys [[0,-19.1785],[20,2.0124],[26,4.143],[34,2.0124],[118,2.0124],[141,1.9173],[152,1.9924],[167,1.9603],[182,-22.1843],[194,0.1921],[215,-22.1843],[244,20.1427],[255,-3.2486],[275,-4.4952],[292,1.6213],[313,2.58],[322,-0.7677],[330,-19.1785],[352,-22.1843],[369,2.0124],[385,-19.1785]]
       }
       thumb_02_l_ctrl.Scale.X {type=MovieSceneScriptingFloatChannel, default=1, const [1,25]}
       thumb_02_l_ctrl.Scale.Y {type=MovieSceneScriptingFloatChannel, default=1, const [1,25]}
       thumb_02_l_ctrl.Scale.Z {type=MovieSceneScriptingFloatChannel, default=1, const [1,25]}
       thumb_03_l_ctrl.Location.X {type=MovieSceneScriptingFloatChannel, default=0, const [0,25]}
       thumb_03_l_ctrl.Location.Y {type=MovieSceneScriptingFloatChannel, default=0, const [0,25]}
       thumb_03_l_ctrl.Location.Z {type=MovieSceneScriptingFloatChannel, default=0, const [0,25]}
       thumb_03_l_ctrl.Rotation.X {type=MovieSceneScriptingFloatChannel, default=0, thinned_from=25, keys [[0,0],[385,0]]}
       thumb_03_l_ctrl.Rotation.Y {type=MovieSceneScriptingFloatChannel, default=0, thinned_from=25, keys [[0,0],[385,0]]}
       thumb_03_l_ctrl.Rotation.Z {
        type=MovieSceneScriptingFloatChannel
        default=-32.6763
        keys [[0,-30.3305],[20,-6.6459],[26,-4.5153],[34,-39.2071],[47,-46.9935],[80,-6.6459],[93,-6.6459],[104,-46.9935],[118,-37.4375],[141,-6.7385],[152,-30.9745],[167,-22.052],[182,-34.5206],[194,-3.1043],[215,-34.5206],[244,-7.1189],[255,-22.2619],[275,-9.8626],[292,-1.3364],[313,0],[322,-4.6666],[330,-30.3305],[352,-34.5206],[369,-6.6459],[385,-30.3305]]
       }
       thumb_03_l_ctrl.Scale.X {type=MovieSceneScriptingFloatChannel, default=1, const [1,25]}
       thumb_03_l_ctrl.Scale.Y {type=MovieSceneScriptingFloatChannel, default=1, const [1,25]}
       thumb_03_l_ctrl.Scale.Z {type=MovieSceneScriptingFloatChannel, default=1, const [1,25]}
       index_metacarpal_l_ctrl.Location.X {type=MovieSceneScriptingFloatChannel, default=0, const [0,22]}
       index_metacarpal_l_ctrl.Location.Y {type=MovieSceneScriptingFloatChannel, default=0, const [0,22]}
       index_metacarpal_l_ctrl.Location.Z {type=MovieSceneScriptingFloatChannel, default=0, const [0,22]}
       index_metacarpal_l_ctrl.Rotation.X {type=MovieSceneScriptingFloatChannel, default=0.955, thinned_from=22, keys [[0,0],[385,0]]}
       index_metacarpal_l_ctrl.Rotation.Y {type=MovieSceneScriptingFloatChannel, default=3.1321, thinned_from=22, keys [[0,0],[385,0]]}
       index_metacarpal_l_ctrl.Rotation.Z {
        type=MovieSceneScriptingFloatChannel
        default=0.0216
        keys [[0,-3.7343],[20,-3.7915],[26,-0.8985],[34,-2.9011],[47,-3.2511],[80,-3.7343],[93,-5.9385],[104,-3.2511],[118,-3.379],[128,4.7734],[141,-3.3487],[152,-3.3727],[167,-4.0053],[194,-4.3963],[244,-1.3088],[275,-1.5658],[292,-1.4497],[313,-1.4315],[322,-1.4371],[352,-1.507],[369,-3.7343],[385,-3.7343]]
       }
       index_metacarpal_l_ctrl.Scale.X {type=MovieSceneScriptingFloatChannel, default=1, const [1,22]}
       index_metacarpal_l_ctrl.Scale.Y {type=MovieSceneScriptingFloatChannel, default=1, const [1,22]}
       index_metacarpal_l_ctrl.Scale.Z {type=MovieSceneScriptingFloatChannel, default=1, const [1,22]}
       index_01_l_ctrl.Location.X {type=MovieSceneScriptingFloatChannel, default=0, const [0,26]}
       index_01_l_ctrl.Location.Y {type=MovieSceneScriptingFloatChannel, default=0, const [0,26]}
       index_01_l_ctrl.Location.Z {type=MovieSceneScriptingFloatChannel, default=0, const [0,26]}
       index_01_l_ctrl.Rotation.X {type=MovieSceneScriptingFloatChannel, default=0, thinned_from=26, keys [[0,0],[385,0]]}
       index_01_l_ctrl.Rotation.Y {type=MovieSceneScriptingFloatChannel, default=0, thinned_from=26, keys [[0,0],[385,0]]}
       index_01_l_ctrl.Rotation.Z {
        type=MovieSceneScriptingFloatChannel
        default=-42.016
        keys [[0,-65.7549],[20,2.7684],[26,7.2471],[34,-45.0289],[47,-57.3421],[80,2.9111],[93,-0.5552],[104,-57.3421],[118,-43.1054],[128,3.7878],[141,3.5904],[152,-33.2747],[167,-20.4325],[182,-63.7766],[194,-3.5382],[215,-76.1295],[244,10.4011],[255,-37.4184],[275,-26.0277],[292,-8.7035],[313,-5.9881],[322,-2.2107],[330,-65.7549],[352,-76.1295],[369,2.9111],[385,-65.7549]]
       }
       index_01_l_ctrl.Scale.X {type=MovieSceneScriptingFloatChannel, default=1, const [1,26]}
       index_01_l_ctrl.Scale.Y {type=MovieSceneScriptingFloatChannel, default=1, const [1,26]}
       index_01_l_ctrl.Scale.Z {type=MovieSceneScriptingFloatChannel, default=1, const [1,26]}
       index_02_l_ctrl.Location.X {type=MovieSceneScriptingFloatChannel, default=0, const [0,26]}
       index_02_l_ctrl.Location.Y {type=MovieSceneScriptingFloatChannel, default=0, const [0,26]}
       index_02_l_ctrl.Location.Z {type=MovieSceneScriptingFloatChannel, default=0, const [0,26]}
       index_02_l_ctrl.Rotation.X {type=MovieSceneScriptingFloatChannel, default=0, thinned_from=26, keys [[0,0],[385,0]]}
       index_02_l_ctrl.Rotation.Y {type=MovieSceneScriptingFloatChannel, default=5.6922, thinned_from=26, keys [[0,0],[385,0]]}
       index_02_l_ctrl.Rotation.Z {
        type=MovieSceneScriptingFloatChannel
        default=-63.5811
        keys [[0,-34.7174],[20,4.0892],[26,10.2633],[34,-37.1003],[47,-48.1676],[80,4.2859],[93,-0.4935],[104,-48.1676],[118,-35.791],[128,11.6417],[141,5.2026],[152,-27.1608],[167,-16.2606],[182,-29.0857],[194,-5.0203],[215,-41.4386],[244,3.2663],[255,-21.439],[275,-12.8365],[292,-2.9464],[313,-1.3962],[322,6.4501],[330,-34.7174],[352,-41.4386],[369,4.2859],[385,-34.7174]]
       }
       index_02_l_ctrl.Scale.X {type=MovieSceneScriptingFloatChannel, default=1, const [1,26]}
       index_02_l_ctrl.Scale.Y {type=MovieSceneScriptingFloatChannel, default=1, const [1,26]}
       index_02_l_ctrl.Scale.Z {type=MovieSceneScriptingFloatChannel, default=1, const [1,26]}
       index_03_l_ctrl.Location.X {type=MovieSceneScriptingFloatChannel, default=0, const [0,26]}
       index_03_l_ctrl.Location.Y {type=MovieSceneScriptingFloatChannel, default=0, const [0,26]}
       index_03_l_ctrl.Location.Z {type=MovieSceneScriptingFloatChannel, default=0, const [0,26]}
       index_03_l_ctrl.Rotation.X {type=MovieSceneScriptingFloatChannel, default=0, thinned_from=26, keys [[0,0],[385,0]]}
       index_03_l_ctrl.Rotation.Y {type=MovieSceneScriptingFloatChannel, default=0, thinned_from=26, keys [[0,0],[385,0]]}
       index_03_l_ctrl.Rotation.Z {
        type=MovieSceneScriptingFloatChannel
        default=-83.0396
        keys [[0,-71.5903],[20,4.0892],[26,10.2633],[34,-37.1003],[47,-48.1676],[80,4.2859],[93,-0.4935],[104,-48.1676],[118,-35.791],[128,11.6417],[141,5.2427],[152,-27.1523],[167,-16.2476],[182,-81.7998],[194,-4.9999],[215,-81.7998],[244,12.0265],[255,-39.8249],[275,-32.6691],[292,-15.6807],[313,-13.0179],[322,-9.0567],[330,-71.5903],[352,-81.7998],[369,4.2859],[385,-71.5903]]
       }
       index_03_l_ctrl.Scale.X {type=MovieSceneScriptingFloatChannel, default=1, const [1,26]}
       index_03_l_ctrl.Scale.Y {type=MovieSceneScriptingFloatChannel, default=1, const [1,26]}
       index_03_l_ctrl.Scale.Z {type=MovieSceneScriptingFloatChannel, default=1, const [1,26]}
       middle_metacarpal_l_ctrl.Location.X {type=MovieSceneScriptingFloatChannel, default=0, const [0,22]}
       middle_metacarpal_l_ctrl.Location.Y {type=MovieSceneScriptingFloatChannel, default=0, const [0,22]}
       middle_metacarpal_l_ctrl.Location.Z {type=MovieSceneScriptingFloatChannel, default=0, const [0,22]}
       middle_metacarpal_l_ctrl.Rotation.X {type=MovieSceneScriptingFloatChannel, default=0, const [0,22]}
       middle_metacarpal_l_ctrl.Rotation.Y {type=MovieSceneScriptingFloatChannel, default=0, const [0,22]}
       middle_metacarpal_l_ctrl.Rotation.Z {
        type=MovieSceneScriptingFloatChannel
        default=-4.0425
        keys [[0,-9.0007],[20,-9.0579],[26,-6.165],[34,-8.1675],[47,-8.5174],[80,-9.0007],[93,-11.205],[104,-8.5174],[118,-8.6454],[128,-0.4931],[141,-8.6151],[152,-8.6391],[167,-9.2717],[194,-9.6624],[244,-6.5756],[275,-6.8325],[292,-6.7164],[313,-6.6982],[322,-6.7037],[352,-6.7735],[369,-9.0007],[385,-9.0007]]
       }
       middle_metacarpal_l_ctrl.Scale.X {type=MovieSceneScriptingFloatChannel, default=1, const [1,22]}
       middle_metacarpal_l_ctrl.Scale.Y {type=MovieSceneScriptingFloatChannel, default=1, const [1,22]}
       middle_metacarpal_l_ctrl.Scale.Z {type=MovieSceneScriptingFloatChannel, default=1, const [1,22]}
       middle_01_l_ctrl.Location.X {type=MovieSceneScriptingFloatChannel, default=0, const [0,26]}
       middle_01_l_ctrl.Location.Y {type=MovieSceneScriptingFloatChannel, default=0, const [0,26]}
       middle_01_l_ctrl.Location.Z {type=MovieSceneScriptingFloatChannel, default=0, const [0,26]}
       middle_01_l_ctrl.Rotation.X {type=MovieSceneScriptingFloatChannel, default=0, thinned_from=26, keys [[0,0],[385,0]]}
       middle_01_l_ctrl.Rotation.Y {type=MovieSceneScriptingFloatChannel, default=0, thinned_from=26, keys [[0,0],[385,0]]}
       middle_01_l_ctrl.Rotation.Z {
        type=MovieSceneScriptingFloatChannel
        default=-40.2488
        keys [[0,-75.3952],[20,-5.6814],[26,0.2859],[34,-46.9356],[47,-57.9779],[80,-5.49],[93,-10.1093],[104,-57.9779],[118,-45.5919],[128,1.2586],[141,-4.6212],[152,-36.9665],[167,-26.0256],[182,-72.6544],[194,-12.454],[215,-85.0073],[244,1.0222],[255,-46.5204],[275,-39.3926],[292,-23.62],[313,-21.1478],[322,-16.5212],[330,-75.3952],[352,-85.0073],[369,-5.49],[385,-75.3952]]
       }
       middle_01_l_ctrl.Scale.X {type=MovieSceneScriptingFloatChannel, default=1, const [1,26]}
       middle_01_l_ctrl.Scale.Y {type=MovieSceneScriptingFloatChannel, default=1, const [1,26]}
       middle_01_l_ctrl.Scale.Z {type=MovieSceneScriptingFloatChannel, default=1, const [1,26]}
       middle_02_l_ctrl.Location.X {type=MovieSceneScriptingFloatChannel, default=0, const [0,26]}
       middle_02_l_ctrl.Location.Y {type=MovieSceneScriptingFloatChannel, default=0, const [0,26]}
       middle_02_l_ctrl.Location.Z {type=MovieSceneScriptingFloatChannel, default=0, const [0,26]}
       middle_02_l_ctrl.Rotation.X {type=MovieSceneScriptingFloatChannel, default=-0.1946, thinned_from=26, keys [[0,0],[385,0]]}
       middle_02_l_ctrl.Rotation.Y {type=MovieSceneScriptingFloatChannel, default=0.6551, thinned_from=26, keys [[0,0],[385,0]]}
       middle_02_l_ctrl.Rotation.Z {
        type=MovieSceneScriptingFloatChannel
        default=-67.8397
        keys [[0,-44.3728],[20,-5.6814],[26,0.2859],[34,-46.9356],[47,-57.9779],[80,-5.49],[93,-10.1093],[104,-57.9779],[118,-45.5919],[128,1.2586],[141,-4.5644],[152,-36.9545],[167,-26.0071],[182,-38.7469],[194,-14.6748],[215,-51.0998],[244,-6.3143],[255,-31.0642],[275,-22.4639],[292,-12.5623],[313,-11.0102],[322,-3.1703],[330,-44.3728],[352,-51.0998],[369,-5.49],[385,-44.3728]]
       }
       middle_02_l_ctrl.Scale.X {type=MovieSceneScriptingFloatChannel, default=1, const [1,26]}
       middle_02_l_ctrl.Scale.Y {type=MovieSceneScriptingFloatChannel, default=1, const [1,26]}
       middle_02_l_ctrl.Scale.Z {type=MovieSceneScriptingFloatChannel, default=1, const [1,26]}
       middle_03_l_ctrl.Location.X {type=MovieSceneScriptingFloatChannel, default=0, const [0,26]}
       middle_03_l_ctrl.Location.Y {type=MovieSceneScriptingFloatChannel, default=0, const [0,26]}
       middle_03_l_ctrl.Location.Z {type=MovieSceneScriptingFloatChannel, default=0, const [0,26]}
       middle_03_l_ctrl.Rotation.X {type=MovieSceneScriptingFloatChannel, default=0, thinned_from=26, keys [[0,0],[385,0]]}
       middle_03_l_ctrl.Rotation.Y {type=MovieSceneScriptingFloatChannel, default=0, thinned_from=26, keys [[0,0],[385,0]]}
       middle_03_l_ctrl.Rotation.Z {
        type=MovieSceneScriptingFloatChannel
        default=-72.502
        keys [[0,-81.2604],[20,-5.6814],[26,0.2859],[34,-46.9356],[47,-57.9779],[80,-5.49],[93,-10.1093],[104,-57.9779],[118,-45.5919],[128,1.2586],[141,-4.5548],[152,-36.9525],[167,-26.004],[182,-91.4765],[194,-14.6697],[215,-91.4765],[244,2.4419],[255,-49.4604],[275,-42.3073],[292,-25.3056],[313,-22.6407],[322,-18.6868],[330,-81.2604],[352,-91.4765],[369,-5.49],[385,-81.2604]]
       }
       middle_03_l_ctrl.Scale.X {type=MovieSceneScriptingFloatChannel, default=1, const [1,26]}
       middle_03_l_ctrl.Scale.Y {type=MovieSceneScriptingFloatChannel, default=1, const [1,26]}
       middle_03_l_ctrl.Scale.Z {type=MovieSceneScriptingFloatChannel, default=1, const [1,26]}
       ring_metacarpal_l_ctrl.Location.X {type=MovieSceneScriptingFloatChannel, default=0, const [0,22]}
       ring_metacarpal_l_ctrl.Location.Y {type=MovieSceneScriptingFloatChannel, default=0, const [0,22]}
       ring_metacarpal_l_ctrl.Location.Z {type=MovieSceneScriptingFloatChannel, default=0, const [0,22]}
       ring_metacarpal_l_ctrl.Rotation.X {type=MovieSceneScriptingFloatChannel, default=0, thinned_from=22, keys [[0,0],[385,0]]}
       ring_metacarpal_l_ctrl.Rotation.Y {type=MovieSceneScriptingFloatChannel, default=0, thinned_from=22, keys [[0,0],[385,0]]}
       ring_metacarpal_l_ctrl.Rotation.Z {
        type=MovieSceneScriptingFloatChannel
        default=-4.0425
        thinned_from=22
        keys [[0,-15.2327],[20,-15.2902],[26,-12.3973],[34,-14.3992],[47,-14.749],[80,-15.2327],[93,-17.4373],[104,-14.749],[118,-14.8772],[128,-6.7252],[141,-14.8474],[152,-14.8709],[167,-15.5037],[194,-15.8952],[244,-12.8078],[275,-13.0648],[292,-12.9486],[322,-12.936],[352,-13.0058],[369,-15.2327],[385,-15.2327]]
       }
       ring_metacarpal_l_ctrl.Scale.X {type=MovieSceneScriptingFloatChannel, default=1, const [1,22]}
       ring_metacarpal_l_ctrl.Scale.Y {type=MovieSceneScriptingFloatChannel, default=1, const [1,22]}
       ring_metacarpal_l_ctrl.Scale.Z {type=MovieSceneScriptingFloatChannel, default=1, const [1,22]}
       ring_01_l_ctrl.Location.X {type=MovieSceneScriptingFloatChannel, default=0, const [0,26]}
       ring_01_l_ctrl.Location.Y {type=MovieSceneScriptingFloatChannel, default=0, const [0,26]}
       ring_01_l_ctrl.Location.Z {type=MovieSceneScriptingFloatChannel, default=0, const [0,26]}
       ring_01_l_ctrl.Rotation.X {type=MovieSceneScriptingFloatChannel, default=0, thinned_from=26, keys [[0,0],[385,0]]}
       ring_01_l_ctrl.Rotation.Y {type=MovieSceneScriptingFloatChannel, default=0, thinned_from=26, keys [[0,0],[385,0]]}
       ring_01_l_ctrl.Rotation.Z {
        type=MovieSceneScriptingFloatChannel
        default=-36.9
        keys [[0,-83.0551],[20,-11.8937],[26,-4.5186],[34,-52.6956],[47,-63.9072],[80,-11.658],[93,-17.3673],[104,-63.9072],[118,-51.5882],[128,-0.7815],[141,-10.669],[152,-42.9737],[167,-32.3235],[182,-79.477],[194,-19.3161],[215,-91.8298],[244,-6.2913],[255,-53.5626],[275,-51.1429],[292,-37.0742],[313,-34.869],[322,-29.3098],[330,-83.0551],[352,-91.8298],[369,-11.658],[385,-83.0551]]
       }
       ring_01_l_ctrl.Scale.X {type=MovieSceneScriptingFloatChannel, default=1, const [1,26]}
       ring_01_l_ctrl.Scale.Y {type=MovieSceneScriptingFloatChannel, default=1, const [1,26]}
       ring_01_l_ctrl.Scale.Z {type=MovieSceneScriptingFloatChannel, default=1, const [1,26]}
       ring_02_l_ctrl.Location.X {type=MovieSceneScriptingFloatChannel, default=0, const [0,26]}
       ring_02_l_ctrl.Location.Y {type=MovieSceneScriptingFloatChannel, default=0, const [0,26]}
       ring_02_l_ctrl.Location.Z {type=MovieSceneScriptingFloatChannel, default=0, const [0,26]}
       ring_02_l_ctrl.Rotation.X {type=MovieSceneScriptingFloatChannel, default=0, thinned_from=26, keys [[0,0],[385,0]]}
       ring_02_l_ctrl.Rotation.Y {type=MovieSceneScriptingFloatChannel, default=0, thinned_from=26, keys [[0,0],[385,0]]}
       ring_02_l_ctrl.Rotation.Z {
        type=MovieSceneScriptingFloatChannel
        default=-63.2888
        keys [[0,-50.7232],[20,-11.8289],[26,-4.3664],[34,-52.6032],[47,-63.8255],[80,-11.5904],[93,-17.3673],[104,-63.8255],[118,-51.5105],[128,-0.458],[141,-10.5186],[152,-42.8806],[167,-32.1824],[182,-45.3749],[194,-21.3411],[215,-57.7278],[244,-13.4504],[255,-37.9195],[275,-27.4582],[292,-16.9915],[313,-15.351],[322,-7.8203],[330,-50.7232],[352,-57.7278],[369,-11.5904],[385,-50.7232]]
       }
       ring_02_l_ctrl.Scale.X {type=MovieSceneScriptingFloatChannel, default=1, const [1,26]}
       ring_02_l_ctrl.Scale.Y {type=MovieSceneScriptingFloatChannel, default=1, const [1,26]}
       ring_02_l_ctrl.Scale.Z {type=MovieSceneScriptingFloatChannel, default=1, const [1,26]}
       ring_03_l_ctrl.Location.X {type=MovieSceneScriptingFloatChannel, default=0, const [0,26]}
       ring_03_l_ctrl.Location.Y {type=MovieSceneScriptingFloatChannel, default=0, const [0,26]}
       ring_03_l_ctrl.Location.Z {type=MovieSceneScriptingFloatChannel, default=0, const [0,26]}
       ring_03_l_ctrl.Rotation.X {type=MovieSceneScriptingFloatChannel, default=0, thinned_from=26, keys [[0,0],[385,0]]}
       ring_03_l_ctrl.Rotation.Y {type=MovieSceneScriptingFloatChannel, default=0, thinned_from=26, keys [[0,0],[385,0]]}
       ring_03_l_ctrl.Rotation.Z {
        type=MovieSceneScriptingFloatChannel
        default=-53.3451
        keys [[0,-88.3554],[20,-11.84],[26,-4.3923],[34,-52.6191],[47,-63.8396],[80,-11.6019],[93,-17.3673],[104,-63.8396],[118,-51.5239],[128,-0.5131],[141,-10.5527],[152,-42.8984],[167,-32.2001],[182,-98.1264],[194,-21.3566],[215,-98.1264],[244,-4.6989],[255,-56.3299],[275,-51.5767],[292,-35.4808],[313,-32.9579],[322,-28.5083],[330,-88.3554],[352,-98.1264],[369,-11.6019],[385,-88.3554]]
       }
       ring_03_l_ctrl.Scale.X {type=MovieSceneScriptingFloatChannel, default=1, const [1,26]}
       ring_03_l_ctrl.Scale.Y {type=MovieSceneScriptingFloatChannel, default=1, const [1,26]}
       ring_03_l_ctrl.Scale.Z {type=MovieSceneScriptingFloatChannel, default=1, const [1,26]}
       pinky_metacarpal_l_ctrl.Location.X {type=MovieSceneScriptingFloatChannel, default=0, const [0,22]}
       pinky_metacarpal_l_ctrl.Location.Y {type=MovieSceneScriptingFloatChannel, default=0, const [0,22]}
       pinky_metacarpal_l_ctrl.Location.Z {type=MovieSceneScriptingFloatChannel, default=0, const [0,22]}
       pinky_metacarpal_l_ctrl.Rotation.X {type=MovieSceneScriptingFloatChannel, default=0, thinned_from=22, keys [[0,0],[385,0]]}
       pinky_metacarpal_l_ctrl.Rotation.Y {type=MovieSceneScriptingFloatChannel, default=-0.8451, thinned_from=22, keys [[0,0],[385,0]]}
       pinky_metacarpal_l_ctrl.Rotation.Z {
        type=MovieSceneScriptingFloatChannel
        default=-4.0425
        thinned_from=22
        keys [[0,-15.2327],[20,-15.2902],[26,-12.3973],[34,-14.3992],[47,-14.749],[80,-15.2327],[93,-17.4373],[104,-14.749],[118,-14.8772],[128,-6.7252],[141,-14.8474],[152,-14.8709],[167,-15.5037],[194,-15.8952],[244,-12.8078],[275,-13.0648],[292,-12.9486],[322,-12.936],[352,-13.0058],[369,-15.2327],[385,-15.2327]]
       }
       pinky_metacarpal_l_ctrl.Scale.X {type=MovieSceneScriptingFloatChannel, default=1, const [1,22]}
       pinky_metacarpal_l_ctrl.Scale.Y {type=MovieSceneScriptingFloatChannel, default=1, const [1,22]}
       pinky_metacarpal_l_ctrl.Scale.Z {type=MovieSceneScriptingFloatChannel, default=1, const [1,22]}
       pinky_01_l_ctrl.Location.X {type=MovieSceneScriptingFloatChannel, default=0, const [0,26]}
       pinky_01_l_ctrl.Location.Y {type=MovieSceneScriptingFloatChannel, default=0, const [0,26]}
       pinky_01_l_ctrl.Location.Z {type=MovieSceneScriptingFloatChannel, default=0, const [0,26]}
       pinky_01_l_ctrl.Rotation.X {type=MovieSceneScriptingFloatChannel, default=0.0001, thinned_from=26, keys [[0,0],[385,0]]}
       pinky_01_l_ctrl.Rotation.Y {type=MovieSceneScriptingFloatChannel, default=0, thinned_from=26, keys [[0,0],[385,0]]}
       pinky_01_l_ctrl.Rotation.Z {
        type=MovieSceneScriptingFloatChannel
        default=-38.0791
        keys [[0,-84.33],[20,-12.9457],[26,-6.9136],[34,-54.1803],[47,-65.2306],[80,-12.7526],[93,-17.4217],[104,-65.2306],[118,-52.8473],[128,-5.8138],[141,-11.8936],[152,-44.2255],[167,-33.2966],[182,-79.9464],[194,-19.7465],[215,-92.2993],[244,-6.2854],[255,-53.8194],[275,-56.3527],[292,-43.9232],[313,-41.9749],[322,-35.5186],[330,-84.33],[352,-92.2993],[369,-12.7526],[385,-84.33]]
       }
       pinky_01_l_ctrl.Scale.X {type=MovieSceneScriptingFloatChannel, default=1, const [1,26]}
       pinky_01_l_ctrl.Scale.Y {type=MovieSceneScriptingFloatChannel, default=1, const [1,26]}
       pinky_01_l_ctrl.Scale.Z {type=MovieSceneScriptingFloatChannel, default=1, const [1,26]}
       pinky_02_l_ctrl.Location.X {type=MovieSceneScriptingFloatChannel, default=0, const [0,26]}
       pinky_02_l_ctrl.Location.Y {type=MovieSceneScriptingFloatChannel, default=0, const [0,26]}
       pinky_02_l_ctrl.Location.Z {type=MovieSceneScriptingFloatChannel, default=0, const [0,26]}
       pinky_02_l_ctrl.Rotation.X {type=MovieSceneScriptingFloatChannel, default=0, thinned_from=26, keys [[0,0],[385,0]]}
       pinky_02_l_ctrl.Rotation.Y {type=MovieSceneScriptingFloatChannel, default=0, thinned_from=26, keys [[0,0],[385,0]]}
       pinky_02_l_ctrl.Rotation.Z {
        type=MovieSceneScriptingFloatChannel
        default=-57.5584
        keys [[0,-53.36],[20,-12.9454],[26,-6.9128],[34,-54.1785],[47,-65.2286],[80,-12.7519],[93,-17.4217],[104,-65.2286],[118,-52.8457],[128,-5.8118],[141,-11.8531],[152,-44.2157],[167,-33.2826],[182,-46.0478],[194,-21.9768],[215,-58.4007],[244,-13.6253],[255,-38.3696],[275,-39.6884],[292,-33.218],[313,-32.2038],[322,-22.4858],[330,-53.36],[352,-58.4007],[369,-12.7519],[385,-53.36]]
       }
       pinky_02_l_ctrl.Scale.X {type=MovieSceneScriptingFloatChannel, default=1, const [1,26]}
       pinky_02_l_ctrl.Scale.Y {type=MovieSceneScriptingFloatChannel, default=1, const [1,26]}
       pinky_02_l_ctrl.Scale.Z {type=MovieSceneScriptingFloatChannel, default=1, const [1,26]}
       pinky_03_l_ctrl.Location.X {type=MovieSceneScriptingFloatChannel, default=0, const [0,26]}
       pinky_03_l_ctrl.Location.Y {type=MovieSceneScriptingFloatChannel, default=0, const [0,26]}
       pinky_03_l_ctrl.Location.Z {type=MovieSceneScriptingFloatChannel, default=0, const [0,26]}
       pinky_03_l_ctrl.Rotation.X {type=MovieSceneScriptingFloatChannel, default=0, thinned_from=26, keys [[0,0],[385,0]]}
       pinky_03_l_ctrl.Rotation.Y {type=MovieSceneScriptingFloatChannel, default=0, thinned_from=26, keys [[0,0],[385,0]]}
       pinky_03_l_ctrl.Rotation.Z {
        type=MovieSceneScriptingFloatChannel
        default=-64.2111
        keys [[0,-88.9703],[20,-12.9454],[26,-6.9128],[34,-54.1785],[47,-65.2286],[80,-12.7519],[93,-17.4217],[104,-65.2286],[118,-52.8457],[128,-5.8118],[141,-11.8394],[152,-44.2128],[167,-33.2781],[182,-98.7754],[194,-21.9696],[215,-98.7754],[244,-4.8674],[255,-56.7639],[275,-52.025],[292,-35.8596],[313,-33.3258],[322,-28.9141],[330,-88.9703],[352,-98.7754],[369,-12.7519],[385,-88.9703]]
       }
       pinky_03_l_ctrl.Scale.X {type=MovieSceneScriptingFloatChannel, default=1, const [1,26]}
       pinky_03_l_ctrl.Scale.Y {type=MovieSceneScriptingFloatChannel, default=1, const [1,26]}
       pinky_03_l_ctrl.Scale.Z {type=MovieSceneScriptingFloatChannel, default=1, const [1,26]}
       hand_l_ik_ctrl.Location.X {
        type=MovieSceneScriptingFloatChannel
        default=4.1588
        thinned_from=13
        keys [[15,-16.0981],[34,-16.0981],[61,-6.2084],[80,0.5331],[113,-16.0981],[163,-16.0981],[233,0.0989],[264,1.7327],[317,-16.0981],[352,2.3385],[369,0.5331]]
       }
       hand_l_ik_ctrl.Location.Y {
        type=MovieSceneScriptingFloatChannel
        default=8.6278
        thinned_from=12
        keys [[15,19.3138],[34,19.3138],[61,17.9055],[80,17.387],[113,19.3138],[163,19.3138],[233,10.3531],[264,5.6449],[352,4.4956],[369,17.387]]
       }
       hand_l_ik_ctrl.Location.Z {
        type=MovieSceneScriptingFloatChannel
        default=11.0864
        thinned_from=12
        keys [[15,14.4366],[34,14.4366],[61,-14.4508],[80,4.2156],[113,14.4366],[163,14.4366],[233,-1.0157],[264,-1.9752],[352,2.2277],[369,4.2156]]
       }
       hand_l_ik_ctrl.Rotation.X {
        type=MovieSceneScriptingFloatChannel
        default=41.8572
        thinned_from=12
        keys [[15,-67.2042],[34,-67.2042],[61,53.8647],[80,71.502],[113,-67.2042],[163,-67.2042],[233,3.4029],[264,15.9744],[352,30.8489],[369,71.502]]
       }
       hand_l_ik_ctrl.Rotation.Y {
        type=MovieSceneScriptingFloatChannel
        default=2.0272
        thinned_from=12
        keys [[15,74.3337],[34,74.3337],[61,-5.5109],[80,14.6338],[113,74.3337],[163,74.3337],[233,13.0541],[264,8.626],[352,9.6948],[369,14.6338]]
       }
       hand_l_ik_ctrl.Rotation.Z {
        type=MovieSceneScriptingFloatChannel
        default=34.5602
        thinned_from=12
        keys [[15,-45.0588],[34,-45.0588],[61,45.1652],[80,49.1856],[113,-45.0588],[163,-45.0588],[233,8.1684],[264,13.4557],[352,22.489],[369,49.1856]]
       }
       hand_l_ik_ctrl.Scale.X {
        type=MovieSceneScriptingFloatChannel
        default=1
        thinned_from=15
        keys [[15,1],[80,1.0016],[163,1],[264,1.0016],[344,1],[352,1.0016],[369,1.0016]]
       }
       hand_l_ik_ctrl.Scale.Y {
        type=MovieSceneScriptingFloatChannel
        default=1
        thinned_from=15
        keys [[15,1],[80,1.0016],[163,1],[264,1.0016],[344,1],[352,1.0016],[369,1.0016]]
       }
       hand_l_ik_ctrl.Scale.Z {
        type=MovieSceneScriptingFloatChannel
        default=1
        thinned_from=15
        keys [[15,1],[80,1.0016],[163,1],[264,1.0016],[344,1],[352,1.0016],[369,1.0016]]
       }
       Softness {type=MovieSceneScriptingFloatChannel, default=0, thinned_from=32, keys [[0,0],[393,0]]}
       hand_l_ikorient_ctrl.Rotation.Roll {type=MovieSceneScriptingFloatChannel, default=0, thinned_from=15, keys [[0,0],[393,0]]}
       hand_l_ikorient_ctrl.Rotation.Pitch {type=MovieSceneScriptingFloatChannel, default=0, thinned_from=15, keys [[0,0],[393,0]]}
       hand_l_ikorient_ctrl.Rotation.Yaw {type=MovieSceneScriptingFloatChannel, default=0, thinned_from=15, keys [[0,0],[393,0]]}
       arm_l_pv_ik_ctrl.Location.X {
        type=MovieSceneScriptingFloatChannel
        default=17.3769
        thinned_from=13
        keys [[15,-7.0841],[34,-7.0841],[61,13.6077],[80,13.6438],[113,-7.0841],[163,-7.0841],[233,0.5618],[264,7.1105],[317,-7.0841],[352,15.4268],[369,13.6438]]
       }
       arm_l_pv_ik_ctrl.Location.Y {
        type=MovieSceneScriptingFloatChannel
        default=21.5796
        thinned_from=12
        keys [[15,-5.0655],[34,-5.0655],[61,4.0239],[80,20.0187],[113,-5.0655],[163,-5.0655],[233,-0.7831],[264,0.8794],[352,8.6183],[369,20.0187]]
       }
       arm_l_pv_ik_ctrl.Location.Z {
        type=MovieSceneScriptingFloatChannel
        default=-2.1953
        thinned_from=12
        keys [[15,-2.2844],[34,-2.2844],[61,1.2413],[80,-2.2799],[163,-2.2844],[233,-2.8538],[264,1.0898],[352,2.6917],[369,-2.2799]]
       }
       arm_l_pv_ik_ctrl.Rotation.X {
        type=MovieSceneScriptingFloatChannel
        default=-0.7312
        thinned_from=12
        keys [[15,6.7137],[34,6.7137],[61,17.5526],[80,5.3282],[113,6.7137],[163,6.7137],[233,2.6324],[264,10.5219],[352,10.5184],[369,5.3282]]
       }
       arm_l_pv_ik_ctrl.Rotation.Y {
        type=MovieSceneScriptingFloatChannel
        default=-11.7468
        thinned_from=13
        keys [[15,-13.8994],[34,-13.8994],[61,-21.3501],[80,-7.1251],[113,-13.8994],[163,-13.8994],[233,-5.5097],[264,-6.8249],[317,-13.8994],[352,-6.8413],[369,-7.1251]]
       }
       arm_l_pv_ik_ctrl.Rotation.Z {
        type=MovieSceneScriptingFloatChannel
        default=47.7759
        thinned_from=12
        keys [[15,16.907],[34,16.907],[61,58.5239],[80,63.5416],[113,16.907],[163,16.907],[233,15.7041],[264,17.6776],[352,33.8277],[369,63.5416]]
       }
       arm_l_pv_ik_ctrl.Scale.X {type=MovieSceneScriptingFloatChannel, default=1, const [1,12]}
       arm_l_pv_ik_ctrl.Scale.Y {type=MovieSceneScriptingFloatChannel, default=1, const [1,12]}
       arm_l_pv_ik_ctrl.Scale.Z {type=MovieSceneScriptingFloatChannel, default=1, const [1,12]}
       upperarm_r_fk_ctrl.Location.X {type=MovieSceneScriptingFloatChannel, default=0, thinned_from=67, keys [[0,0],[385,0]]}
       upperarm_r_fk_ctrl.Location.Y {type=MovieSceneScriptingFloatChannel, default=0, thinned_from=67, keys [[0,0],[385,0]]}
       upperarm_r_fk_ctrl.Location.Z {type=MovieSceneScriptingFloatChannel, default=0, thinned_from=67, keys [[0,0],[385,0]]}
       upperarm_r_fk_ctrl.Rotation.X {
        type=MovieSceneScriptingFloatChannel
        default=24.195
        keys [[0,45.9215],[4,50.3209],[9,46.3568],[14,30.9805],[18,14.6694],[26,-13.4344],[32,15.5537],[38,36.6086],[43,42.4348],[49,45.7216],[52,46.3192],[55,49.503],[61,23.4207],[68,54.1321],[73,50.3048],[80,38.2953],[85,0.9455],[90,-14.6972],[96,15.5537],[102,36.6086],[107,50.2994],[113,46.4966],[119,34.2493],[124,28.9694],[127,28.7472],[131,38.7624],[134,47.4868],[137,54.6132],[143,42.6345],[151,39.1052],[158,19.1992],[168,24.6017],[180,36.6086],[185,42.771],[192,46.4966],[198,34.2493],[203,28.9694],[206,33.3517],[210,43.8392],[213,48.4135],[216,50.6195],[224,47.909],[231,50.2652],[237,19.1992],[245,6.8138],[253,0.936],[263,16.7486],[271,14.3053],[276,14.6159],[282,21.2706],[288,30.0205],[294,38.9955],[299,43.3989],[305,45.4406],[312,42.2107],[319,43.8546],[327,13.1294],[333,37.373],[339,44.4371],[348,50.5174],[356,54.2796],[363,57.7711],[369,59.4015],[376,49.7462],[381,48.2416],[385,45.9215]]
       }
       upperarm_r_fk_ctrl.Rotation.Y {
        type=MovieSceneScriptingFloatChannel
        default=-5.5277
        keys [[0,0.7038],[4,3.4118],[9,3.8375],[14,3.1452],[18,-5.9771],[26,-11.598],[32,-8.6047],[38,0.0699],[43,8.3591],[49,14.6258],[52,23.5779],[55,31.0463],[61,4.951],[68,-3.6604],[73,-0.0802],[80,6.3945],[85,-8.0981],[90,-9.9342],[96,-8.6047],[102,0.0699],[107,10.1739],[113,21.801],[119,8.5497],[124,-7.8487],[127,-14.9458],[131,-10.0955],[134,-5.6482],[137,-3.8413],[143,-5.3332],[151,-4.5991],[158,-8.2022],[168,-2.5701],[180,0.0699],[185,10.1739],[192,21.801],[198,8.5497],[203,-7.8487],[206,-11.5262],[210,-10.5399],[213,-11.5726],[216,-12.4362],[224,-9.3815],[231,-6.8814],[237,-8.2022],[245,-7.588],[253,-7.0556],[263,-3.2338],[271,-1.2616],[276,-1.6808],[282,-3.933],[288,-7.9422],[294,-9.5391],[299,-8.9749],[305,-8.6031],[312,-7.3548],[319,-3.5549],[327,-8.7468],[333,6.5875],[339,16.1153],[348,23.5742],[356,29.2875],[363,13.3382],[369,-1.5834],[376,-5.1846],[381,-2.0883],[385,0.7038]]
       }
       upperarm_r_fk_ctrl.Rotation.Z {
        type=MovieSceneScriptingFloatChannel
        default=13.1497
        keys [[0,37.9515],[4,38.6466],[9,37.2448],[14,34.6338],[18,12.3029],[26,-8.9321],[32,-17.2769],[38,-23.8065],[43,-26.0393],[49,-28.1047],[52,-29.8008],[55,-22.5007],[68,30.777],[73,36.7897],[80,42.121],[85,15.1511],[90,-2.1807],[96,-17.2769],[102,-23.8065],[107,-22.3569],[113,-18.967],[119,-8.0297],[124,14.7779],[127,26.6204],[131,37.5754],[134,40.6693],[137,43.6477],[143,39.8608],[151,42.7603],[158,19.3382],[168,-2.744],[180,-23.8065],[185,-22.3569],[192,-18.967],[198,-8.0297],[203,9.8481],[206,17.6288],[210,21.3097],[213,24.2394],[216,23.9277],[224,23.698],[231,23.56],[237,10.8159],[245,-10.6614],[253,-18.0471],[263,-15.0999],[271,-14.984],[276,-13.7979],[282,-6.0867],[288,5.015],[294,13.6917],[299,18.3305],[305,20.837],[312,21.533],[319,21.1369],[327,0.9988],[333,-14.2712],[339,-21.2499],[348,-26.8712],[356,-31.7836],[363,-11.8445],[369,14.0733],[376,27.293],[381,34.225],[385,37.9515]]
       }
       upperarm_r_fk_ctrl.Scale.X {type=MovieSceneScriptingFloatChannel, default=1, const [1.0005,66]}
       upperarm_r_fk_ctrl.Scale.Y {type=MovieSceneScriptingFloatChannel, default=1, const [1.0005,66]}
       upperarm_r_fk_ctrl.Scale.Z {type=MovieSceneScriptingFloatChannel, default=1, const [1.0005,66]}
       lowerarm_r_fk_ctrl.Location.X {type=MovieSceneScriptingFloatChannel, default=0, thinned_from=46, keys [[0,0],[385,0]]}
       lowerarm_r_fk_ctrl.Location.Y {type=MovieSceneScriptingFloatChannel, default=0, thinned_from=46, keys [[0,0],[385,0]]}
       lowerarm_r_fk_ctrl.Location.Z {type=MovieSceneScriptingFloatChannel, default=0, thinned_from=46, keys [[0,0],[385,0]]}
       lowerarm_r_fk_ctrl.Rotation.X {type=MovieSceneScriptingFloatChannel, default=-3.8332, thinned_from=26, keys [[0,0],[385,0]]}
       lowerarm_r_fk_ctrl.Rotation.Y {type=MovieSceneScriptingFloatChannel, default=1.5865, thinned_from=26, keys [[0,0],[385,0]]}
       lowerarm_r_fk_ctrl.Rotation.Z {
        type=MovieSceneScriptingFloatChannel
        default=-4.1227
        keys [[0,22.1664],[4,23.7682],[9,27.2336],[18,33.1192],[28,-17.0291],[38,-12.2279],[48,14.7392],[54,46.3721],[61,-15.8914],[69,17.8632],[73,40.0189],[77,44.6456],[84,36.8328],[88,19.1726],[94,-11.2779],[102,14.7392],[114,46.3721],[122,-15.8914],[132,22.0839],[138,31.7422],[147,35.1274],[154,40.655],[171,-7.2092],[181,14.7392],[193,46.3721],[201,-15.8914],[211,13.2411],[217,10.5446],[231,14.2237],[239,4.6508],[255,-9.2545],[263,-2.4702],[277,3.1967],[282,-3.7209],[291,0.2002],[299,3.3012],[304,3.5991],[312,5.9306],[319,7.8801],[330,-5.8998],[334,8.4314],[340,28.1936],[345,25.5697],[357,33.6052],[363,6.5937],[368,-14.3243],[376,1.5094],[381,17.5015],[385,22.1664]]
       }
       lowerarm_r_fk_ctrl.Scale.X {type=MovieSceneScriptingFloatChannel, default=1, const [1.0005,46]}
       lowerarm_r_fk_ctrl.Scale.Y {type=MovieSceneScriptingFloatChannel, default=1, const [1.0005,46]}
       lowerarm_r_fk_ctrl.Scale.Z {type=MovieSceneScriptingFloatChannel, default=1, const [1.0005,46]}
       hand_r_fk_ctrl.Location.X {type=MovieSceneScriptingFloatChannel, default=0, thinned_from=48, keys [[0,0],[385,0]]}
       hand_r_fk_ctrl.Location.Y {
        type=MovieSceneScriptingFloatChannel
        default=0
        thinned_from=48
        keys [[0,-0.4281],[9,-0.4281],[18,-0.2893],[28,-0.1549],[38,-0.0687],[48,-0.0475],[61,-0.1708],[69,-0.419],[78,-0.2893],[88,-0.1549],[98,-0.0687],[108,-0.0475],[115,-0.0946],[121,-0.1708],[126,-0.1549],[131,-0.0933],[136,-0.0933],[152,-0.1708],[158,-0.2186],[171,-0.3354],[181,-0.4051],[187,-0.0475],[194,-0.1163],[200,-0.1708],[205,-0.1549],[215,-0.0436],[230,-0.1708],[234,-0.2743],[239,-0.3754],[248,-0.3842],[255,-0.3875],[282,-0.3875],[294,-0.3938],[314,-0.4073],[320,-0.4188],[329,-0.423],[334,-0.4242],[340,-0.4291],[346,-0.4312],[352,-0.4312],[369,-0.3354],[385,-0.4281]]
       }
       hand_r_fk_ctrl.Location.Z {
        type=MovieSceneScriptingFloatChannel
        default=0
        thinned_from=48
        keys [[0,0.0664],[9,0.0664],[18,0.0448],[28,0.024],[38,0.0068],[48,0.0074],[61,0.0265],[69,0.0649],[78,0.0448],[88,0.024],[98,0.0068],[108,0.0074],[115,0.0148],[121,0.0265],[126,0.024],[131,0.0145],[136,0.0145],[152,0.0265],[158,0.0339],[171,0.052],[181,0.0628],[187,0.0074],[194,0.018],[200,0.0265],[205,0.024],[215,0.0068],[230,0.0265],[234,0.0425],[239,0.0581],[248,0.0594],[255,0.0599],[282,0.0599],[299,0.0612],[314,0.0627],[320,0.0644],[329,0.065],[334,0.0652],[340,0.066],[346,0.0666],[352,0.0668],[369,0.052],[385,0.0664]]
       }
       hand_r_fk_ctrl.Rotation.X {
        type=MovieSceneScriptingFloatChannel
        default=-12.0844
        thinned_from=48
        keys [[0,-1.8776],[9,-1.8776],[18,7.7149],[28,22.2818],[38,28.5438],[48,33.1007],[61,28.3039],[69,1.0086],[78,7.7149],[88,32.1817],[98,28.5438],[108,33.1007],[115,31.3889],[121,21.3262],[126,21.3513],[131,-15.6066],[136,-16.0521],[152,-25.2917],[158,7.2809],[171,-1.7513],[181,-10.7708],[187,33.1007],[194,3.6552],[200,-5.0094],[205,14.8045],[215,12.8412],[230,26.4315],[234,13.2385],[239,2.1854],[248,2.7937],[255,-4.9158],[259,-9.9666],[265,-19.256],[270,-20.8586],[276,-20.8586],[282,-9.9166],[294,-2.4473],[299,-10.201],[314,-11.1764],[320,-11.2447],[329,-9.4447],[334,-2.678],[340,-9.257],[346,-14.0302],[352,-11.1071],[369,8.2797],[385,-1.8776]]
       }
       hand_r_fk_ctrl.Rotation.Y {
        type=MovieSceneScriptingFloatChannel
        default=4.2557
        thinned_from=48
        keys [[0,16.6908],[9,16.6908],[18,11.4991],[28,12.1171],[38,17.409],[48,6.9275],[61,1.9489],[69,8.0489],[78,11.4991],[88,41.5666],[98,17.409],[108,6.9275],[115,3.295],[121,-24.7862],[126,13.8176],[131,21.4326],[136,25.0869],[152,21.532],[158,38.9417],[171,9.8607],[181,-9.7233],[187,6.9275],[194,-17.5008],[200,-9.3443],[205,15.1767],[215,32.7194],[230,17.0676],[234,14.747],[239,27.2963],[248,13.4096],[255,0.0911],[259,-1.1999],[265,-7.6578],[270,-1.4518],[276,-1.4518],[282,-1.8975],[294,2.9997],[299,12.1842],[314,17.0243],[320,25.3038],[329,19.3307],[334,15.8248],[340,15.528],[346,3.1245],[352,5.2739],[369,10.5942],[385,16.6908]]
       }
       hand_r_fk_ctrl.Rotation.Z {
        type=MovieSceneScriptingFloatChannel
        default=-5.7448
        thinned_from=48
        keys [[0,-22.1156],[9,-22.1156],[18,-4.6312],[28,-1.1923],[38,-1.7527],[48,1.0001],[61,-0.4578],[69,-0.4608],[78,-4.6312],[88,3.8079],[98,-1.7527],[108,1.0001],[115,0.1872],[121,7.3314],[126,12.3049],[131,-0.9319],[136,-2.0595],[152,-0.9732],[158,-15.8021],[171,-11.524],[181,-17.6933],[187,1.0001],[194,15.691],[200,-19.8302],[205,-9.1703],[215,-16.6381],[230,1.0289],[234,-1.7132],[239,-14.1914],[248,-19.1831],[255,-12.4967],[259,-18.2998],[265,-12.2643],[270,-13.095],[276,-13.095],[282,-23.3531],[294,-23.2355],[299,-22.7507],[314,-26.832],[320,-36.1578],[329,-30.1702],[334,-19.1911],[340,-13.8328],[352,-13.3364],[369,-10.6159],[385,-22.1156]]
       }
       hand_r_fk_ctrl.Scale.X {type=MovieSceneScriptingFloatChannel, default=1, const [1.0005,48]}
       hand_r_fk_ctrl.Scale.Y {type=MovieSceneScriptingFloatChannel, default=1, const [1.0005,48]}
       hand_r_fk_ctrl.Scale.Z {type=MovieSceneScriptingFloatChannel, default=1, const [1.0005,48]}
       thumb_01_r_ctrl.Location.X {type=MovieSceneScriptingFloatChannel, default=0.324, thinned_from=22, keys [[0,0],[385,0]]}
       thumb_01_r_ctrl.Location.Y {type=MovieSceneScriptingFloatChannel, default=-0.6993, thinned_from=22, keys [[0,0],[385,0]]}
       thumb_01_r_ctrl.Location.Z {type=MovieSceneScriptingFloatChannel, default=0.2082, thinned_from=22, keys [[0,0],[385,0]]}
       thumb_01_r_ctrl.Rotation.X {
        type=MovieSceneScriptingFloatChannel
        default=39.2366
        thinned_from=22
        keys [[0,-11.0865],[10,-11.0865],[25,19.4329],[37,-9.0849],[47,-9.0849],[61,-9.4197],[89,-9.0849],[126,-11.0865],[143,-3.0553],[158,-3.6285],[167,10.8568],[194,-8.1981],[207,-11.0865],[233,-7.4085],[258,18.3701],[282,18.3701],[294,-1.8116],[307,-11.0865],[352,-11.0865],[369,18.9177],[385,-11.0865]]
       }
       thumb_01_r_ctrl.Rotation.Y {
        type=MovieSceneScriptingFloatChannel
        default=-0.8156
        thinned_from=22
        keys [[0,6.853],[10,6.853],[25,6.8897],[37,6.8589],[47,6.8589],[61,1.6584],[89,6.8589],[158,6.8596],[167,13.2531],[194,6.8555],[233,6.8562],[258,7.8365],[282,7.8365],[294,7.1626],[307,6.853],[352,6.853],[369,6.6695],[385,6.853]]
       }
       thumb_01_r_ctrl.Rotation.Z {
        type=MovieSceneScriptingFloatChannel
        default=-16.9537
        thinned_from=22
        keys [[0,0.7557],[10,0.7557],[25,3.962],[37,1.0589],[47,1.0589],[61,3.0204],[89,1.0589],[126,0.7557],[143,1.5991],[158,1.5653],[167,-2.1286],[194,1.059],[207,0.7557],[233,1.1543],[258,-8.5692],[282,-8.5692],[294,-2.1804],[307,0.7557],[352,0.7557],[369,-1.5462],[385,0.7557]]
       }
       thumb_01_r_ctrl.Scale.X {type=MovieSceneScriptingFloatChannel, default=1, const [1,22]}
       thumb_01_r_ctrl.Scale.Y {type=MovieSceneScriptingFloatChannel, default=1, const [1,22]}
       thumb_01_r_ctrl.Scale.Z {type=MovieSceneScriptingFloatChannel, default=1, const [1,22]}
       thumb_02_r_ctrl.Location.X {type=MovieSceneScriptingFloatChannel, default=0, const [0,22]}
       thumb_02_r_ctrl.Location.Y {type=MovieSceneScriptingFloatChannel, default=0, const [0,22]}
       thumb_02_r_ctrl.Location.Z {type=MovieSceneScriptingFloatChannel, default=0, const [0,22]}
       thumb_02_r_ctrl.Rotation.X {
        type=MovieSceneScriptingFloatChannel
        default=6.8139
        thinned_from=22
        keys [[0,0.2467],[25,0.2467],[37,0.2521],[47,0.2521],[61,0.2467],[89,0.2521],[126,0.2467],[233,0.2466],[258,0.0907],[282,0.0907],[294,0.1976],[307,0.2467],[352,0.2467],[369,0.015],[385,0.2467]]
       }
       thumb_02_r_ctrl.Rotation.Y {
        type=MovieSceneScriptingFloatChannel
        default=-7.4121
        thinned_from=22
        keys [[0,2.2017],[25,2.2017],[37,2.2338],[47,2.2338],[61,2.2017],[89,2.2338],[126,2.2017],[158,2.1992],[233,2.2005],[258,-1.1733],[282,-1.1733],[294,1.139],[307,2.2017],[352,2.2017],[369,0.1549],[385,2.2017]]
       }
       thumb_02_r_ctrl.Rotation.Z {
        type=MovieSceneScriptingFloatChannel
        default=20.3412
        thinned_from=22
        keys [[0,8.2659],[25,8.2659],[37,8.4068],[47,8.4068],[61,8.2659],[89,8.4068],[126,8.2659],[233,8.2661],[258,-2.4527],[282,-2.4527],[294,4.891],[307,8.2659],[352,8.2659],[369,2.1738],[385,8.2659]]
       }
       thumb_02_r_ctrl.Scale.X {type=MovieSceneScriptingFloatChannel, default=1, const [1,22]}
       thumb_02_r_ctrl.Scale.Y {type=MovieSceneScriptingFloatChannel, default=1, const [1,22]}
       thumb_02_r_ctrl.Scale.Z {type=MovieSceneScriptingFloatChannel, default=1, const [1,22]}
       thumb_03_r_ctrl.Location.X {type=MovieSceneScriptingFloatChannel, default=0, thinned_from=22, keys [[0,0],[385,0]]}
       thumb_03_r_ctrl.Location.Y {type=MovieSceneScriptingFloatChannel, default=0, thinned_from=22, keys [[0,0],[385,0]]}
       thumb_03_r_ctrl.Location.Z {type=MovieSceneScriptingFloatChannel, default=0, thinned_from=22, keys [[0,0],[385,0]]}
       thumb_03_r_ctrl.Rotation.X {
        type=MovieSceneScriptingFloatChannel
        default=3.7087
        thinned_from=22
        keys [[0,-1.8824],[10,-1.8824],[25,0.2467],[37,-1.7382],[47,-1.7382],[61,-1.1809],[89,-1.7382],[126,-1.8824],[143,-1.3223],[158,-1.3623],[167,0.2467],[194,-1.681],[207,-1.8824],[233,-1.6259],[258,0.0717],[282,0.0717],[294,-1.2671],[307,-1.8824],[352,-1.8824],[369,0.015],[385,-1.8824]]
       }
       thumb_03_r_ctrl.Rotation.Y {
        type=MovieSceneScriptingFloatChannel
        default=-5.4191
        thinned_from=22
        keys [[0,1.0593],[10,1.0593],[25,2.1353],[37,1.164],[47,1.164],[61,1.8086],[89,1.164],[126,1.0593],[143,1.3397],[158,1.3285],[167,2.1353],[194,1.1602],[207,1.0593],[233,1.1918],[258,-0.9277],[282,-0.9277],[294,0.4337],[307,1.0593],[352,1.0593],[369,0.1522],[385,1.0593]]
       }
       thumb_03_r_ctrl.Rotation.Z {
        type=MovieSceneScriptingFloatChannel
        default=-32.9804
        thinned_from=22
        keys [[0,-65.5364],[10,-65.5364],[25,1.6868],[37,-61.0277],[47,-61.0277],[61,-38.0441],[89,-61.0277],[126,-65.5364],[143,-47.8456],[158,-49.1083],[167,1.6868],[194,-47.8456],[207,-65.5364],[233,-57.4346],[258,-15.4155],[282,-15.4155],[294,-49.755],[307,-65.5364],[352,-65.5364],[369,-6.3947],[385,-65.5364]]
       }
       thumb_03_r_ctrl.Scale.X {type=MovieSceneScriptingFloatChannel, default=1, const [1,22]}
       thumb_03_r_ctrl.Scale.Y {type=MovieSceneScriptingFloatChannel, default=1, const [1,22]}
       thumb_03_r_ctrl.Scale.Z {type=MovieSceneScriptingFloatChannel, default=1, const [1,22]}
       index_metacarpal_r_ctrl.Location.X {type=MovieSceneScriptingFloatChannel, default=0, thinned_from=22, keys [[0,0],[385,0]]}
       index_metacarpal_r_ctrl.Location.Y {type=MovieSceneScriptingFloatChannel, default=0, thinned_from=22, keys [[0,0],[385,0]]}
       index_metacarpal_r_ctrl.Location.Z {type=MovieSceneScriptingFloatChannel, default=0, thinned_from=22, keys [[0,0],[385,0]]}
       index_metacarpal_r_ctrl.Rotation.X {type=MovieSceneScriptingFloatChannel, default=0.955, thinned_from=22, keys [[0,0],[385,0]]}
       index_metacarpal_r_ctrl.Rotation.Y {type=MovieSceneScriptingFloatChannel, default=3.1321, thinned_from=22, keys [[0,0],[385,0]]}
       index_metacarpal_r_ctrl.Rotation.Z {
        type=MovieSceneScriptingFloatChannel
        default=0.0216
        thinned_from=22
        keys [[0,1.1791],[10,1.1791],[25,0.0842],[37,2.4651],[47,2.4651],[61,1.1791],[89,2.4651],[126,1.1791],[143,7.1445],[158,6.7187],[167,0.0842],[194,3.3245],[207,1.1791],[233,3.911],[258,-2.884],[282,-2.884],[294,-0.1003],[307,1.1791],[352,1.1791],[369,7.4785],[385,1.1791]]
       }
       index_metacarpal_r_ctrl.Scale.X {type=MovieSceneScriptingFloatChannel, default=1, const [1,22]}
       index_metacarpal_r_ctrl.Scale.Y {type=MovieSceneScriptingFloatChannel, default=1, const [1,22]}
       index_metacarpal_r_ctrl.Scale.Z {type=MovieSceneScriptingFloatChannel, default=1, const [1,22]}
       index_01_r_ctrl.Location.X {type=MovieSceneScriptingFloatChannel, default=-0.0015, thinned_from=22, keys [[0,0],[385,0]]}
       index_01_r_ctrl.Location.Y {type=MovieSceneScriptingFloatChannel, default=-0.4272, thinned_from=22, keys [[0,0],[385,0]]}
       index_01_r_ctrl.Location.Z {type=MovieSceneScriptingFloatChannel, default=0.0739, thinned_from=22, keys [[0,0],[385,0]]}
       index_01_r_ctrl.Rotation.X {type=MovieSceneScriptingFloatChannel, default=8.5327, thinned_from=22, keys [[0,0],[385,0]]}
       index_01_r_ctrl.Rotation.Y {type=MovieSceneScriptingFloatChannel, default=5.3682, thinned_from=22, keys [[0,0],[385,0]]}
       index_01_r_ctrl.Rotation.Z {
        type=MovieSceneScriptingFloatChannel
        default=-45.3409
        thinned_from=22
        keys [[0,-74.6687],[10,-74.6687],[25,-22.8284],[37,-57.7367],[47,-57.7367],[61,-25.6926],[89,-57.7367],[126,-74.6687],[143,-53.9819],[158,-41.5308],[167,-9.2139],[194,-53.9819],[207,-74.6687],[233,-58.7041],[258,3.9839],[282,3.9839],[294,-49.9037],[307,-74.6687],[352,-74.6687],[369,-70.7655],[385,-74.6687]]
       }
       index_01_r_ctrl.Scale.X {type=MovieSceneScriptingFloatChannel, default=1, const [1,22]}
       index_01_r_ctrl.Scale.Y {type=MovieSceneScriptingFloatChannel, default=1, const [1,22]}
       index_01_r_ctrl.Scale.Z {type=MovieSceneScriptingFloatChannel, default=1, const [1,22]}
       index_02_r_ctrl.Location.X {type=MovieSceneScriptingFloatChannel, default=0, thinned_from=22, keys [[0,0],[385,0]]}
       index_02_r_ctrl.Location.Y {type=MovieSceneScriptingFloatChannel, default=0, thinned_from=22, keys [[0,0],[385,0]]}
       index_02_r_ctrl.Location.Z {type=MovieSceneScriptingFloatChannel, default=0, thinned_from=22, keys [[0,0],[385,0]]}
       index_02_r_ctrl.Rotation.X {type=MovieSceneScriptingFloatChannel, default=0.6828, thinned_from=22, keys [[0,0],[385,0]]}
       index_02_r_ctrl.Rotation.Y {type=MovieSceneScriptingFloatChannel, default=5.6513, thinned_from=22, keys [[0,0],[385,0]]}
       index_02_r_ctrl.Rotation.Z {
        type=MovieSceneScriptingFloatChannel
        default=-56.6805
        thinned_from=22
        keys [[0,-60.2778],[10,-60.2778],[25,1.2613],[37,-38.9625],[47,-38.9625],[61,-11.3017],[89,-38.9625],[126,-60.2778],[143,-37.0534],[158,-24.5987],[167,16.4221],[194,-37.0534],[207,-60.2778],[233,-43.065],[258,-5.5202],[282,-5.5202],[294,-43.0365],[307,-60.2778],[352,-60.2778],[369,1.6447],[385,-60.2778]]
       }
       index_02_r_ctrl.Scale.X {type=MovieSceneScriptingFloatChannel, default=1, const [1,22]}
       index_02_r_ctrl.Scale.Y {type=MovieSceneScriptingFloatChannel, default=1, const [1,22]}
       index_02_r_ctrl.Scale.Z {type=MovieSceneScriptingFloatChannel, default=1, const [1,22]}
       index_03_r_ctrl.Location.X {type=MovieSceneScriptingFloatChannel, default=0, thinned_from=22, keys [[0,0],[385,0]]}
       index_03_r_ctrl.Location.Y {type=MovieSceneScriptingFloatChannel, default=0, thinned_from=22, keys [[0,0],[385,0]]}
       index_03_r_ctrl.Location.Z {type=MovieSceneScriptingFloatChannel, default=0, thinned_from=22, keys [[0,0],[385,0]]}
       index_03_r_ctrl.Rotation.X {type=MovieSceneScriptingFloatChannel, default=0, thinned_from=22, keys [[0,0],[385,0]]}
       index_03_r_ctrl.Rotation.Y {type=MovieSceneScriptingFloatChannel, default=0, thinned_from=22, keys [[0,0],[385,0]]}
       index_03_r_ctrl.Rotation.Z {
        type=MovieSceneScriptingFloatChannel
        default=-83.0396
        thinned_from=22
        keys [[0,-73.5041],[10,-73.5041],[25,1.2613],[37,-49.0083],[47,-49.0083],[61,-39.5797],[89,-49.0083],[126,-73.5041],[143,-46.7991],[158,-34.3393],[167,16.4221],[194,-46.7991],[207,-73.5041],[233,-54.5791],[258,-5.5201],[282,-5.5201],[294,-52.0982],[307,-73.5041],[352,-73.5041],[369,1.6447],[385,-73.5041]]
       }
       index_03_r_ctrl.Scale.X {type=MovieSceneScriptingFloatChannel, default=1, const [1,22]}
       index_03_r_ctrl.Scale.Y {type=MovieSceneScriptingFloatChannel, default=1, const [1,22]}
       index_03_r_ctrl.Scale.Z {type=MovieSceneScriptingFloatChannel, default=1, const [1,22]}
       middle_metacarpal_r_ctrl.Location.X {type=MovieSceneScriptingFloatChannel, default=0, const [0,22]}
       middle_metacarpal_r_ctrl.Location.Y {type=MovieSceneScriptingFloatChannel, default=0, const [0,22]}
       middle_metacarpal_r_ctrl.Location.Z {type=MovieSceneScriptingFloatChannel, default=0, const [0,22]}
       middle_metacarpal_r_ctrl.Rotation.X {type=MovieSceneScriptingFloatChannel, default=0, thinned_from=22, keys [[0,0],[385,0]]}
       middle_metacarpal_r_ctrl.Rotation.Y {type=MovieSceneScriptingFloatChannel, default=0, thinned_from=22, keys [[0,0],[385,0]]}
       middle_metacarpal_r_ctrl.Rotation.Z {
        type=MovieSceneScriptingFloatChannel
        default=-4.0425
        thinned_from=22
        keys [[0,-0.2382],[10,-0.2382],[25,-5.4309],[37,0.5633],[47,0.5633],[61,-0.2382],[89,0.5633],[126,-0.2382],[143,4.6558],[158,4.3065],[167,-5.4309],[194,1.522],[207,-0.2382],[233,2.0031],[258,-7.821],[282,-7.821],[294,-2.6257],[307,-0.2382],[352,-0.2382],[369,2.3039],[385,-0.2382]]
       }
       middle_metacarpal_r_ctrl.Scale.X {type=MovieSceneScriptingFloatChannel, default=1, const [1,22]}
       middle_metacarpal_r_ctrl.Scale.Y {type=MovieSceneScriptingFloatChannel, default=1, const [1,22]}
       middle_metacarpal_r_ctrl.Scale.Z {type=MovieSceneScriptingFloatChannel, default=1, const [1,22]}
       middle_01_r_ctrl.Location.X {type=MovieSceneScriptingFloatChannel, default=-0.7341, thinned_from=22, keys [[0,0],[385,0]]}
       middle_01_r_ctrl.Location.Y {type=MovieSceneScriptingFloatChannel, default=0.0877, thinned_from=22, keys [[0,0],[385,0]]}
       middle_01_r_ctrl.Location.Z {type=MovieSceneScriptingFloatChannel, default=-0.0446, thinned_from=22, keys [[0,0],[385,0]]}
       middle_01_r_ctrl.Rotation.X {type=MovieSceneScriptingFloatChannel, default=0, thinned_from=22, keys [[0,0],[385,0]]}
       middle_01_r_ctrl.Rotation.Y {type=MovieSceneScriptingFloatChannel, default=4.1303, thinned_from=22, keys [[0,0],[385,0]]}
       middle_01_r_ctrl.Rotation.Z {
        type=MovieSceneScriptingFloatChannel
        default=-40.2488
        thinned_from=22
        keys [[0,-92.1901],[10,-92.1901],[25,-32.3825],[37,-73.5966],[47,-73.5966],[61,-43.214],[89,-73.5966],[126,-92.1901],[143,-69.4066],[158,-56.9525],[167,-18.768],[194,-69.4066],[207,-92.1901],[233,-75.1941],[258,-5.5701],[282,-5.5701],[294,-64.9164],[307,-92.1901],[352,-92.1901],[369,-7.7119],[385,-92.1901]]
       }
       middle_01_r_ctrl.Scale.X {type=MovieSceneScriptingFloatChannel, default=1, const [1,22]}
       middle_01_r_ctrl.Scale.Y {type=MovieSceneScriptingFloatChannel, default=1, const [1,22]}
       middle_01_r_ctrl.Scale.Z {type=MovieSceneScriptingFloatChannel, default=1, const [1,22]}
       middle_02_r_ctrl.Location.X {type=MovieSceneScriptingFloatChannel, default=0, const [0,22]}
       middle_02_r_ctrl.Location.Y {type=MovieSceneScriptingFloatChannel, default=0, const [0,22]}
       middle_02_r_ctrl.Location.Z {type=MovieSceneScriptingFloatChannel, default=0, const [0,22]}
       middle_02_r_ctrl.Rotation.X {type=MovieSceneScriptingFloatChannel, default=-0.1946, thinned_from=22, keys [[0,0],[385,0]]}
       middle_02_r_ctrl.Rotation.Y {type=MovieSceneScriptingFloatChannel, default=0.6551, thinned_from=22, keys [[0,0],[385,0]]}
       middle_02_r_ctrl.Rotation.Z {
        type=MovieSceneScriptingFloatChannel
        default=-67.8397
        thinned_from=22
        keys [[0,-57.1826],[10,-57.1826],[25,-8.2928],[37,-38.9089],[47,-38.9089],[61,-8.2065],[89,-38.9089],[126,-57.1826],[143,-37.2869],[158,-24.837],[167,6.868],[194,-37.2869],[207,-57.1826],[233,-41.6072],[258,-15.074],[282,-15.074],[294,-43.924],[307,-57.1826],[352,-57.1826],[369,-7.9094],[385,-57.1826]]
       }
       middle_02_r_ctrl.Scale.X {type=MovieSceneScriptingFloatChannel, default=1, const [1,22]}
       middle_02_r_ctrl.Scale.Y {type=MovieSceneScriptingFloatChannel, default=1, const [1,22]}
       middle_02_r_ctrl.Scale.Z {type=MovieSceneScriptingFloatChannel, default=1, const [1,22]}
       middle_03_r_ctrl.Location.X {type=MovieSceneScriptingFloatChannel, default=0, const [0,22]}
       middle_03_r_ctrl.Location.Y {type=MovieSceneScriptingFloatChannel, default=0, const [0,22]}
       middle_03_r_ctrl.Location.Z {type=MovieSceneScriptingFloatChannel, default=0, const [0,22]}
       middle_03_r_ctrl.Rotation.X {type=MovieSceneScriptingFloatChannel, default=0, thinned_from=22, keys [[0,0],[385,0]]}
       middle_03_r_ctrl.Rotation.Y {type=MovieSceneScriptingFloatChannel, default=0, thinned_from=22, keys [[0,0],[385,0]]}
       middle_03_r_ctrl.Rotation.Z {
        type=MovieSceneScriptingFloatChannel
        default=-72.502
        thinned_from=22
        keys [[0,-57.1826],[10,-57.1826],[25,-8.2928],[37,-38.9089],[47,-38.9089],[61,-23.2582],[89,-38.9089],[126,-57.1826],[143,-37.2869],[158,-24.837],[167,6.868],[194,-37.2869],[207,-57.1826],[233,-41.6072],[258,-15.074],[282,-15.074],[294,-43.924],[307,-57.1826],[352,-57.1826],[369,-7.9094],[385,-57.1826]]
       }
       middle_03_r_ctrl.Scale.X {type=MovieSceneScriptingFloatChannel, default=1, const [1,22]}
       middle_03_r_ctrl.Scale.Y {type=MovieSceneScriptingFloatChannel, default=1, const [1,22]}
       middle_03_r_ctrl.Scale.Z {type=MovieSceneScriptingFloatChannel, default=1, const [1,22]}
       ring_metacarpal_r_ctrl.Location.X {type=MovieSceneScriptingFloatChannel, default=0, const [0,22]}
       ring_metacarpal_r_ctrl.Location.Y {type=MovieSceneScriptingFloatChannel, default=0, const [0,22]}
       ring_metacarpal_r_ctrl.Location.Z {type=MovieSceneScriptingFloatChannel, default=0, const [0,22]}
       ring_metacarpal_r_ctrl.Rotation.X {type=MovieSceneScriptingFloatChannel, default=0, thinned_from=22, keys [[0,0],[385,0]]}
       ring_metacarpal_r_ctrl.Rotation.Y {type=MovieSceneScriptingFloatChannel, default=0, thinned_from=22, keys [[0,0],[385,0]]}
       ring_metacarpal_r_ctrl.Rotation.Z {
        type=MovieSceneScriptingFloatChannel
        default=-4.0425
        thinned_from=22
        keys [[0,-5.3042],[10,-5.3042],[25,-11.7418],[37,-4.5419],[47,-4.5419],[61,-5.3042],[89,-4.5419],[126,-5.3042],[143,-0.7469],[158,-1.0722],[167,-11.7418],[194,-3.6652],[207,-5.3042],[233,-3.2171],[258,-12.7433],[282,-12.7433],[294,-7.6465],[307,-5.3042],[352,-5.3042],[369,-3.8605],[385,-5.3042]]
       }
       ring_metacarpal_r_ctrl.Scale.X {type=MovieSceneScriptingFloatChannel, default=1, const [1,22]}
       ring_metacarpal_r_ctrl.Scale.Y {type=MovieSceneScriptingFloatChannel, default=1, const [1,22]}
       ring_metacarpal_r_ctrl.Scale.Z {type=MovieSceneScriptingFloatChannel, default=1, const [1,22]}
       ring_01_r_ctrl.Location.X {type=MovieSceneScriptingFloatChannel, default=-0.4312, thinned_from=22, keys [[0,0],[385,0]]}
       ring_01_r_ctrl.Location.Y {type=MovieSceneScriptingFloatChannel, default=0.0023, thinned_from=22, keys [[0,0],[385,0]]}
       ring_01_r_ctrl.Location.Z {type=MovieSceneScriptingFloatChannel, default=0, thinned_from=22, keys [[0,0],[385,0]]}
       ring_01_r_ctrl.Rotation.X {type=MovieSceneScriptingFloatChannel, default=0, thinned_from=22, keys [[0,0],[385,0]]}
       ring_01_r_ctrl.Rotation.Y {type=MovieSceneScriptingFloatChannel, default=0, thinned_from=22, keys [[0,0],[385,0]]}
       ring_01_r_ctrl.Rotation.Z {
        type=MovieSceneScriptingFloatChannel
        default=-42.3431
        thinned_from=22
        keys [[0,-94.873],[10,-94.873],[25,-39.6949],[37,-77.0002],[47,-77.0002],[61,-45.6461],[89,-77.0002],[126,-94.873],[143,-73.3078],[158,-60.8555],[167,-26.0804],[194,-73.3078],[207,-94.873],[233,-78.4763],[258,-12.8825],[282,-12.8825],[294,-69.057],[307,-94.873],[352,-94.873],[369,-14.9502],[385,-94.873]]
       }
       ring_01_r_ctrl.Scale.X {type=MovieSceneScriptingFloatChannel, default=1, const [1,22]}
       ring_01_r_ctrl.Scale.Y {type=MovieSceneScriptingFloatChannel, default=1, const [1,22]}
       ring_01_r_ctrl.Scale.Z {type=MovieSceneScriptingFloatChannel, default=1, const [1,22]}
       ring_02_r_ctrl.Location.X {type=MovieSceneScriptingFloatChannel, default=0, thinned_from=22, keys [[0,0],[385,0]]}
       ring_02_r_ctrl.Location.Y {type=MovieSceneScriptingFloatChannel, default=0, thinned_from=22, keys [[0,0],[385,0]]}
       ring_02_r_ctrl.Location.Z {type=MovieSceneScriptingFloatChannel, default=0, thinned_from=22, keys [[0,0],[385,0]]}
       ring_02_r_ctrl.Rotation.X {type=MovieSceneScriptingFloatChannel, default=0, thinned_from=22, keys [[0,0],[385,0]]}
       ring_02_r_ctrl.Rotation.Y {type=MovieSceneScriptingFloatChannel, default=0, thinned_from=22, keys [[0,0],[385,0]]}
       ring_02_r_ctrl.Rotation.Z {
        type=MovieSceneScriptingFloatChannel
        default=-63.2888
        thinned_from=22
        keys [[0,-61.2562],[10,-61.2562],[25,-15.6052],[37,-43.7615],[47,-43.7615],[61,-12.0294],[89,-43.7615],[126,-61.2562],[143,-42.2129],[158,-29.7642],[167,-0.4444],[194,-42.2129],[207,-61.2562],[233,-46.1],[258,-22.011],[282,-22.011],[294,-48.8992],[307,-61.2562],[352,-61.2562],[369,-15.2218],[385,-61.2562]]
       }
       ring_02_r_ctrl.Scale.X {type=MovieSceneScriptingFloatChannel, default=1, const [1,22]}
       ring_02_r_ctrl.Scale.Y {type=MovieSceneScriptingFloatChannel, default=1, const [1,22]}
       ring_02_r_ctrl.Scale.Z {type=MovieSceneScriptingFloatChannel, default=1, const [1,22]}
       ring_03_r_ctrl.Location.X {type=MovieSceneScriptingFloatChannel, default=0, const [0,22]}
       ring_03_r_ctrl.Location.Y {type=MovieSceneScriptingFloatChannel, default=0, const [0,22]}
       ring_03_r_ctrl.Location.Z {type=MovieSceneScriptingFloatChannel, default=0, const [0,22]}
       ring_03_r_ctrl.Rotation.X {type=MovieSceneScriptingFloatChannel, default=0, thinned_from=22, keys [[0,0],[385,0]]}
       ring_03_r_ctrl.Rotation.Y {type=MovieSceneScriptingFloatChannel, default=0, thinned_from=22, keys [[0,0],[385,0]]}
       ring_03_r_ctrl.Rotation.Z {
        type=MovieSceneScriptingFloatChannel
        default=-53.3451
        thinned_from=22
        keys [[0,-61.2562],[10,-61.2562],[25,-15.6052],[37,-43.7615],[47,-43.7615],[61,-40.6345],[89,-43.7615],[126,-61.2562],[143,-42.2129],[158,-29.7642],[167,-0.4444],[194,-42.2129],[207,-61.2562],[233,-46.1],[258,-20.7731],[282,-20.7731],[294,-48.5095],[307,-61.2562],[352,-61.2562],[369,-15.2218],[385,-61.2562]]
       }
       ring_03_r_ctrl.Scale.X {type=MovieSceneScriptingFloatChannel, default=1, const [1,22]}
       ring_03_r_ctrl.Scale.Y {type=MovieSceneScriptingFloatChannel, default=1, const [1,22]}
       ring_03_r_ctrl.Scale.Z {type=MovieSceneScriptingFloatChannel, default=1, const [1,22]}
       pinky_metacarpal_r_ctrl.Location.X {type=MovieSceneScriptingFloatChannel, default=0, const [0,22]}
       pinky_metacarpal_r_ctrl.Location.Y {type=MovieSceneScriptingFloatChannel, default=0, thinned_from=22, keys [[0,0],[385,0]]}
       pinky_metacarpal_r_ctrl.Location.Z {type=MovieSceneScriptingFloatChannel, default=0, thinned_from=22, keys [[0,0],[385,0]]}
       pinky_metacarpal_r_ctrl.Rotation.X {type=MovieSceneScriptingFloatChannel, default=0, thinned_from=22, keys [[0,0],[385,0]]}
       pinky_metacarpal_r_ctrl.Rotation.Y {type=MovieSceneScriptingFloatChannel, default=-0.8451, thinned_from=22, keys [[0,0],[385,0]]}
       pinky_metacarpal_r_ctrl.Rotation.Z {
        type=MovieSceneScriptingFloatChannel
        default=-4.0425
        thinned_from=22
        keys [[0,-13.7503],[10,-13.7503],[25,-11.7418],[37,-11.253],[47,-11.253],[61,-13.7503],[89,-11.253],[126,-13.7503],[143,-6.9704],[158,-7.4543],[167,-11.7418],[194,-11.3119],[207,-13.7503],[233,-10.6453],[258,-12.7433],[282,-12.7433],[294,-13.4332],[307,-13.7503],[352,-13.7503],[369,-3.8605],[385,-13.7503]]
       }
       pinky_metacarpal_r_ctrl.Scale.X {type=MovieSceneScriptingFloatChannel, default=1, const [1,22]}
       pinky_metacarpal_r_ctrl.Scale.Y {type=MovieSceneScriptingFloatChannel, default=1, const [1,22]}
       pinky_metacarpal_r_ctrl.Scale.Z {type=MovieSceneScriptingFloatChannel, default=1, const [1,22]}
       pinky_01_r_ctrl.Location.X {type=MovieSceneScriptingFloatChannel, default=0, const [0,22]}
       pinky_01_r_ctrl.Location.Y {type=MovieSceneScriptingFloatChannel, default=0, const [0,22]}
       pinky_01_r_ctrl.Location.Z {type=MovieSceneScriptingFloatChannel, default=0, const [0,22]}
       pinky_01_r_ctrl.Rotation.X {type=MovieSceneScriptingFloatChannel, default=0.0001, thinned_from=22, keys [[0,0],[385,0]]}
       pinky_01_r_ctrl.Rotation.Y {type=MovieSceneScriptingFloatChannel, default=0, thinned_from=22, keys [[0,0],[385,0]]}
       pinky_01_r_ctrl.Rotation.Z {
        type=MovieSceneScriptingFloatChannel
        default=-38.0791
        thinned_from=22
        keys [[0,-80.5159],[10,-80.5159],[25,-39.6949],[37,-65.8766],[47,-65.8766],[61,-44.8425],[89,-65.8766],[126,-80.5159],[143,-62.7289],[158,-50.282],[167,-26.0804],[194,-62.7289],[207,-80.5159],[233,-65.9777],[258,-12.8825],[282,-12.8825],[294,-59.2204],[307,-80.5159],[352,-80.5159],[369,-15.0417],[385,-80.5159]]
       }
       pinky_01_r_ctrl.Scale.X {type=MovieSceneScriptingFloatChannel, default=1, const [1,22]}
       pinky_01_r_ctrl.Scale.Y {type=MovieSceneScriptingFloatChannel, default=1, const [1,22]}
       pinky_01_r_ctrl.Scale.Z {type=MovieSceneScriptingFloatChannel, default=1, const [1,22]}
       pinky_02_r_ctrl.Location.X {type=MovieSceneScriptingFloatChannel, default=0, const [0,22]}
       pinky_02_r_ctrl.Location.Y {type=MovieSceneScriptingFloatChannel, default=0, const [0,22]}
       pinky_02_r_ctrl.Location.Z {type=MovieSceneScriptingFloatChannel, default=0, const [0,22]}
       pinky_02_r_ctrl.Rotation.X {type=MovieSceneScriptingFloatChannel, default=0, thinned_from=22, keys [[0,0],[385,0]]}
       pinky_02_r_ctrl.Rotation.Y {type=MovieSceneScriptingFloatChannel, default=0, thinned_from=22, keys [[0,0],[385,0]]}
       pinky_02_r_ctrl.Rotation.Z {
        type=MovieSceneScriptingFloatChannel
        default=-57.5584
        thinned_from=22
        keys [[0,-74.119],[10,-74.119],[25,-15.6052],[37,-53.5312],[47,-53.5312],[61,-38.4509],[89,-53.5312],[126,-74.119],[143,-51.6907],[158,-39.2371],[167,-0.4444],[194,-51.6907],[207,-74.119],[233,-57.2978],[258,-21.2967],[282,-21.2967],[294,-57.4871],[307,-74.119],[352,-74.119],[369,-15.2218],[385,-74.119]]
       }
       pinky_02_r_ctrl.Scale.X {type=MovieSceneScriptingFloatChannel, default=1, const [1,22]}
       pinky_02_r_ctrl.Scale.Y {type=MovieSceneScriptingFloatChannel, default=1, const [1,22]}
       pinky_02_r_ctrl.Scale.Z {type=MovieSceneScriptingFloatChannel, default=1, const [1,22]}
       pinky_03_r_ctrl.Location.X {type=MovieSceneScriptingFloatChannel, default=0, const [0,22]}
       pinky_03_r_ctrl.Location.Y {type=MovieSceneScriptingFloatChannel, default=0, const [0,22]}
       pinky_03_r_ctrl.Location.Z {type=MovieSceneScriptingFloatChannel, default=0, const [0,22]}
       pinky_03_r_ctrl.Rotation.X {type=MovieSceneScriptingFloatChannel, default=0, thinned_from=22, keys [[0,0],[385,0]]}
       pinky_03_r_ctrl.Rotation.Y {type=MovieSceneScriptingFloatChannel, default=0, thinned_from=22, keys [[0,0],[385,0]]}
       pinky_03_r_ctrl.Rotation.Z {
        type=MovieSceneScriptingFloatChannel
        default=-64.2111
        thinned_from=22
        keys [[0,-74.119],[10,-74.119],[25,-15.6052],[37,-53.5312],[89,-53.5312],[126,-74.119],[143,-51.6907],[158,-39.2371],[167,-0.4444],[194,-51.6907],[207,-74.119],[233,-57.2978],[258,-22.1256],[282,-22.1256],[294,-57.7481],[307,-74.119],[352,-74.119],[369,-15.2218],[385,-74.119]]
       }
       pinky_03_r_ctrl.Scale.X {type=MovieSceneScriptingFloatChannel, default=1, const [1,22]}
       pinky_03_r_ctrl.Scale.Y {type=MovieSceneScriptingFloatChannel, default=1, const [1,22]}
       pinky_03_r_ctrl.Scale.Z {type=MovieSceneScriptingFloatChannel, default=1, const [1,22]}
       hand_r_ik_ctrl.Location.X {
        type=MovieSceneScriptingFloatChannel
        default=-11.1389
        thinned_from=13
        keys [[15,-5.8125],[34,-5.8125],[61,-1.1024],[80,-4.8395],[113,-5.8125],[163,-5.8125],[233,2.7964],[264,-1.4651],[342,-12.7316],[352,-5.6189],[369,-4.8395]]
       }
       hand_r_ik_ctrl.Location.Y {
        type=MovieSceneScriptingFloatChannel
        default=19.728
        thinned_from=12
        keys [[15,5.4966],[34,5.4966],[61,14.8362],[80,11.5437],[113,5.4966],[163,5.4966],[233,4.4349],[264,6.9773],[352,8.1399],[369,11.5437]]
       }
       hand_r_ik_ctrl.Location.Z {
        type=MovieSceneScriptingFloatChannel
        default=-7.4919
        thinned_from=12
        keys [[15,16.9637],[34,16.9637],[61,8.1921],[80,-10.0037],[113,16.9637],[163,16.9637],[233,7.8554],[264,-2.414],[352,-6.7536],[369,-10.0037]]
       }
       hand_r_ik_ctrl.Rotation.X {
        type=MovieSceneScriptingFloatChannel
        default=-16.9227
        thinned_from=14
        keys [[15,-40.0387],[34,-40.0387],[61,71.0052],[80,60.9758],[113,-40.0387],[163,-40.0387],[233,21.405],[264,9.4386],[338,-69.7451],[342,-70.4677],[352,-4.7772],[369,60.9758]]
       }
       hand_r_ik_ctrl.Rotation.Y {
        type=MovieSceneScriptingFloatChannel
        default=4.2254
        thinned_from=12
        keys [[15,81.9355],[34,81.9355],[61,14.0652],[80,3.9532],[113,81.9355],[163,81.9355],[233,25.1504],[264,12.9048],[352,8.2509],[369,3.9532]]
       }
       hand_r_ik_ctrl.Rotation.Z {
        type=MovieSceneScriptingFloatChannel
        default=-5.7911
        thinned_from=13
        keys [[15,-25.2719],[34,-25.2719],[61,48.5474],[80,41.6174],[113,-25.2719],[163,-25.2719],[233,14.5892],[264,8.4903],[342,-63.6685],[352,-0.7879],[369,41.6174]]
       }
       hand_r_ik_ctrl.Scale.X {
        type=MovieSceneScriptingFloatChannel
        default=1
        thinned_from=16
        keys [[15,1],[80,1.0016],[163,1],[264,1.0016],[348,1],[352,1.0016],[369,1.0016]]
       }
       hand_r_ik_ctrl.Scale.Y {
        type=MovieSceneScriptingFloatChannel
        default=1
        thinned_from=16
        keys [[15,1],[80,1.0016],[163,1],[264,1.0016],[348,1],[352,1.0016],[369,1.0016]]
       }
       hand_r_ik_ctrl.Scale.Z {
        type=MovieSceneScriptingFloatChannel
        default=1
        thinned_from=16
        keys [[15,1],[80,1.0016],[163,1],[264,1.0016],[348,1],[352,1.0016],[369,1.0016]]
       }
       Softness_2 {type=MovieSceneScriptingFloatChannel, default=0, thinned_from=33, keys [[0,0],[393,0]]}
       hand_r_ikorient_ctrl.Rotation.Roll {type=MovieSceneScriptingFloatChannel, default=0, thinned_from=15, keys [[0,0],[393,0]]}
       hand_r_ikorient_ctrl.Rotation.Pitch {type=MovieSceneScriptingFloatChannel, default=0, thinned_from=15, keys [[0,0],[393,0]]}
       hand_r_ikorient_ctrl.Rotation.Yaw {type=MovieSceneScriptingFloatChannel, default=0, thinned_from=15, keys [[0,0],[393,0]]}
       arm_r_pv_ik_ctrl.Location.X {
        type=MovieSceneScriptingFloatChannel
        default=-20.3856
        thinned_from=12
        keys [[15,6.6981],[34,6.6981],[61,16.9347],[80,20.7785],[113,6.6981],[163,6.6981],[233,8.2057],[264,-0.492],[352,-11.3801],[369,20.7785]]
       }
       arm_r_pv_ik_ctrl.Location.Y {
        type=MovieSceneScriptingFloatChannel
        default=-11.2154
        thinned_from=13
        keys [[15,1.9083],[34,1.9083],[61,17.8583],[80,7.1029],[113,1.9083],[163,1.9083],[233,6.4019],[264,-4.5472],[342,-8.6042],[352,-10.006],[369,7.1029]]
       }
       arm_r_pv_ik_ctrl.Location.Z {
        type=MovieSceneScriptingFloatChannel
        default=-2.2167
        thinned_from=13
        keys [[15,-0.1935],[34,-0.1935],[61,-0.3268],[80,4.1657],[113,-0.1935],[163,-0.1935],[233,0.0189],[264,2.1697],[348,-2.5783],[352,3.5085],[369,4.1657]]
       }
       arm_r_pv_ik_ctrl.Rotation.X {
        type=MovieSceneScriptingFloatChannel
        default=-4.3577
        thinned_from=12
        keys [[15,8.3099],[34,8.3099],[61,6.7333],[80,18.9731],[113,8.3099],[163,8.3099],[233,5.1846],[264,11.0041],[352,11.0082],[369,18.9731]]
       }
       arm_r_pv_ik_ctrl.Rotation.Y {
        type=MovieSceneScriptingFloatChannel
        default=-2.123
        thinned_from=13
        keys [[15,-9.858],[34,-9.858],[61,-5.2424],[80,-18.4384],[113,-9.858],[163,-9.858],[233,-2.4279],[264,-3.4975],[338,-11.8577],[352,-3.4813],[369,-18.4384]]
       }
       arm_r_pv_ik_ctrl.Rotation.Z {
        type=MovieSceneScriptingFloatChannel
        default=1.5482
        thinned_from=12
        keys [[15,14.7064],[34,14.7064],[61,56.9884],[80,59.218],[113,14.7064],[163,14.7064],[233,20.0052],[264,8.6084],[352,-7.5415],[369,59.218]]
       }
       arm_r_pv_ik_ctrl.Scale.X {type=MovieSceneScriptingFloatChannel, default=1, const [1,12]}
       arm_r_pv_ik_ctrl.Scale.Y {type=MovieSceneScriptingFloatChannel, default=1, const [1,12]}
       arm_r_pv_ik_ctrl.Scale.Z {type=MovieSceneScriptingFloatChannel, default=1, const [1,12]}
       body_ctrl.Location.X {
        type=MovieSceneScriptingFloatChannel
        default=0
        keys [[0,-5.0993],[5,-5.0792],[10,-4.9522],[15,-4.0086],[20,0.1283],[25,2.7493],[35,4.4325],[45,4.1396],[50,3.6741],[55,1.9178],[60,-1.8076],[65,-5.5841],[70,-7.1414],[75,-7.654],[80,-7.4126],[85,-6.1606],[90,-3.6346],[95,0.1168],[100,2.9784],[105,4.1206],[110,4.4138],[115,4.595],[120,2.3952],[130,-4.3052],[135,-5.1109],[140,-5.2596],[150,-4.2592],[155,-2.2445],[160,0.1283],[165,2.6637],[170,3.7911],[175,4.1592],[180,4.259],[185,4.1203],[190,3.7829],[195,2.2042],[200,-0.5942],[205,-2.4465],[210,-3.0921],[215,-3.3643],[220,-3.4733],[225,-3.496],[230,-3.3607],[235,-2.6639],[240,-1.4612],[245,0.2229],[250,1.9664],[255,3.0246],[260,3.3931],[265,3.5454],[270,3.373],[275,2.9529],[280,2.2099],[285,0.7506],[300,-2.6209],[305,-3.484],[310,-3.936],[315,-3.44],[320,-1.5116],[327,1.9678],[335,4.2924],[351,5.2942],[356,4.8423],[366,-0.0673],[370,-1.8963],[376,-4.1581],[381,-5.0001],[385,-5.0993]]
       }
       body_ctrl.Location.Y {
        type=MovieSceneScriptingFloatChannel
        default=0
        thinned_from=79
        keys [[0,-0.695],[5,-0.6719],[10,-0.6291],[15,-1.1177],[20,-1.966],[25,-1.7751],[30,-1.1855],[35,-0.7645],[40,-0.6472],[45,-0.6335],[50,-0.7567],[55,-1.2661],[60,-1.9141],[65,-2.0498],[70,-1.8573],[75,-1.5333],[80,-1.1422],[85,-0.6792],[90,-0.3001],[95,-0.0716],[100,-0.02],[105,-0.1335],[110,-0.4291],[115,-0.9229],[120,-1.632],[125,-1.6957],[130,-1.4579],[140,-0.839],[145,-0.6215],[150,-0.6028],[155,-1.1244],[160,-1.966],[165,-1.775],[170,-1.1861],[175,-0.765],[180,-0.6477],[185,-0.6364],[190,-0.7563],[195,-1.2616],[200,-1.9131],[205,-2.0255],[210,-1.5431],[215,-0.9397],[220,-0.6739],[225,-0.5765],[230,-0.5386],[235,-0.6674],[240,-0.9474],[245,-0.7226],[250,-0.3015],[255,-0.3632],[260,-0.6235],[265,-0.6193],[270,-0.488],[275,-0.5851],[280,-0.8764],[285,-0.95],[290,-0.9142],[295,-0.8504],[300,-0.7677],[305,-0.6691],[310,-0.607],[315,-0.9583],[320,-1.8649],[325,-1.8912],[330,-1.2808],[335,-0.781],[341,-0.695],[346,-0.695],[351,-0.725],[356,-1.0569],[361,-1.6526],[366,-1.9105],[370,-1.4985],[371,-1.3398],[376,-0.7835],[381,-0.7064],[385,-0.695]]
       }
       body_ctrl.Location.Z {
        type=MovieSceneScriptingFloatChannel
        default=-1.7064
        thinned_from=78
        keys [[0,-2.7447],[5,-1.464],[10,-1.0598],[15,-2.226],[20,-4.1995],[25,-3.3169],[31,-0.9463],[35,-1.1384],[40,-2.0103],[45,-0.87],[50,-1.1084],[55,-3.3316],[60,-7.7531],[65,-4.8559],[70,-2.3813],[75,-1.5941],[80.6288,-2.8814],[85,-2.311],[90,-4.2079],[95,-7.2715],[100,-3.7262],[105,-1.7537],[108.7413,-1.5044],[115,-2.4975],[120,-5.3394],[125,-2.0841],[130,-0.9286],[135,-1.2239],[140,-2.389],[145,-1.8447],[150,-1.0598],[155,-2.5351],[160,-5.977],[165,-3.3169],[170,-1.0663],[175,-1.4582],[180,-2.2343],[185,-0.87],[190,-1.1084],[195,-2.6846],[200,-4.6442],[205,-2.593],[210,-1.7619],[215,-2.2356],[220,-3.6566],[225,-2.0406],[230,-1.8436],[235,-2.2924],[240,-3.5356],[245,-2.5816],[250,-2.2627],[255,-2.4937],[260,-3.1855],[265,-1.8921],[270,-1.5266],[275,-1.9114],[280,-2.8923],[285,-2.0645],[290,-1.1732],[293,-1.0199],[297,-1.262],[302,-2.438],[310,-1.0239],[315,-1.5986],[320,-2.9435],[330,-1.6081],[335,-1.2648],[341,-2.5933],[344,-1.5127],[349,-1.0564],[354,-1.4521],[360,-4.1501],[365,-1.9928],[370,-1.0789],[376,-0.8705],[381,-1.5719],[385,-2.7447]]
       }
       body_ctrl.Rotation.X {
        type=MovieSceneScriptingFloatChannel
        default=0
        thinned_from=78
        keys [[0,1.0728],[130,1.0728],[135,0.9701],[140,1.0728],[356,1.0728],[361,1.1016],[366,1.0728],[385,1.0728]]
       }
       body_ctrl.Rotation.Y {
        type=MovieSceneScriptingFloatChannel
        default=0
        keys [[0,0.9928],[5,0.9609],[10,0.7503],[15,0.4453],[20,-0.0131],[25,-0.3923],[30,-0.6978],[35,-0.9003],[40,-0.8608],[45,-0.6881],[50,-0.5932],[55,-1.4597],[60,-3.3335],[65,-3.7413],[70,-3.3073],[75,-2.7058],[80,-2.5265],[85,-2.6192],[90,-2.5294],[95,-2.0938],[100,-1.5771],[105,-1.001],[110,-0.4916],[115,-0.9586],[120,-2.1341],[125,-2.5967],[130,-0.5401],[135,0.7483],[140,1.0188],[145,0.8601],[150,0.6268],[155,0.3215],[160,-0.0315],[165,-0.394],[170,-0.6971],[175,-0.9003],[180,-0.8608],[185,-0.6881],[190,-0.5932],[195,-0.827],[200,-1.6728],[205,-0.8432],[213.0512,0.4614],[218.1687,0.8498],[225,0.8561],[230,0.9773],[235,1.7614],[240,1.1045],[245,0.8687],[250,0.8646],[255,0.5555],[265,-0.5188],[270,-1.5544],[275,-1.8952],[280,-1.72],[285,-0.418],[290,1.013],[295,1.3052],[300,0.5257],[305,0.0466],[310,-0.1121],[315,0.4722],[320,1.2464],[325,1.5909],[330,0.4042],[335,-0.7346],[341,-0.9244],[346,-0.854],[351,-0.7681],[356,-0.8604],[361,-1.9558],[366,-1.2174],[371,0.1724],[376,0.7113],[381,0.9114],[385,0.9928]]
       }
       body_ctrl.Rotation.Z {
        type=MovieSceneScriptingFloatChannel
        default=0
        keys [[0,-5.4214],[5,-5.3931],[10,-4.5546],[15,-0.9883],[20,7.1837],[25,15.1118],[30,20.6696],[35,21.8702],[40,21.763],[45,20.1112],[50,15.1153],[55,6.9618],[60,-0.7641],[65,-4.5579],[70,-5.1218],[75,-5.1059],[80,-4.6997],[85,-0.4596],[90,7.1739],[95,13.9673],[100,14.8744],[105,12.8041],[110,9.1334],[115,4.3228],[120,-0.8421],[125,-5.7356],[130,-9.6555],[135,-11.4782],[140,-10.681],[145,-7.9571],[150,-5.0181],[155,-1.8225],[160,4.7187],[165,14.8827],[170,20.6696],[175,21.8702],[180,21.763],[185,20.1112],[190,17.3473],[195,11.4451],[200,-0.0335],[205,-4.5579],[210,-5.1247],[215,-5.1155],[220,-4.9014],[225,-4.5028],[230,-3.6173],[235,-0.4141],[240,13.522],[245,21.6486],[250,25.1756],[265,27.8352],[275,27.5296],[280,26.1883],[285,19.091],[295,3.557],[300,2.3359],[305,2.2131],[315,2.3134],[320,2.9438],[325,9.8628],[330,19.1077],[335,21.3077],[344,22.0077],[351,21.6964],[356,19.3459],[367,-0.3085],[370,-2.5385],[381,-5.2302],[385,-5.4214]]
       }
       body_ctrl.Scale.X {type=MovieSceneScriptingFloatChannel, default=1, const [1,79]}
       body_ctrl.Scale.Y {type=MovieSceneScriptingFloatChannel, default=1, const [1,79]}
       body_ctrl.Scale.Z {type=MovieSceneScriptingFloatChannel, default=1, const [1,79]}
       hips_ctrl.Location.X {type=MovieSceneScriptingFloatChannel, default=0, thinned_from=15, keys [[0,0],[393,0]]}
       hips_ctrl.Location.Y {type=MovieSceneScriptingFloatChannel, default=0, thinned_from=15, keys [[0,0],[393,0]]}
       hips_ctrl.Location.Z {type=MovieSceneScriptingFloatChannel, default=0, thinned_from=15, keys [[0,0],[393,0]]}
       hips_ctrl.Rotation.X {type=MovieSceneScriptingFloatChannel, default=0, thinned_from=15, keys [[0,0],[393,0]]}
       hips_ctrl.Rotation.Y {type=MovieSceneScriptingFloatChannel, default=0, thinned_from=15, keys [[0,0],[393,0]]}
       hips_ctrl.Rotation.Z {type=MovieSceneScriptingFloatChannel, default=0, const [4.418,14]}
       hips_ctrl.Scale.X {type=MovieSceneScriptingFloatChannel, default=1, const [1,14]}
       hips_ctrl.Scale.Y {type=MovieSceneScriptingFloatChannel, default=1, const [1,14]}
       hips_ctrl.Scale.Z {type=MovieSceneScriptingFloatChannel, default=1, const [1,14]}
       hips_tan_ctrl.Location.X {
        type=MovieSceneScriptingFloatChannel
        default=0
        keys [[0,-0.0322],[30,-0.0125],[37,-0.0165],[52,-0.0005],[104,-0.0054],[109,-0.0054],[170,-0.0125],[177,-0.0165],[192,-0.0005],[230,-0.0019],[331,-0.0125],[347,-0.0322],[350,-0.0165],[357,-0.0125]]
       }
       hips_tan_ctrl.Location.Y {
        type=MovieSceneScriptingFloatChannel
        default=0.008
        keys [[0,0.2517],[30,0.4038],[37,0.4038],[52,0.0204],[104,0.1371],[109,0.1371],[170,0.4038],[177,0.4038],[192,0.0204],[230,0.0524],[331,0.4038],[347,0.2517],[350,0.4038],[357,0.4038]]
       }
       hips_tan_ctrl.Location.Z {
        type=MovieSceneScriptingFloatChannel
        default=-0.0073
        keys [[0,0.0112],[30,0.0351],[37,0.0351],[52,-0.006],[104,0.0065],[109,0.0065],[170,0.0351],[177,0.0351],[192,-0.006],[230,-0.0026],[331,0.0351],[347,0.0112],[350,0.0351],[357,0.0351]]
       }
       hips_tan_ctrl.Rotation.X {type=MovieSceneScriptingFloatChannel, default=0, thinned_from=15, keys [[0,0],[393,0]]}
       hips_tan_ctrl.Rotation.Y {type=MovieSceneScriptingFloatChannel, default=0, thinned_from=15, keys [[0,0],[393,0]]}
       hips_tan_ctrl.Rotation.Z {type=MovieSceneScriptingFloatChannel, default=0, thinned_from=15, keys [[0,0],[393,0]]}
       hips_tan_ctrl.Scale.X {type=MovieSceneScriptingFloatChannel, default=1, const [1,14]}
       hips_tan_ctrl.Scale.Y {type=MovieSceneScriptingFloatChannel, default=1, const [1,14]}
       hips_tan_ctrl.Scale.Z {type=MovieSceneScriptingFloatChannel, default=1, const [1,14]}
       spine_01_ctrl.Location.X {type=MovieSceneScriptingFloatChannel, default=0, thinned_from=15, keys [[0,0],[393,0]]}
       spine_01_ctrl.Location.Y {type=MovieSceneScriptingFloatChannel, default=0, thinned_from=15, keys [[0,0],[393,0]]}
       spine_01_ctrl.Location.Z {type=MovieSceneScriptingFloatChannel, default=0, thinned_from=15, keys [[0,0],[393,0]]}
       spine_01_ctrl.Rotation.X {type=MovieSceneScriptingFloatChannel, default=0, const [-0.4849,14]}
       spine_01_ctrl.Rotation.Y {type=MovieSceneScriptingFloatChannel, default=0, thinned_from=15, keys [[0,0],[393,0]]}
       spine_01_ctrl.Rotation.Z {type=MovieSceneScriptingFloatChannel, default=0, const [4.4085,14]}
       spine_01_ctrl.Scale.X {type=MovieSceneScriptingFloatChannel, default=1, const [1,14]}
       spine_01_ctrl.Scale.Y {type=MovieSceneScriptingFloatChannel, default=1, const [1,14]}
       spine_01_ctrl.Scale.Z {type=MovieSceneScriptingFloatChannel, default=1, const [1,14]}
       spine_02_ctrl.Location.X {type=MovieSceneScriptingFloatChannel, default=0, thinned_from=15, keys [[0,0],[393,0]]}
       spine_02_ctrl.Location.Y {type=MovieSceneScriptingFloatChannel, default=0, thinned_from=15, keys [[0,0],[393,0]]}
       spine_02_ctrl.Location.Z {type=MovieSceneScriptingFloatChannel, default=0, thinned_from=15, keys [[0,0],[393,0]]}
       spine_02_ctrl.Rotation.X {type=MovieSceneScriptingFloatChannel, default=0, const [4.5541,14]}
       spine_02_ctrl.Rotation.Y {type=MovieSceneScriptingFloatChannel, default=0, thinned_from=15, keys [[0,0],[393,0]]}
       spine_02_ctrl.Rotation.Z {type=MovieSceneScriptingFloatChannel, default=0, const [4.4085,14]}
       spine_02_ctrl.Scale.X {type=MovieSceneScriptingFloatChannel, default=1, const [1,14]}
       spine_02_ctrl.Scale.Y {type=MovieSceneScriptingFloatChannel, default=1, const [1,14]}
       spine_02_ctrl.Scale.Z {type=MovieSceneScriptingFloatChannel, default=1, const [1,14]}
       spine_03_ctrl.Location.X {type=MovieSceneScriptingFloatChannel, default=0, thinned_from=44, keys [[0,0],[404,0]]}
       spine_03_ctrl.Location.Y {type=MovieSceneScriptingFloatChannel, default=0, thinned_from=44, keys [[0,0],[404,0]]}
       spine_03_ctrl.Location.Z {type=MovieSceneScriptingFloatChannel, default=0, thinned_from=44, keys [[0,0],[404,0]]}
       spine_03_ctrl.Rotation.X {
        type=MovieSceneScriptingFloatChannel
        default=0
        thinned_from=84
        keys [[0,0.9619],[4,0.967],[8,0.9999],[12,1.1114],[16,1.2692],[20,1.5989],[24,1.4911],[28,0.7965],[32,0.4562],[36,0.2484],[40,0.3864],[44,0.8199],[48,1.2965],[52,1.3635],[56,0.3521],[60,0.8325],[64,1.344],[68,1.8198],[72,1.7853],[76,1.7405],[80,1.6918],[84,2.1548],[88,4.287],[93,6.0074],[100,2.2661],[108,-0.5606],[117,-1.083],[125,0.3534],[132,1.3116],[140,1.4079],[144,1.4492],[150,1.1864],[156,1.2775],[160,1.292],[164,1.1086],[168,0.7965],[172,0.4562],[176,0.2484],[180,0.3864],[184,0.8197],[188,1.2937],[192,1.6114],[196,1.6985],[200,1.6148],[204,1.4293],[208,1.1812],[216,0.6201],[220,0.3673],[224,0.1729],[228,0.0612],[232,0.0379],[236,0.0823],[240,0.1706],[244,0.2968],[248,0.4652],[252,0.6684],[256,0.8893],[260,1.1121],[264,1.3314],[268,1.5458],[272,1.7453],[280,2.1174],[284,2.2987],[288,2.4618],[292,2.5879],[296,2.6314],[300,2.467],[304,1.9926],[308,1.4435],[312,1.1902],[320,1.2901],[324,1.1738],[328,0.9265],[332,0.5342],[340,0.2577],[344,0.2577],[348,0.2903],[352,0.2613],[359,-0.4711],[370,0.7063],[384,0.9619]]
       }
       spine_03_ctrl.Rotation.Y {
        type=MovieSceneScriptingFloatChannel
        default=0
        thinned_from=84
        keys [[0,-1.2484],[4,-1.2638],[8,-0.9637],[12,1.1466],[16,4.829],[20,7.0352],[24,4.8732],[28,1.725],[32,-0.2585],[36,-0.9068],[40,-1.4791],[44,-2.2084],[48,-3.2724],[52,-4.911],[56,-7.9908],[60,-6.5153],[64,-4.5041],[68,-1.498],[72,-0.3556],[76,0.7607],[80,1.7471],[84,3.3127],[88,6.7162],[93,7.3211],[100,2.0157],[108,-1.5553],[117,-7.9486],[125,-0.9073],[132,1.4954],[140,1.079],[144,0.4802],[150,0.4901],[156,3.9722],[160,5.4016],[164,3.3208],[168,0.9052],[172,-0.2585],[176,-0.9068],[180,-1.4791],[184,-2.2084],[188,-3.2727],[192,-4.1023],[196,-4.1203],[200,-3.7157],[204,-3.1632],[208,-2.5862],[212,-2.0472],[216,-1.4858],[220,-0.9072],[224,-0.4105],[228,-0.1062],[232,-0.0186],[236,-0.0714],[240,-0.169],[244,-0.2946],[248,-0.4754],[252,-0.6934],[256,-0.9283],[260,-1.1598],[264,-1.378],[268,-1.6111],[272,-1.813],[276,-1.9935],[284,-2.3844],[288,-2.5609],[292,-2.7205],[296,-2.7744],[300,-2.8075],[304,-2.7762],[308,-2.7321],[312,-2.4482],[316,-1.2784],[320,1.2667],[324,3.7762],[328,2.7725],[332,1.6797],[340,-0.3961],[344,-0.5473],[348,-0.7586],[352,-1.4218],[359,-5.8296],[370,-3.0391],[384,-1.0825]]
       }
       spine_03_ctrl.Rotation.Z {
        type=MovieSceneScriptingFloatChannel
        default=0
        thinned_from=70
        keys [[0,-0.9234],[4,2.0184],[12,10.7521],[16,16.0818],[24,20.5992],[48,3.5252],[52,4.6541],[56,6.0371],[68,10.4143],[72,12.194],[76,13.8448],[88,22.1215],[93,25.4646],[98,27.1621],[108,21.3978],[117,10.7561],[119,7.7255],[125,0.4731],[132,-5.0705],[152,8.3211],[156,13.8141],[160,17.2894],[164,12.0297],[167,6.9528],[172,3.4931],[176,4.2658],[180,4.8136],[184,5.1846],[188,4.9126],[192,4.2853],[196,3.8715],[200,3.438],[204,3.0572],[208,2.8017],[212,2.745],[216,2.938],[220,3.3115],[228,4.224],[236,4.9896],[240,5.33],[244,5.5801],[248,5.7765],[252,5.8996],[256,5.9263],[260,5.8338],[264,5.5856],[268,5.1358],[272,4.5922],[276,3.4243],[280,0.8122],[284,-0.3179],[288,-0.8284],[296,-1.0353],[304,0.4442],[308,1.6936],[312,2.8625],[316,3.3916],[320,5.1635],[324,7.4494],[328,14.6559],[332,15.8207],[340,15.9403],[344,15.8265],[348,15.4969],[352,14.9075],[359,3.6596],[370,-0.3489],[384,-1.1152]]
       }
       spine_03_ctrl.Scale.X {type=MovieSceneScriptingFloatChannel, default=1, const [1,43]}
       spine_03_ctrl.Scale.Y {type=MovieSceneScriptingFloatChannel, default=1, const [1,43]}
       spine_03_ctrl.Scale.Z {type=MovieSceneScriptingFloatChannel, default=1, const [1,43]}
       spine_04_ctrl.Location.X {type=MovieSceneScriptingFloatChannel, default=0, thinned_from=33, keys [[0,0],[325,0],[347,0.0013],[350,0],[384,0]]}
       spine_04_ctrl.Location.Y {type=MovieSceneScriptingFloatChannel, default=0, thinned_from=33, keys [[0,0],[384,0]]}
       spine_04_ctrl.Location.Z {type=MovieSceneScriptingFloatChannel, default=0, thinned_from=33, keys [[0,0],[384,0]]}
       spine_04_ctrl.Rotation.X {
        type=MovieSceneScriptingFloatChannel
        default=0
        thinned_from=33
        keys [[0,3.9108],[232,3.9108],[241,3.4249],[272,3.4249],[293,3.9108],[384,3.9108]]
       }
       spine_04_ctrl.Rotation.Y {type=MovieSceneScriptingFloatChannel, default=0, thinned_from=33, keys [[0,0],[232,0],[241,0.8788],[272,0.8788],[293,0],[384,0]]}
       spine_04_ctrl.Rotation.Z {
        type=MovieSceneScriptingFloatChannel
        default=-17.0961
        thinned_from=34
        keys [[0,-24.3268],[225,-24.3268],[232,-23.3163],[234,-22.5324],[241,-19.3782],[247,-16.8295],[272,-19.3782],[293,-24.3268],[384,-24.3268]]
       }
       spine_04_ctrl.Scale.X {type=MovieSceneScriptingFloatChannel, default=1, const [1,33]}
       spine_04_ctrl.Scale.Y {type=MovieSceneScriptingFloatChannel, default=1, const [1,33]}
       spine_04_ctrl.Scale.Z {type=MovieSceneScriptingFloatChannel, default=1, const [1,33]}
       head_ctrl.Location.X {type=MovieSceneScriptingFloatChannel, default=0, thinned_from=15, keys [[0,0],[393,0]]}
       head_ctrl.Location.Y {type=MovieSceneScriptingFloatChannel, default=0, thinned_from=15, keys [[0,0],[393,0]]}
       head_ctrl.Location.Z {type=MovieSceneScriptingFloatChannel, default=0, thinned_from=15, keys [[0,0],[393,0]]}
       head_ctrl.Rotation.X {type=MovieSceneScriptingFloatChannel, default=0, thinned_from=15, keys [[0,0],[393,0]]}
       head_ctrl.Rotation.Y {type=MovieSceneScriptingFloatChannel, default=0, thinned_from=15, keys [[0,0],[393,0]]}
       head_ctrl.Rotation.Z {type=MovieSceneScriptingFloatChannel, default=0, thinned_from=14, keys [[0,0],[357,0]]}
       head_ctrl.Scale.X {type=MovieSceneScriptingFloatChannel, default=1, const [1,14]}
       head_ctrl.Scale.Y {type=MovieSceneScriptingFloatChannel, default=1, const [1,14]}
       head_ctrl.Scale.Z {type=MovieSceneScriptingFloatChannel, default=1, const [1,14]}
       clavicle_l_ctrl.Location.X {
        type=MovieSceneScriptingFloatChannel
        default=0
        thinned_from=26
        keys [[0,3.0967],[9,3.3335],[18,1.8995],[24,1.1483],[38,-0.3753],[48,-0.8116],[61,0.0538],[80,1.7651],[98,3.0967],[107,3.3335],[116,1.8995],[122,1.1483],[136,-1.1132],[152,0.0538],[171,1.7651],[189,3.0967],[206,-0.0826],[233,-0.1127],[258,-0.0843],[282,-0.0843],[299,-0.0781],[319,-0.0607],[352,-0.0607],[369,1.7651],[385,3.0967]]
       }
       clavicle_l_ctrl.Location.Y {
        type=MovieSceneScriptingFloatChannel
        default=0
        thinned_from=27
        keys [[0,0.6761],[9,0.7568],[18,0.7322],[24,0.7128],[38,0.6546],[48,0.6288],[61,0.636],[80,0.6519],[98,0.6761],[107,0.7568],[116,0.7322],[122,0.7128],[136,0.6265],[152,0.636],[171,0.6519],[189,0.6761],[206,0.0614],[233,0.0555],[258,0.038],[282,0.038],[299,0.032],[319,0.0172],[346,0],[352,0],[369,0.6519],[385,0.6761]]
       }
       clavicle_l_ctrl.Location.Z {
        type=MovieSceneScriptingFloatChannel
        default=0
        thinned_from=26
        keys [[0,-0.0619],[9,-0.0672],[18,-0.0446],[24,-0.0327],[38,-0.0086],[48,-0.0011],[61,-0.0134],[80,-0.0392],[98,-0.0619],[107,-0.0672],[116,-0.0446],[122,-0.0327],[136,0.003],[152,-0.0134],[171,-0.0392],[189,-0.0619],[206,-0.0003],[233,0.0003],[352,0.0002],[369,-0.0392],[385,-0.0619]]
       }
       clavicle_l_ctrl.Rotation.X {
        type=MovieSceneScriptingFloatChannel
        default=0
        thinned_from=26
        keys [[0,1.9454],[9,2.1073],[18,1.3367],[24,0.933],[38,0.1143],[48,-0.1334],[61,0.3021],[80,1.1959],[98,1.9454],[107,2.1073],[116,1.3367],[122,0.933],[136,-0.2822],[152,0.3021],[171,1.1959],[189,1.9454],[206,-0.0101],[233,-0.0286],[258,-0.0214],[282,-0.0214],[299,-0.0198],[319,-0.0154],[352,-0.0154],[369,1.1959],[385,1.9454]]
       }
       clavicle_l_ctrl.Rotation.Y {
        type=MovieSceneScriptingFloatChannel
        default=0
        thinned_from=26
        keys [[0,-0.9957],[9,-1.0864],[18,-0.7675],[24,-0.5971],[38,-0.2485],[48,-0.1351],[61,-0.301],[80,-0.6556],[98,-0.9957],[107,-1.0864],[116,-0.7675],[122,-0.5971],[136,-0.0796],[152,-0.301],[171,-0.6556],[189,-0.9957],[206,-0.0149],[233,-0.0056],[258,-0.0039],[282,-0.0039],[299,-0.0034],[319,-0.002],[352,-0.002],[369,-0.6556],[385,-0.9957]]
       }
       clavicle_l_ctrl.Rotation.Z {
        type=MovieSceneScriptingFloatChannel
        default=0
        thinned_from=26
        keys [[0,10.964],[9,11.3605],[18,1.9137],[24,-3.0354],[38,-13.0721],[48,-15.5032],[61,-8.8094],[80,3.3301],[98,10.964],[107,11.3605],[116,1.9137],[122,-3.0354],[136,-17.9336],[152,-8.8094],[171,3.3301],[189,10.964],[206,-1.6952],[233,-1.8149],[258,-1.3588],[282,-1.3588],[299,-1.2595],[319,-0.9802],[352,-0.9802],[369,3.3301],[385,10.964]]
       }
       clavicle_l_ctrl.Scale.X {type=MovieSceneScriptingFloatChannel, default=1, const [1,26]}
       clavicle_l_ctrl.Scale.Y {type=MovieSceneScriptingFloatChannel, default=1, const [1,26]}
       clavicle_l_ctrl.Scale.Z {type=MovieSceneScriptingFloatChannel, default=1, const [1,26]}
       clavicle_r_ctrl.Location.X {
        type=MovieSceneScriptingFloatChannel
        default=0
        thinned_from=25
        keys [[0,3.2447],[9,3.585],[18,2.976],[38,1.3834],[48,1.3958],[61,1.8207],[80,2.5078],[98,3.2447],[107,3.585],[116,2.976],[122,2.5979],[136,1.3834],[152,1.8207],[171,2.5078],[189,3.2447],[206,0.2102],[233,0.1813],[258,0.1236],[282,0.1236],[299,0.1038],[345,0],[352,0],[369,2.5078],[385,3.2447]]
       }
       clavicle_r_ctrl.Location.Y {
        type=MovieSceneScriptingFloatChannel
        default=0
        thinned_from=24
        keys [[0,-1.3437],[9,-1.6345],[18,-2.5439],[38,-3.8175],[48,-3.793],[61,-3.0057],[80,-1.9556],[98,-1.3437],[107,-1.6345],[116,-2.5439],[122,-2.8868],[136,-3.8175],[152,-3.0057],[171,-1.9556],[189,-1.3437],[206,-0.6344],[233,-0.629],[258,-0.4399],[282,-0.4399],[299,-0.3809],[352,0],[369,-1.9556],[385,-1.3437]]
       }
       clavicle_r_ctrl.Location.Z {
        type=MovieSceneScriptingFloatChannel
        default=0
        thinned_from=25
        keys [[0,-0.0065],[9,0.0199],[18,0.3067],[38,0.8517],[48,0.8432],[61,0.5701],[80,0.2058],[98,-0.0065],[107,0.0199],[116,0.3067],[122,0.4449],[136,0.8517],[152,0.5701],[171,0.2058],[189,-0.0065],[206,0.139],[233,0.1403],[258,0.0957],[282,0.0957],[299,0.0803],[345,0],[352,0],[369,0.2058],[385,-0.0065]]
       }
       clavicle_r_ctrl.Rotation.X {
        type=MovieSceneScriptingFloatChannel
        default=-0.2961
        thinned_from=25
        keys [[0,-0.2797],[9,-0.2177],[18,0.4097],[38,1.5922],[48,1.5737],[61,0.9779],[80,0.1833],[98,-0.2797],[107,-0.2177],[116,0.4097],[122,0.71],[136,1.5922],[152,0.9779],[171,0.1833],[189,-0.2797],[206,0.0123],[233,0.015],[258,-0.0832],[282,-0.0832],[299,-0.1164],[346,-0.2961],[352,-0.2961],[369,0.1833],[385,-0.2797]]
       }
       clavicle_r_ctrl.Rotation.Y {
        type=MovieSceneScriptingFloatChannel
        default=-0.0114
        thinned_from=24
        keys [[0,-0.0922],[9,-0.1291],[18,-0.3764],[38,-0.815],[48,-0.8078],[61,-0.5778],[80,-0.271],[98,-0.0922],[107,-0.1291],[116,-0.3764],[122,-0.4889],[136,-0.815],[152,-0.5778],[171,-0.271],[189,-0.0922],[206,-0.142],[233,-0.1425],[258,-0.1026],[282,-0.1026],[299,-0.0901],[352,-0.0098],[369,-0.271],[385,-0.0922]]
       }
       clavicle_r_ctrl.Rotation.Z {
        type=MovieSceneScriptingFloatChannel
        default=-4.4081
        thinned_from=24
        keys [[0,-20.9332],[9,-21.9618],[18,-11.9856],[38,8.9738],[48,8.7118],[61,0.0911],[80,-12.2749],[98,-20.9332],[107,-21.9618],[116,-11.9856],[122,-6.7593],[136,8.9738],[152,0.0911],[171,-12.2749],[189,-20.9332],[206,-2.1231],[233,-1.9452],[258,-2.4843],[282,-2.4843],[299,-2.6017],[352,-4.0905],[369,-12.2749],[385,-20.9332]]
       }
       clavicle_r_ctrl.Scale.X {type=MovieSceneScriptingFloatChannel, default=1, const [1,24]}
       clavicle_r_ctrl.Scale.Y {type=MovieSceneScriptingFloatChannel, default=1, const [1,24]}
       clavicle_r_ctrl.Scale.Z {type=MovieSceneScriptingFloatChannel, default=1, const [1,24]}
       neck_01_ctrl.Location.X {type=MovieSceneScriptingFloatChannel, default=0, thinned_from=15, keys [[0,0],[393,0]]}
       neck_01_ctrl.Location.Y {type=MovieSceneScriptingFloatChannel, default=0, thinned_from=15, keys [[0,0],[393,0]]}
       neck_01_ctrl.Location.Z {type=MovieSceneScriptingFloatChannel, default=0, thinned_from=15, keys [[0,0],[393,0]]}
       neck_01_ctrl.Rotation.X {type=MovieSceneScriptingFloatChannel, default=0, thinned_from=15, keys [[0,0],[393,0]]}
       neck_01_ctrl.Rotation.Y {type=MovieSceneScriptingFloatChannel, default=0, thinned_from=15, keys [[0,0],[393,0]]}
       neck_01_ctrl.Rotation.Z {type=MovieSceneScriptingFloatChannel, default=0, thinned_from=15, keys [[0,0],[393,0]]}
       neck_01_ctrl.Scale.X {type=MovieSceneScriptingFloatChannel, default=1, const [1,15]}
       neck_01_ctrl.Scale.Y {type=MovieSceneScriptingFloatChannel, default=1, const [1,15]}
       neck_01_ctrl.Scale.Z {type=MovieSceneScriptingFloatChannel, default=1, const [1,15]}
       neck_02_ctrl.Location.X {type=MovieSceneScriptingFloatChannel, default=0, thinned_from=12, keys [[0,0],[393,0]]}
       neck_02_ctrl.Location.Y {type=MovieSceneScriptingFloatChannel, default=0, thinned_from=12, keys [[0,0],[393,0]]}
       neck_02_ctrl.Location.Z {type=MovieSceneScriptingFloatChannel, default=0, thinned_from=12, keys [[0,0],[393,0]]}
       neck_02_ctrl.Rotation.X {type=MovieSceneScriptingFloatChannel, default=0, thinned_from=12, keys [[0,0],[393,0]]}
       neck_02_ctrl.Rotation.Y {type=MovieSceneScriptingFloatChannel, default=0, thinned_from=12, keys [[0,0],[393,0]]}
       neck_02_ctrl.Rotation.Z {type=MovieSceneScriptingFloatChannel, default=0, thinned_from=12, keys [[0,0],[393,0]]}
       neck_02_ctrl.Scale.X {type=MovieSceneScriptingFloatChannel, default=1, const [1,11]}
       neck_02_ctrl.Scale.Y {type=MovieSceneScriptingFloatChannel, default=1, const [1,11]}
       neck_02_ctrl.Scale.Z {type=MovieSceneScriptingFloatChannel, default=1, const [1,11]}
       head_ik_ctrl.Location.X {type=MovieSceneScriptingFloatChannel, default=0, thinned_from=15, keys [[0,0],[393,0]]}
       head_ik_ctrl.Location.Y {type=MovieSceneScriptingFloatChannel, default=0, thinned_from=15, keys [[0,0],[393,0]]}
       head_ik_ctrl.Location.Z {type=MovieSceneScriptingFloatChannel, default=0, thinned_from=15, keys [[0,0],[393,0]]}
       head_ik_ctrl.Rotation.X {type=MovieSceneScriptingFloatChannel, default=0, thinned_from=15, keys [[0,0],[393,0]]}
       head_ik_ctrl.Rotation.Y {type=MovieSceneScriptingFloatChannel, default=0, thinned_from=15, keys [[0,0],[393,0]]}
       head_ik_ctrl.Rotation.Z {type=MovieSceneScriptingFloatChannel, default=0, thinned_from=14, keys [[0,0],[357,0]]}
       head_ik_ctrl.Scale.X {type=MovieSceneScriptingFloatChannel, default=1, const [1,14]}
       head_ik_ctrl.Scale.Y {type=MovieSceneScriptingFloatChannel, default=1, const [1,14]}
       head_ik_ctrl.Scale.Z {type=MovieSceneScriptingFloatChannel, default=1, const [1,14]}
       chest_ctrl.Location.X {
        type=MovieSceneScriptingFloatChannel
        default=0
        keys [[0,0.0833],[30,0.2027],[37,0.269],[52,0.0084],[104,0.0877],[109,0.0877],[170,0.2027],[177,0.269],[192,0.0084],[230,0.0302],[331,0.2027],[347,0.0833],[350,0.269],[357,0.2027]]
       }
       chest_ctrl.Location.Y {
        type=MovieSceneScriptingFloatChannel
        default=0
        keys [[0,0.5277],[30,-1.2318],[37,-1.2318],[52,-0.0387],[104,-0.4017],[109,-0.4017],[170,-1.2318],[177,-1.2318],[192,-0.0387],[230,-0.1381],[331,-1.2318],[347,0.5277],[350,-1.2318],[357,-1.2318]]
       }
       chest_ctrl.Location.Z {
        type=MovieSceneScriptingFloatChannel
        default=0
        keys [[0,0.0383],[30,-0.1775],[37,-0.1775],[52,-0.0056],[104,-0.0579],[109,-0.0579],[170,-0.1775],[177,-0.1775],[192,-0.0056],[230,-0.0199],[331,-0.1775],[347,0.0383],[350,-0.1775],[357,-0.1775]]
       }
       chest_ctrl.Rotation.X {
        type=MovieSceneScriptingFloatChannel
        default=0
        keys [[0,5.0282],[30,-3.7944],[37,-5.0575],[52,-0.1588],[104,-1.6493],[109,-1.6493],[170,-3.7944],[177,-5.0575],[192,-0.1588],[230,-0.5671],[331,-3.7944],[347,4.0618],[350,-5.0575],[357,-3.7944]]
       }
       chest_ctrl.Rotation.Y {
        type=MovieSceneScriptingFloatChannel
        default=0
        keys [[0,-1.4646],[30,0.5914],[37,0.7851],[52,0.0246],[104,0.256],[109,0.256],[170,0.5914],[177,0.7851],[192,0.0246],[230,0.088],[331,0.5914],[347,-0.769],[350,0.7851],[357,0.5914]]
       }
       chest_ctrl.Rotation.Z {
        type=MovieSceneScriptingFloatChannel
        default=0
        keys [[0,12.335],[30,9.958],[37,13.2195],[52,0.415],[104,4.3109],[109,4.3109],[170,9.958],[177,13.2195],[192,0.415],[230,1.4822],[331,9.958],[347,13.1888],[350,13.2195],[357,9.958]]
       }
       chest_ctrl.Scale.X {type=MovieSceneScriptingFloatChannel, default=1, const [1,14]}
       chest_ctrl.Scale.Y {type=MovieSceneScriptingFloatChannel, default=1, const [1,14]}
       chest_ctrl.Scale.Z {type=MovieSceneScriptingFloatChannel, default=1, const [1,14]}
       chest_tan_ctrl.Location.X {
        type=MovieSceneScriptingFloatChannel
        default=0
        keys [[0,0.2312],[30,0.0164],[37,0.0217],[52,0.0007],[104,0.0071],[109,0.0071],[170,0.0164],[177,0.0217],[192,0.0007],[230,0.0024],[331,0.0164],[347,0.1105],[350,0.0217],[357,0.0165]]
       }
       chest_tan_ctrl.Location.Y {
        type=MovieSceneScriptingFloatChannel
        default=-0.0303
        keys [[0,-0.2543],[30,-0.0359],[37,-0.2538],[52,-0.0373],[104,-0.1031],[109,-0.1031],[170,-0.0359],[177,-0.2538],[192,-0.0373],[230,-0.0553],[331,-0.0359],[347,-0.4215],[350,-0.2538],[357,-0.0359]]
       }
       chest_tan_ctrl.Location.Z {
        type=MovieSceneScriptingFloatChannel
        default=0.265
        thinned_from=14
        keys [[0,0.2258],[30,0.2889],[37,0.2667],[52,0.265],[109,0.2655],[170,0.2889],[177,0.2667],[192,0.265],[230,0.2652],[331,0.2889],[347,0.21],[350,0.2667],[357,0.2889]]
       }
       chest_tan_ctrl.Rotation.X {
        type=MovieSceneScriptingFloatChannel
        default=0
        keys [[0,0],[30,1.7293],[37,0],[52,0.1602],[104,1.664],[109,1.664],[170,1.7293],[177,0],[192,0.1602],[230,0.5721],[331,1.7293],[347,0],[350,0],[357,1.7293]]
       }
       chest_tan_ctrl.Rotation.Y {type=MovieSceneScriptingFloatChannel, default=0, thinned_from=14, keys [[0,1.1851],[30,0],[331,0],[347,0.4926],[350,0],[357,0]]}
       chest_tan_ctrl.Rotation.Z {type=MovieSceneScriptingFloatChannel, default=0, thinned_from=15, keys [[0,0],[393,0]]}
       chest_tan_ctrl.Scale.X {type=MovieSceneScriptingFloatChannel, default=1, const [1,14]}
       chest_tan_ctrl.Scale.Y {type=MovieSceneScriptingFloatChannel, default=1, const [1,14]}
       chest_tan_ctrl.Scale.Z {type=MovieSceneScriptingFloatChannel, default=1, const [1,14]}
      ]
     }]
   }
   {
    type=MovieSceneControlRigParameterTrack
    display_name=Fortnite_Legacy_Face_CtrlRig
    sections [{
      type=MovieSceneControlRigParameterSection
      channels [
       Flipbook {type=MovieSceneScriptingBoolChannel, default=false, keys [[0,false],[133,false],[148,false],[267,true],[386,false]]}
       Face_Ctrl_Master.Location.X {type=MovieSceneScriptingFloatChannel, default=47.3466, keys [[0,47.3466],[15,44.827],[133,47.3466],[148,47.3466],[267,0],[386,47.3466]]}
       Face_Ctrl_Master.Location.Y {type=MovieSceneScriptingFloatChannel, default=0.0202, keys [[0,0.0202],[15,0.2195],[133,0.0202],[148,0.0202],[267,0],[386,0.0202]]}
       Face_Ctrl_Master.Location.Z {type=MovieSceneScriptingFloatChannel, default=2.4497, keys [[0,2.4497],[15,-0.2385],[133,2.4497],[148,2.4497],[267,0],[386,2.4497]]}
       Face_Ctrl_Master.Rotation.X {type=MovieSceneScriptingFloatChannel, default=0, const [0,6]}
       Face_Ctrl_Master.Rotation.Y {type=MovieSceneScriptingFloatChannel, default=0, const [0,6]}
       Face_Ctrl_Master.Rotation.Z {type=MovieSceneScriptingFloatChannel, default=0, thinned_from=6, keys [[0,0],[148,0],[267,-32.5791],[386,0]]}
       Face_Ctrl_Master.Scale.X {type=MovieSceneScriptingFloatChannel, default=3, thinned_from=6, keys [[0,3],[148,3],[267,1],[386,3]]}
       Face_Ctrl_Master.Scale.Y {type=MovieSceneScriptingFloatChannel, default=3, thinned_from=6, keys [[0,3],[148,3],[267,1],[386,3]]}
       Face_Ctrl_Master.Scale.Z {type=MovieSceneScriptingFloatChannel, default=3, thinned_from=6, keys [[0,3],[148,3],[267,1],[386,3]]}
       gui_follow_head {type=MovieSceneScriptingFloatChannel, default=1, thinned_from=5, keys [[0,1],[148,1],[267,0],[386,1]]}
      ]
     }]
   }
  ]
 }
 Body {id=#2, parent_id=#1, possessed_class=SkeletalMeshComponentBudgeted, tracks []}
 Head {id=#3, parent_id=#1, possessed_class=SkeletalMeshComponentBudgeted, tracks []}
]
display_rate {denominator=1, numerator=30}
tick_resolution {denominator=1, numerator=24000}
tracks []
