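# Level Sequence for authoring the ChorusPunt_M custom skin's menu animation on FortPlayerMannequin.
# A Fortnite_Biped_CtrlRig track drives the body, while the skeletal animation track shown has no sections.
# Foot roll, tip and pole-vector controls shape the legs, and densely keyed FK arm channels repeat a rhythmic motion.
# Start and end keys mostly match over roughly 161 frames, suggesting a looping idle.
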LevelSequence /BRCosmetics/Animation/Game/MainPlayer/Menu/BR/CustomSkins/_s40/ChorusPunt_M/anim/ChorusPunt_LS
bindings [
 FortPlayerMannequin {
  id=#1
  possessed_class=FortPlayerMannequin
  tracks [
   {type=MovieSceneSpawnTrack, sections [{type=MovieSceneSpawnSection, channels [None {type=MovieSceneScriptingBoolChannel, default=true}]}]}
   {
    type=MovieScene3DTransformTrack
    sections [{
      type=MovieScene3DTransformSection
      channels [
       Location.X {type=MovieSceneScriptingDoubleChannel, default=0}
       Location.Y {type=MovieSceneScriptingDoubleChannel, default=0}
       Location.Z {type=MovieSceneScriptingDoubleChannel, default=0}
       Rotation.X {type=MovieSceneScriptingDoubleChannel, default=0}
       Rotation.Y {type=MovieSceneScriptingDoubleChannel, default=0}
       Rotation.Z {type=MovieSceneScriptingDoubleChannel, default=0}
       Scale.X {type=MovieSceneScriptingDoubleChannel, default=1}
       Scale.Y {type=MovieSceneScriptingDoubleChannel, default=1}
       Scale.Z {type=MovieSceneScriptingDoubleChannel, default=1}
      ]
     }]
   }
   {type=MovieSceneSkeletalAnimationTrack, sections []}
   {
    type=MovieSceneControlRigParameterTrack
    display_name=Fortnite_Biped_CtrlRig
    sections [{
      type=MovieSceneControlRigParameterSection
      channels [
       leg_l_stretch_switch {type=MovieSceneScriptingBoolChannel, default=false, const [true,10]}
       leg_r_stretch_switch {type=MovieSceneScriptingBoolChannel, default=false, const [true,10]}
       arm_l_fk_local {type=MovieSceneScriptingBoolChannel, default=true, const [true,162]}
       Auto Interpolate_3 {type=MovieSceneScriptingBoolChannel, default=false, const [false,1]}
       Auto Interpolate {type=MovieSceneScriptingBoolChannel, default=false, const [false,1]}
       arm_l_stretch_switch {type=MovieSceneScriptingBoolChannel, default=false, const [false,12]}
       Keep Aligned {type=MovieSceneScriptingBoolChannel, default=false, const [false,12]}
       arm_r_fk_local {type=MovieSceneScriptingBoolChannel, default=true, const [true,162]}
       Auto Interpolate_4 {type=MovieSceneScriptingBoolChannel, default=false, const [false,1]}
       Auto Interpolate_2 {type=MovieSceneScriptingBoolChannel, default=false, const [false,1]}
       arm_r_stretch_switch {type=MovieSceneScriptingBoolChannel, default=false, const [false,13]}
       Keep Aligned_2 {type=MovieSceneScriptingBoolChannel, default=false, const [false,13]}
       head_local {type=MovieSceneScriptingBoolChannel, default=true, const [true,10]}
       Stretch {type=MovieSceneScriptingBoolChannel, default=false, const [false,10]}
       head_stretch {type=MovieSceneScriptingBoolChannel, default=false, const [false,10]}
       spine_stretch_switch {type=MovieSceneScriptingBoolChannel, default=false, const [false,10]}
       arm_l_fk_ik_switch {type=MovieSceneScriptingBoolChannel, default=false, const [false,11]}
       leg_l_fk_ik_switch {type=MovieSceneScriptingBoolChannel, default=true, const [true,11]}
       arm_r_fk_ik_switch {type=MovieSceneScriptingBoolChannel, default=false, const [false,11]}
       leg_r_fk_ik_switch {type=MovieSceneScriptingBoolChannel, default=true, const [true,11]}
       ShowBodyControls {type=MovieSceneScriptingBoolChannel, default=true, const [true,11]}
       spine_fk_ik_switch {type=MovieSceneScriptingBoolChannel, default=false, const [false,11]}
       neck_fk_ik_switch {type=MovieSceneScriptingBoolChannel, default=false, const [false,11]}
       space {type=MovieSceneScriptingByteChannel, default=0, const [0,2]}
       global_ctrl.Location.X {type=MovieSceneScriptingFloatChannel, default=0, const [0,11]}
       global_ctrl.Location.Y {type=MovieSceneScriptingFloatChannel, default=0, const [0,11]}
       global_ctrl.Location.Z {type=MovieSceneScriptingFloatChannel, default=0, const [0,11]}
       global_ctrl.Rotation.X {type=MovieSceneScriptingFloatChannel, default=0, const [0,11]}
       global_ctrl.Rotation.Y {type=MovieSceneScriptingFloatChannel, default=0, const [0,11]}
       global_ctrl.Rotation.Z {type=MovieSceneScriptingFloatChannel, default=0, const [0,11]}
       root_ctrl.Location.X {type=MovieSceneScriptingFloatChannel, default=0, const [0,10]}
       root_ctrl.Location.Y {type=MovieSceneScriptingFloatChannel, default=0, const [0,10]}
       root_ctrl.Location.Z {type=MovieSceneScriptingFloatChannel, default=0, const [0,10]}
       root_ctrl.Rotation.X {type=MovieSceneScriptingFloatChannel, default=0, const [0,10]}
       root_ctrl.Rotation.Y {type=MovieSceneScriptingFloatChannel, default=0, const [0,10]}
       root_ctrl.Rotation.Z {type=MovieSceneScriptingFloatChannel, default=0, const [0,10]}
       root_ctrl.Scale.X {type=MovieSceneScriptingFloatChannel, default=1, const [1,10]}
       root_ctrl.Scale.Y {type=MovieSceneScriptingFloatChannel, default=1, const [1,10]}
       root_ctrl.Scale.Z {type=MovieSceneScriptingFloatChannel, default=1, const [1,10]}
       body_offset_ctrl.Location.X {type=MovieSceneScriptingFloatChannel, default=0, const [0,10]}
       body_offset_ctrl.Location.Y {type=MovieSceneScriptingFloatChannel, default=0, const [0,10]}
       body_offset_ctrl.Location.Z {type=MovieSceneScriptingFloatChannel, default=0, const [0,10]}
       body_offset_ctrl.Rotation.X {type=MovieSceneScriptingFloatChannel, default=0, const [0,10]}
       body_offset_ctrl.Rotation.Y {type=MovieSceneScriptingFloatChannel, default=0, const [0,10]}
       body_offset_ctrl.Rotation.Z {type=MovieSceneScriptingFloatChannel, default=0, const [0,10]}
       body_offset_ctrl.Scale.X {type=MovieSceneScriptingFloatChannel, default=1, const [1,10]}
       body_offset_ctrl.Scale.Y {type=MovieSceneScriptingFloatChannel, default=1, const [1,10]}
       body_offset_ctrl.Scale.Z {type=MovieSceneScriptingFloatChannel, default=1, const [1,10]}
       thigh_l_fk_ctrl.Location.X {type=MovieSceneScriptingFloatChannel, default=0, thinned_from=10, keys [[0,0],[161,0]]}
       thigh_l_fk_ctrl.Location.Y {type=MovieSceneScriptingFloatChannel, default=0, thinned_from=10, keys [[0,0],[161,0]]}
       thigh_l_fk_ctrl.Location.Z {type=MovieSceneScriptingFloatChannel, default=0, thinned_from=10, keys [[0,0],[161,0]]}
       thigh_l_fk_ctrl.Rotation.X {
        type=MovieSceneScriptingFloatChannel
        default=1.216
        thinned_from=10
        keys [[0,19.7949],[0.0013,1.216],[15.0013,1.8025],[53,3.3103],[64,4.0396],[88,7.3072],[93,7.3072],[154,38.9608],[161,19.7949]]
       }
       thigh_l_fk_ctrl.Rotation.Y {
        type=MovieSceneScriptingFloatChannel
        default=-1.2128
        thinned_from=10
        keys [[0,-5.1446],[0.0013,-1.2128],[53,-1.2714],[64,-1.2919],[88,-1.3834],[93,-1.3834],[154,4.8327],[161,-5.1446]]
       }
       thigh_l_fk_ctrl.Rotation.Z {
        type=MovieSceneScriptingFloatChannel
        default=-0.1664
        thinned_from=10
        keys [[0,-1.9858],[0.0013,-0.1664],[15.0013,-0.0393],[53,0.2872],[64,0.4452],[88,1.1529],[93,1.1529],[154,-2.8658],[161,-1.9858]]
       }
       thigh_l_fk_ctrl.Scale.X {type=MovieSceneScriptingFloatChannel, default=1, const [1,10]}
       thigh_l_fk_ctrl.Scale.Y {type=MovieSceneScriptingFloatChannel, default=1, const [1,10]}
       thigh_l_fk_ctrl.Scale.Z {type=MovieSceneScriptingFloatChannel, default=1, const [1,10]}
       calf_l_fk_ctrl.Location.X {type=MovieSceneScriptingFloatChannel, default=0, thinned_from=10, keys [[0,-0.1715],[0.0013,0],[93,0],[154,-0.0097],[161,-0.1715]]}
       calf_l_fk_ctrl.Location.Y {type=MovieSceneScriptingFloatChannel, default=0, thinned_from=10, keys [[0,0.0128],[0.0013,0],[93,0],[154,-0.025],[161,0.0128]]}
       calf_l_fk_ctrl.Location.Z {type=MovieSceneScriptingFloatChannel, default=0, thinned_from=10, keys [[0,-0.0037],[0.0013,0],[93,0],[154,0.7199],[161,-0.0037]]}
       calf_l_fk_ctrl.Rotation.X {
        type=MovieSceneScriptingFloatChannel
        default=0.0543
        thinned_from=10
        keys [[0,0.1537],[0.0013,0.0543],[53,0.0431],[64,0.0392],[88,0.0219],[93,0.0219],[154,-2.9248],[161,0.1537]]
       }
       calf_l_fk_ctrl.Rotation.Y {
        type=MovieSceneScriptingFloatChannel
        default=1.2238
        thinned_from=10
        keys [[0,1.2153],[0.0013,1.2238],[93,1.2239],[154,2.034],[161,1.2153]]
       }
       calf_l_fk_ctrl.Rotation.Z {
        type=MovieSceneScriptingFloatChannel
        default=0.3732
        thinned_from=10
        keys [[0,5.0434],[0.0013,0.3732],[15.0013,0.2272],[53,-0.1482],[64,-0.3298],[88,-1.1433],[93,-1.1433],[154,27.124],[161,5.0434]]
       }
       calf_l_fk_ctrl.Scale.X {type=MovieSceneScriptingFloatChannel, default=1, const [1,10]}
       calf_l_fk_ctrl.Scale.Y {type=MovieSceneScriptingFloatChannel, default=1, const [1,10]}
       calf_l_fk_ctrl.Scale.Z {type=MovieSceneScriptingFloatChannel, default=1, const [1,10]}
       foot_l_fk_ctrl.Location.X {type=MovieSceneScriptingFloatChannel, default=0, thinned_from=10, keys [[0,-0.1874],[0.0013,0],[154,0],[161,-0.1874]]}
       foot_l_fk_ctrl.Location.Y {type=MovieSceneScriptingFloatChannel, default=0.0398, thinned_from=10, keys [[0,0.0398],[93,0.0398],[154,0.0389],[161,0.0398]]}
       foot_l_fk_ctrl.Location.Z {type=MovieSceneScriptingFloatChannel, default=-0.8885, thinned_from=10, keys [[0,-0.8885],[93,-0.8885],[154,-0.6669],[161,-0.8885]]}
       foot_l_fk_ctrl.Rotation.X {
        type=MovieSceneScriptingFloatChannel
        default=-1.1647
        thinned_from=10
        keys [[0,-0.355],[0.0013,-1.1647],[64,-1.1703],[93,-1.1768],[154,-5.3719],[161,-0.355]]
       }
       foot_l_fk_ctrl.Rotation.Y {
        type=MovieSceneScriptingFloatChannel
        default=0.0792
        thinned_from=10
        keys [[0,4.6189],[0.0013,0.0792],[15.0013,0.0603],[53,0.0118],[64,-0.0116],[88,-0.1166],[93,-0.1166],[154,1.8476],[161,4.6189]]
       }
       foot_l_fk_ctrl.Rotation.Z {
        type=MovieSceneScriptingFloatChannel
        default=-0.2074
        thinned_from=10
        keys [[0,-2.7464],[0.0013,-0.2074],[15.0013,-0.1322],[53,0.0611],[64,0.1546],[88,0.5735],[93,0.5735],[154,-13.4921],[161,-2.7464]]
       }
       foot_l_fk_ctrl.Scale.X {type=MovieSceneScriptingFloatChannel, default=1, const [1,10]}
       foot_l_fk_ctrl.Scale.Y {type=MovieSceneScriptingFloatChannel, default=1, const [1,10]}
       foot_l_fk_ctrl.Scale.Z {type=MovieSceneScriptingFloatChannel, default=1, const [1,10]}
       ball_l_fk_ctrl.Location.X {type=MovieSceneScriptingFloatChannel, default=0, thinned_from=10, keys [[0,0],[161,0]]}
       ball_l_fk_ctrl.Location.Y {type=MovieSceneScriptingFloatChannel, default=0, const [0,10]}
       ball_l_fk_ctrl.Location.Z {type=MovieSceneScriptingFloatChannel, default=0, const [0,10]}
       ball_l_fk_ctrl.Rotation.X {type=MovieSceneScriptingFloatChannel, default=0, thinned_from=10, keys [[0,0],[161,0]]}
       ball_l_fk_ctrl.Rotation.Y {type=MovieSceneScriptingFloatChannel, default=0, thinned_from=10, keys [[0,0],[161,0]]}
       ball_l_fk_ctrl.Rotation.Z {type=MovieSceneScriptingFloatChannel, default=0, thinned_from=10, keys [[0,0],[161,0]]}
       ball_l_fk_ctrl.Scale.X {type=MovieSceneScriptingFloatChannel, default=1, const [1,10]}
       ball_l_fk_ctrl.Scale.Y {type=MovieSceneScriptingFloatChannel, default=1, const [1,10]}
       ball_l_fk_ctrl.Scale.Z {type=MovieSceneScriptingFloatChannel, default=1, const [1,10]}
       foot_l_ik_ctrl.Location.X {type=MovieSceneScriptingFloatChannel, default=0, const [0,10]}
       foot_l_ik_ctrl.Location.Y {type=MovieSceneScriptingFloatChannel, default=0, const [0,10]}
       foot_l_ik_ctrl.Location.Z {type=MovieSceneScriptingFloatChannel, default=0, const [0,10]}
       foot_l_ik_ctrl.Rotation.X {type=MovieSceneScriptingFloatChannel, default=0, const [0,10]}
       foot_l_ik_ctrl.Rotation.Y {type=MovieSceneScriptingFloatChannel, default=0, const [0,10]}
       foot_l_ik_ctrl.Rotation.Z {type=MovieSceneScriptingFloatChannel, default=0, thinned_from=10, keys [[0,-10.0283],[93,-10.0283],[154,-10.0576],[161,-10.0283]]}
       foot_l_ik_ctrl.Scale.X {type=MovieSceneScriptingFloatChannel, default=1, const [1,10]}
       foot_l_ik_ctrl.Scale.Y {type=MovieSceneScriptingFloatChannel, default=1, const [1,10]}
       foot_l_ik_ctrl.Scale.Z {type=MovieSceneScriptingFloatChannel, default=1, const [1,10]}
       ball_l_ik_ctrl.Location.X {type=MovieSceneScriptingFloatChannel, default=0, const [0,10]}
       ball_l_ik_ctrl.Location.Y {type=MovieSceneScriptingFloatChannel, default=0, const [0,10]}
       ball_l_ik_ctrl.Location.Z {type=MovieSceneScriptingFloatChannel, default=0, const [0,10]}
       ball_l_ik_ctrl.Rotation.X {type=MovieSceneScriptingFloatChannel, default=0, thinned_from=10, keys [[0,0],[161,0]]}
       ball_l_ik_ctrl.Rotation.Y {type=MovieSceneScriptingFloatChannel, default=0, const [0,10]}
       ball_l_ik_ctrl.Rotation.Z {type=MovieSceneScriptingFloatChannel, default=0, thinned_from=10, keys [[0,0],[161,0]]}
       ball_l_ik_ctrl.Scale.X {type=MovieSceneScriptingFloatChannel, default=1, const [1,10]}
       ball_l_ik_ctrl.Scale.Y {type=MovieSceneScriptingFloatChannel, default=1, const [1,10]}
       ball_l_ik_ctrl.Scale.Z {type=MovieSceneScriptingFloatChannel, default=1, const [1,10]}
       foot_bk1_l_ctrl.Location.X {type=MovieSceneScriptingFloatChannel, default=0, const [0,25]}
       foot_bk1_l_ctrl.Location.Y {type=MovieSceneScriptingFloatChannel, default=0, const [0,25]}
       foot_bk1_l_ctrl.Location.Z {type=MovieSceneScriptingFloatChannel, default=0, const [0,25]}
       foot_bk1_l_ctrl.Rotation.X {type=MovieSceneScriptingFloatChannel, default=0, thinned_from=25, keys [[0,0],[161,0]]}
       foot_bk1_l_ctrl.Rotation.Y {type=MovieSceneScriptingFloatChannel, default=0, thinned_from=25, keys [[0,0],[161,0]]}
       foot_bk1_l_ctrl.Rotation.Z {
        type=MovieSceneScriptingFloatChannel
        default=0
        thinned_from=24
        keys [[0,0],[15.0013,0],[23,-2.191],[33,-6.9313],[41,-0.0458],[45,0],[64,0],[71,-4.3831],[78,-0.9615],[80,0],[88,-0.0141],[101,-0.0141],[117,-7.0231],[127,0],[139,-1.0048],[147,0],[161,0]]
       }
       foot_bk1_l_ctrl.Scale.X {type=MovieSceneScriptingFloatChannel, default=1, const [1,25]}
       foot_bk1_l_ctrl.Scale.Y {type=MovieSceneScriptingFloatChannel, default=1, const [1,25]}
       foot_bk1_l_ctrl.Scale.Z {type=MovieSceneScriptingFloatChannel, default=1, const [1,25]}
       foot_roll_l_ctrl.Location.X {type=MovieSceneScriptingFloatChannel, default=0, const [0,10]}
       foot_roll_l_ctrl.Location.Y {type=MovieSceneScriptingFloatChannel, default=0, const [0,10]}
       foot_roll_l_ctrl.Location.Z {type=MovieSceneScriptingFloatChannel, default=0, const [0,10]}
       foot_roll_l_ctrl.Rotation.X {type=MovieSceneScriptingFloatChannel, default=0, const [0,10]}
       foot_roll_l_ctrl.Rotation.Y {type=MovieSceneScriptingFloatChannel, default=0, const [0,10]}
       foot_roll_l_ctrl.Rotation.Z {type=MovieSceneScriptingFloatChannel, default=0, const [0,10]}
       foot_roll_l_ctrl.Scale.X {type=MovieSceneScriptingFloatChannel, default=1, const [1,10]}
       foot_roll_l_ctrl.Scale.Y {type=MovieSceneScriptingFloatChannel, default=1, const [1,10]}
       foot_roll_l_ctrl.Scale.Z {type=MovieSceneScriptingFloatChannel, default=1, const [1,10]}
       roll_blend_l_ctrl {type=MovieSceneScriptingFloatChannel, default=1, const [1,10]}
       tip_l_ctrl.Location.X {type=MovieSceneScriptingFloatChannel, default=0, const [0,16]}
       tip_l_ctrl.Location.Y {type=MovieSceneScriptingFloatChannel, default=0, const [0,16]}
       tip_l_ctrl.Location.Z {type=MovieSceneScriptingFloatChannel, default=0, const [0,16]}
       tip_l_ctrl.Rotation.X {type=MovieSceneScriptingFloatChannel, default=0, thinned_from=16, keys [[0,0],[161,0]]}
       tip_l_ctrl.Rotation.Y {
        type=MovieSceneScriptingFloatChannel
        default=0
        thinned_from=15
        keys [[0,0],[21,0],[33,-6.7601],[38,-7.6294],[61,-7.6294],[79,-0.7625],[107,-0.7625],[116,-2.088],[138,-2.088],[154,0],[161,0]]
       }
       tip_l_ctrl.Rotation.Z {type=MovieSceneScriptingFloatChannel, default=0, thinned_from=16, keys [[0,0],[161,0]]}
       tip_l_ctrl.Scale.X {type=MovieSceneScriptingFloatChannel, default=1, const [1,16]}
       tip_l_ctrl.Scale.Y {type=MovieSceneScriptingFloatChannel, default=1, const [1,16]}
       tip_l_ctrl.Scale.Z {type=MovieSceneScriptingFloatChannel, default=1, const [1,16]}
       heel_l_ctrl.Location.X {type=MovieSceneScriptingFloatChannel, default=0, const [0,10]}
       heel_l_ctrl.Location.Y {type=MovieSceneScriptingFloatChannel, default=0, const [0,10]}
       heel_l_ctrl.Location.Z {type=MovieSceneScriptingFloatChannel, default=0, const [0,10]}
       heel_l_ctrl.Rotation.X {type=MovieSceneScriptingFloatChannel, default=0, thinned_from=10, keys [[0,0],[161,0]]}
       heel_l_ctrl.Rotation.Y {type=MovieSceneScriptingFloatChannel, default=0, thinned_from=10, keys [[0,0],[161,0]]}
       heel_l_ctrl.Rotation.Z {type=MovieSceneScriptingFloatChannel, default=0, const [0,10]}
       heel_l_ctrl.Scale.X {type=MovieSceneScriptingFloatChannel, default=1, const [1,10]}
       heel_l_ctrl.Scale.Y {type=MovieSceneScriptingFloatChannel, default=1, const [1,10]}
       heel_l_ctrl.Scale.Z {type=MovieSceneScriptingFloatChannel, default=1, const [1,10]}
       leg_l_softness {type=MovieSceneScriptingFloatChannel, default=0, const [0.5,10]}
       leg_l_pv_ik_ctrl.Location.X {
        type=MovieSceneScriptingFloatChannel
        default=0
        thinned_from=6
        keys [[0,-0.4772],[15.0013,-1.312],[66,-4.5884],[154,-1.6977],[161,-0.4772]]
       }
       leg_l_pv_ik_ctrl.Location.Y {
        type=MovieSceneScriptingFloatChannel
        default=0
        thinned_from=6
        keys [[0,-0.0881],[15.0013,-0.2423],[66,-0.8475],[154,-0.4677],[161,-0.0881]]
       }
       leg_l_pv_ik_ctrl.Location.Z {type=MovieSceneScriptingFloatChannel, default=0, thinned_from=6, keys [[0,-0.0473],[15.0013,-0.13],[66,-0.4546],[154,-0.1666],[161,-0.0473]]}
       leg_l_pv_ik_ctrl.Rotation.X {type=MovieSceneScriptingFloatChannel, default=0, const [0,6]}
       leg_l_pv_ik_ctrl.Rotation.Y {type=MovieSceneScriptingFloatChannel, default=0, thinned_from=6, keys [[0,0],[161,0]]}
       leg_l_pv_ik_ctrl.Rotation.Z {type=MovieSceneScriptingFloatChannel, default=0, thinned_from=6, keys [[0,0],[161,0]]}
       leg_l_pv_ik_ctrl.Scale.X {type=MovieSceneScriptingFloatChannel, default=1, const [1,6]}
       leg_l_pv_ik_ctrl.Scale.Y {type=MovieSceneScriptingFloatChannel, default=1, const [1,6]}
       leg_l_pv_ik_ctrl.Scale.Z {type=MovieSceneScriptingFloatChannel, default=1, const [1,6]}
       thigh_r_fk_ctrl.Location.X {type=MovieSceneScriptingFloatChannel, default=0, thinned_from=10, keys [[0,0],[161,0]]}
       thigh_r_fk_ctrl.Location.Y {type=MovieSceneScriptingFloatChannel, default=0, thinned_from=10, keys [[0,0],[161,0]]}
       thigh_r_fk_ctrl.Location.Z {type=MovieSceneScriptingFloatChannel, default=0, thinned_from=10, keys [[0,0],[161,0]]}
       thigh_r_fk_ctrl.Rotation.X {
        type=MovieSceneScriptingFloatChannel
        default=1.2139
        thinned_from=10
        keys [[0,12.3362],[0.0013,1.2139],[15.0013,1.9687],[53,3.9093],[64,4.848],[88,9.0534],[93,9.0534],[154,16.3967],[161,12.3362]]
       }
       thigh_r_fk_ctrl.Rotation.Y {
        type=MovieSceneScriptingFloatChannel
        default=-1.213
        thinned_from=10
        keys [[0,4.3309],[0.0013,-1.213],[53,-1.0471],[64,-0.9893],[88,-0.7304],[93,-0.7304],[154,3.926],[161,4.3309]]
       }
       thigh_r_fk_ctrl.Rotation.Z {
        type=MovieSceneScriptingFloatChannel
        default=-0.1592
        thinned_from=10
        keys [[0,-2.8816],[0.0013,-0.1592],[15.0013,-0.223],[53,-0.3871],[64,-0.4665],[88,-0.8222],[93,-0.8222],[154,-12.3522],[161,-2.8816]]
       }
       thigh_r_fk_ctrl.Scale.X {type=MovieSceneScriptingFloatChannel, default=1, const [1,10]}
       thigh_r_fk_ctrl.Scale.Y {type=MovieSceneScriptingFloatChannel, default=1, const [1,10]}
       thigh_r_fk_ctrl.Scale.Z {type=MovieSceneScriptingFloatChannel, default=1, const [1,10]}
       calf_r_fk_ctrl.Location.X {type=MovieSceneScriptingFloatChannel, default=0, thinned_from=10, keys [[0,-0.6727],[0.0013,0],[93,0],[154,0.0003],[161,-0.6727]]}
       calf_r_fk_ctrl.Location.Y {type=MovieSceneScriptingFloatChannel, default=0, thinned_from=10, keys [[0,0.0515],[0.0013,0],[154,0],[161,0.0515]]}
       calf_r_fk_ctrl.Location.Z {type=MovieSceneScriptingFloatChannel, default=0, thinned_from=10, keys [[0,-0.0141],[0.0013,0],[93,0],[154,-0.0135],[161,-0.0141]]}
       calf_r_fk_ctrl.Rotation.X {
        type=MovieSceneScriptingFloatChannel
        default=0.0561
        thinned_from=10
        keys [[0,0.049],[0.0013,0.0561],[53,0.059],[64,0.06],[88,0.0646],[93,0.0646],[154,0.5898],[161,0.049]]
       }
       calf_r_fk_ctrl.Rotation.Y {type=MovieSceneScriptingFloatChannel, default=1.1982, thinned_from=10, keys [[0,1.1985],[93,1.1977],[154,1.0716],[161,1.1985]]}
       calf_r_fk_ctrl.Rotation.Z {
        type=MovieSceneScriptingFloatChannel
        default=0.2616
        thinned_from=10
        keys [[0,-0.0767],[0.0013,0.2616],[15.0013,0.3008],[53,0.4017],[64,0.4505],[88,0.669],[93,0.669],[154,23.2386],[161,-0.0767]]
       }
       calf_r_fk_ctrl.Scale.X {type=MovieSceneScriptingFloatChannel, default=1, const [1,10]}
       calf_r_fk_ctrl.Scale.Y {type=MovieSceneScriptingFloatChannel, default=1, const [1,10]}
       calf_r_fk_ctrl.Scale.Z {type=MovieSceneScriptingFloatChannel, default=1, const [1,10]}
       foot_r_fk_ctrl.Location.X {type=MovieSceneScriptingFloatChannel, default=0, thinned_from=10, keys [[0,-0.7354],[0.0013,0],[154,0],[161,-0.7354]]}
       foot_r_fk_ctrl.Location.Y {type=MovieSceneScriptingFloatChannel, default=-0.0329, thinned_from=10, keys [[0,-0.0329],[93,-0.0329],[154,-0.0337],[161,-0.0329]]}
       foot_r_fk_ctrl.Location.Z {type=MovieSceneScriptingFloatChannel, default=-0.9075, thinned_from=10, keys [[0,-0.9075],[93,-0.9075],[154,-0.686],[161,-0.9075]]}
       foot_r_fk_ctrl.Rotation.X {
        type=MovieSceneScriptingFloatChannel
        default=-1.1648
        thinned_from=10
        keys [[0,10.5474],[0.0013,-1.1648],[53,-1.2063],[64,-1.2207],[88,-1.2855],[93,-1.2855],[154,-13.1649],[161,10.5474]]
       }
       foot_r_fk_ctrl.Rotation.Y {
        type=MovieSceneScriptingFloatChannel
        default=0.1067
        thinned_from=10
        keys [[0,-5.6758],[0.0013,0.1067],[53,0.1091],[64,0.1099],[88,0.1136],[93,0.1136],[154,0.7308],[161,-5.6758]]
       }
       foot_r_fk_ctrl.Rotation.Z {
        type=MovieSceneScriptingFloatChannel
        default=-0.1036
        thinned_from=10
        keys [[0,6.2399],[0.0013,-0.1036],[15.0013,0.0022],[53,0.2742],[64,0.4058],[88,0.9952],[93,0.9952],[154,-2.8578],[161,6.2399]]
       }
       foot_r_fk_ctrl.Scale.X {type=MovieSceneScriptingFloatChannel, default=1, const [1,10]}
       foot_r_fk_ctrl.Scale.Y {type=MovieSceneScriptingFloatChannel, default=1, const [1,10]}
       foot_r_fk_ctrl.Scale.Z {type=MovieSceneScriptingFloatChannel, default=1, const [1,10]}
       ball_r_fk_ctrl.Location.X {type=MovieSceneScriptingFloatChannel, default=0, thinned_from=10, keys [[0,0],[161,0]]}
       ball_r_fk_ctrl.Location.Y {type=MovieSceneScriptingFloatChannel, default=0, const [0,10]}
       ball_r_fk_ctrl.Location.Z {type=MovieSceneScriptingFloatChannel, default=0, const [0,10]}
       ball_r_fk_ctrl.Rotation.X {type=MovieSceneScriptingFloatChannel, default=0, const [0,10]}
       ball_r_fk_ctrl.Rotation.Y {type=MovieSceneScriptingFloatChannel, default=0, thinned_from=10, keys [[0,0],[161,0]]}
       ball_r_fk_ctrl.Rotation.Z {type=MovieSceneScriptingFloatChannel, default=0, thinned_from=10, keys [[0,0],[161,0]]}
       ball_r_fk_ctrl.Scale.X {type=MovieSceneScriptingFloatChannel, default=1, const [1,10]}
       ball_r_fk_ctrl.Scale.Y {type=MovieSceneScriptingFloatChannel, default=1, const [1,10]}
       ball_r_fk_ctrl.Scale.Z {type=MovieSceneScriptingFloatChannel, default=1, const [1,10]}
       foot_r_ik_ctrl.Location.X {type=MovieSceneScriptingFloatChannel, default=0, const [0,10]}
       foot_r_ik_ctrl.Location.Y {type=MovieSceneScriptingFloatChannel, default=0, const [0,10]}
       foot_r_ik_ctrl.Location.Z {type=MovieSceneScriptingFloatChannel, default=0, const [0,10]}
       foot_r_ik_ctrl.Rotation.X {type=MovieSceneScriptingFloatChannel, default=0, const [0,10]}
       foot_r_ik_ctrl.Rotation.Y {type=MovieSceneScriptingFloatChannel, default=0, const [0,10]}
       foot_r_ik_ctrl.Rotation.Z {type=MovieSceneScriptingFloatChannel, default=0, thinned_from=10, keys [[0,-32.6084],[161,-32.6084]]}
       foot_r_ik_ctrl.Scale.X {type=MovieSceneScriptingFloatChannel, default=1, const [1,10]}
       foot_r_ik_ctrl.Scale.Y {type=MovieSceneScriptingFloatChannel, default=1, const [1,10]}
       foot_r_ik_ctrl.Scale.Z {type=MovieSceneScriptingFloatChannel, default=1, const [1,10]}
       ball_r_ik_ctrl.Location.X {type=MovieSceneScriptingFloatChannel, default=0, const [0,11]}
       ball_r_ik_ctrl.Location.Y {type=MovieSceneScriptingFloatChannel, default=0, const [0,11]}
       ball_r_ik_ctrl.Location.Z {type=MovieSceneScriptingFloatChannel, default=0, thinned_from=11, keys [[0,0],[161,0]]}
       ball_r_ik_ctrl.Rotation.X {type=MovieSceneScriptingFloatChannel, default=0, thinned_from=11, keys [[0,0],[161,0]]}
       ball_r_ik_ctrl.Rotation.Y {type=MovieSceneScriptingFloatChannel, default=0, thinned_from=11, keys [[0,0],[161,0]]}
       ball_r_ik_ctrl.Rotation.Z {type=MovieSceneScriptingFloatChannel, default=0, thinned_from=11, keys [[0,0],[93,0],[131,5.4153],[154,0],[161,0]]}
       ball_r_ik_ctrl.Scale.X {type=MovieSceneScriptingFloatChannel, default=1, const [1,11]}
       ball_r_ik_ctrl.Scale.Y {type=MovieSceneScriptingFloatChannel, default=1, const [1,11]}
       ball_r_ik_ctrl.Scale.Z {type=MovieSceneScriptingFloatChannel, default=1, const [1,11]}
       foot_bk1_r_ctrl.Location.X {type=MovieSceneScriptingFloatChannel, default=0, thinned_from=25, keys [[0,0],[161,0]]}
       foot_bk1_r_ctrl.Location.Y {type=MovieSceneScriptingFloatChannel, default=0, thinned_from=25, keys [[0,0],[161,0]]}
       foot_bk1_r_ctrl.Location.Z {type=MovieSceneScriptingFloatChannel, default=0, thinned_from=25, keys [[0,0],[161,0]]}
       foot_bk1_r_ctrl.Rotation.X {type=MovieSceneScriptingFloatChannel, default=0, thinned_from=25, keys [[0,0],[161,0]]}
       foot_bk1_r_ctrl.Rotation.Y {type=MovieSceneScriptingFloatChannel, default=0, thinned_from=25, keys [[0,0],[161,0]]}
       foot_bk1_r_ctrl.Rotation.Z {
        type=MovieSceneScriptingFloatChannel
        default=0
        thinned_from=24
        keys [[0,0],[5,-0.246],[11,-3.7635],[15,-3.8839],[20,-2.1001],[23,-0.3264],[27,0],[35,0.0089],[53,0],[64,0],[71,-2.2055],[78,-1.0808],[82,0],[85,0],[88,-0.0043],[93,-0.0059],[106,0],[116,-3.4006],[126,0],[145,0],[154,-3.3127],[161,0]]
       }
       foot_bk1_r_ctrl.Scale.X {type=MovieSceneScriptingFloatChannel, default=1, const [1,25]}
       foot_bk1_r_ctrl.Scale.Y {type=MovieSceneScriptingFloatChannel, default=1, const [1,25]}
       foot_bk1_r_ctrl.Scale.Z {type=MovieSceneScriptingFloatChannel, default=1, const [1,25]}
       foot_roll_r_ctrl.Location.X {type=MovieSceneScriptingFloatChannel, default=0, const [0,11]}
       foot_roll_r_ctrl.Location.Y {type=MovieSceneScriptingFloatChannel, default=0, const [0,11]}
       foot_roll_r_ctrl.Location.Z {type=MovieSceneScriptingFloatChannel, default=0, const [0,11]}
       foot_roll_r_ctrl.Rotation.X {type=MovieSceneScriptingFloatChannel, default=0, const [0,11]}
       foot_roll_r_ctrl.Rotation.Y {type=MovieSceneScriptingFloatChannel, default=0, const [0,11]}
       foot_roll_r_ctrl.Rotation.Z {type=MovieSceneScriptingFloatChannel, default=0, thinned_from=11, keys [[0,0],[161,0]]}
       foot_roll_r_ctrl.Scale.X {type=MovieSceneScriptingFloatChannel, default=1, const [1,11]}
       foot_roll_r_ctrl.Scale.Y {type=MovieSceneScriptingFloatChannel, default=1, const [1,11]}
       foot_roll_r_ctrl.Scale.Z {type=MovieSceneScriptingFloatChannel, default=1, const [1,11]}
       roll_blend_r_ctrl {type=MovieSceneScriptingFloatChannel, default=1, const [1,10]}
       tip_r_ctrl.Location.X {type=MovieSceneScriptingFloatChannel, default=0, thinned_from=19, keys [[0,0],[161,0]]}
       tip_r_ctrl.Location.Y {type=MovieSceneScriptingFloatChannel, default=0, thinned_from=19, keys [[0,0],[161,0]]}
       tip_r_ctrl.Location.Z {type=MovieSceneScriptingFloatChannel, default=0, thinned_from=19, keys [[0,0],[161,0]]}
       tip_r_ctrl.Rotation.X {type=MovieSceneScriptingFloatChannel, default=0, thinned_from=19, keys [[0,0],[161,0]]}
       tip_r_ctrl.Rotation.Y {
        type=MovieSceneScriptingFloatChannel
        default=0
        thinned_from=17
        keys [[0,-0.0619],[6,-0.159],[12,1.0979],[16,4.1049],[20,7.8876],[25,8.4213],[67,8.4213],[74,4.1767],[81,2.5856],[97,2.5856],[107,2.5524],[138,2.5524],[154,-0.0994],[161,-0.0994]]
       }
       tip_r_ctrl.Rotation.Z {type=MovieSceneScriptingFloatChannel, default=0, thinned_from=19, keys [[0,0],[161,0]]}
       tip_r_ctrl.Scale.X {type=MovieSceneScriptingFloatChannel, default=1, const [1,19]}
       tip_r_ctrl.Scale.Y {type=MovieSceneScriptingFloatChannel, default=1, const [1,19]}
       tip_r_ctrl.Scale.Z {type=MovieSceneScriptingFloatChannel, default=1, const [1,19]}
       heel_r_ctrl.Location.X {type=MovieSceneScriptingFloatChannel, default=0, const [0,10]}
       heel_r_ctrl.Location.Y {type=MovieSceneScriptingFloatChannel, default=0, const [0,10]}
       heel_r_ctrl.Location.Z {type=MovieSceneScriptingFloatChannel, default=0, const [0,10]}
       heel_r_ctrl.Rotation.X {type=MovieSceneScriptingFloatChannel, default=0, thinned_from=10, keys [[0,0],[161,0]]}
       heel_r_ctrl.Rotation.Y {type=MovieSceneScriptingFloatChannel, default=0, thinned_from=10, keys [[0,0],[161,0]]}
       heel_r_ctrl.Rotation.Z {type=MovieSceneScriptingFloatChannel, default=0, const [0,10]}
       heel_r_ctrl.Scale.X {type=MovieSceneScriptingFloatChannel, default=1, const [1,10]}
       heel_r_ctrl.Scale.Y {type=MovieSceneScriptingFloatChannel, default=1, const [1,10]}
       heel_r_ctrl.Scale.Z {type=MovieSceneScriptingFloatChannel, default=1, const [1,10]}
       leg_r_softness {type=MovieSceneScriptingFloatChannel, default=0, const [0.5,10]}
       leg_r_pv_ik_ctrl.Location.X {type=MovieSceneScriptingFloatChannel, default=0, thinned_from=6, keys [[0,-8.8039],[66,-8.8917],[154,-8.8519],[161,-8.8039]]}
       leg_r_pv_ik_ctrl.Location.Y {type=MovieSceneScriptingFloatChannel, default=0, thinned_from=6, keys [[0,-5.5873],[66,-5.6431],[154,-5.6184],[161,-5.5873]]}
       leg_r_pv_ik_ctrl.Location.Z {type=MovieSceneScriptingFloatChannel, default=0, thinned_from=6, keys [[0,1.7006],[66,1.7175],[154,1.7099],[161,1.7006]]}
       leg_r_pv_ik_ctrl.Rotation.X {type=MovieSceneScriptingFloatChannel, default=0, thinned_from=6, keys [[0,0.0011],[161,0.0011]]}
       leg_r_pv_ik_ctrl.Rotation.Y {type=MovieSceneScriptingFloatChannel, default=0, thinned_from=6, keys [[0,-0.0032],[154,-0.0033],[161,-0.0032]]}
       leg_r_pv_ik_ctrl.Rotation.Z {type=MovieSceneScriptingFloatChannel, default=0, thinned_from=6, keys [[0,0.0007],[161,0.0007]]}
       leg_r_pv_ik_ctrl.Scale.X {type=MovieSceneScriptingFloatChannel, default=1, const [1,6]}
       leg_r_pv_ik_ctrl.Scale.Y {type=MovieSceneScriptingFloatChannel, default=1, const [1,6]}
       leg_r_pv_ik_ctrl.Scale.Z {type=MovieSceneScriptingFloatChannel, default=1, const [1,6]}
       upperarm_l_fk_ctrl.Location.X {type=MovieSceneScriptingFloatChannel, default=0, thinned_from=162, keys [[0,0],[161,0]]}
       upperarm_l_fk_ctrl.Location.Y {type=MovieSceneScriptingFloatChannel, default=0, thinned_from=162, keys [[0,0],[161,0]]}
       upperarm_l_fk_ctrl.Location.Z {type=MovieSceneScriptingFloatChannel, default=0, thinned_from=162, keys [[0,0],[161,0]]}
       upperarm_l_fk_ctrl.Rotation.X {
        type=MovieSceneScriptingFloatChannel
        default=24.4912
        thinned_from=162
        keys [[0,15.6458],[9,15.6486],[10,15.6674],[11,15.737],[12,15.9221],[13,16.2978],[14,16.8423],[15,17.4735],[16,18.1747],[18,19.7367],[19,20.4846],[20,21.1224],[21,21.6135],[22,21.9948],[23,22.3172],[24,22.5937],[25,22.8091],[26,22.9249],[27,22.9257],[28,22.8059],[29,22.4879],[30,21.8527],[31,20.7993],[32,19.3709],[33,17.8252],[34,16.6119],[35,15.9535],[36,15.6692],[37,15.5581],[41,15.4091],[43,15.398],[45,15.46],[48,15.6511],[51,15.7233],[52,15.8156],[53,16.1338],[54,16.7156],[55,17.4153],[56,18.0471],[57,18.5385],[59,19.3495],[60,19.8496],[62,21.1084],[63,21.6106],[64,21.9946],[65,22.3172],[66,22.5937],[67,22.8091],[68,22.9249],[69,22.9257],[70,22.8059],[71,22.4879],[72,21.8527],[73,20.7993],[74,19.3709],[75,17.8252],[76,16.6119],[77,15.9535],[78,15.6692],[79,15.5581],[83,15.4091],[85,15.398],[87,15.46],[90,15.6511],[92,15.711],[96,15.6746],[97,15.7739],[98,16.1367],[99,16.8622],[100,17.8853],[102,20.2072],[103,21.1821],[104,21.8763],[105,22.3103],[106,22.5986],[107,22.8102],[108,22.925],[109,22.9257],[110,22.8059],[111,22.4879],[112,21.8527],[113,20.7993],[114,19.3709],[115,17.8252],[116,16.6119],[117,15.9535],[118,15.6692],[119,15.5581],[123,15.4091],[125,15.398],[127,15.46],[130,15.6511],[132,15.711],[136,15.6806],[137,15.8189],[138,16.3055],[139,17.2505],[140,18.5235],[141,19.8712],[142,21.0329],[143,21.8361],[144,22.3048],[145,22.5983],[146,22.8102],[147,22.925],[148,22.9257],[149,22.8059],[150,22.4879],[151,21.8527],[152,20.7991],[153,19.3699],[154,17.8234],[155,16.6118],[156,15.9617],[157,15.7009],[158,15.6283],[161,15.64]]
       }
       upperarm_l_fk_ctrl.Rotation.Y {
        type=MovieSceneScriptingFloatChannel
        default=-10.9057
        thinned_from=162
        keys [[0,-10.3949],[9,-10.3886],[10,-10.3479],[11,-10.1959],[12,-9.7864],[13,-8.9232],[14,-7.5834],[15,-5.8534],[16,-3.6406],[17,-0.8951],[18,2.0837],[19,4.7849],[20,6.6211],[21,7.4479],[22,7.6054],[23,7.2927],[24,6.678],[25,5.9126],[26,4.9364],[27,3.5568],[28,1.4617],[29,-1.423],[30,-4.3193],[31,-6.4582],[32,-7.9661],[33,-9.1194],[34,-9.8362],[35,-10.0985],[36,-10.0393],[37,-9.661],[38,-9.0181],[40,-7.5096],[41,-6.866],[42,-6.4831],[43,-6.4054],[44,-6.6071],[45,-7.1679],[46,-8.025],[47,-8.953],[48,-9.7607],[49,-10.3243],[50,-10.6022],[51,-10.6176],[52,-10.3163],[53,-9.4337],[54,-7.8754],[55,-5.9632],[56,-4.1196],[57,-2.5076],[58,-1.0104],[59,0.5787],[60,2.5097],[61,4.7575],[62,6.5703],[63,7.4373],[64,7.6048],[65,7.2926],[66,6.678],[67,5.9126],[68,4.9364],[69,3.5568],[70,1.4617],[71,-1.423],[72,-4.3192],[73,-6.4582],[74,-7.9661],[75,-9.1194],[76,-9.8362],[77,-10.0985],[78,-10.0393],[79,-9.661],[80,-9.0181],[82,-7.5096],[83,-6.866],[84,-6.4831],[85,-6.4054],[86,-6.6071],[87,-7.1679],[88,-8.025],[89,-8.953],[90,-9.7607],[91,-10.3242],[92,-10.6033],[93,-10.6542],[94,-10.5992],[95,-10.5151],[96,-10.3902],[97,-9.9759],[98,-8.7213],[99,-6.3151],[100,-3.0306],[101,0.5594],[102,3.8283],[103,6.1857],[104,7.2703],[105,7.2482],[106,6.6758],[107,5.9125],[108,4.9364],[109,3.5568],[110,1.4617],[111,-1.423],[112,-4.3192],[113,-6.4582],[114,-7.9661],[115,-9.1194],[116,-9.8362],[117,-10.0985],[118,-10.0393],[119,-9.661],[120,-9.0181],[122,-7.5096],[123,-6.866],[124,-6.4831],[125,-6.4054],[126,-6.6071],[127,-7.1679],[128,-8.025],[129,-8.953],[130,-9.7607],[131,-10.3242],[132,-10.6033],[133,-10.6542],[134,-10.5992],[135,-10.5142],[136,-10.3718],[137,-9.8408],[138,-8.2201],[139,-5.1859],[140,-1.2416],[141,2.6931],[142,5.6824],[143,7.1346],[144,7.2297],[145,6.6749],[146,5.9125],[147,4.9364],[148,3.5568],[149,1.4617],[150,-1.423],[151,-4.3192],[152,-6.4575],[153,-7.9626],[154,-9.1155],[155,-9.8508],[156,-10.163],[157,-10.2349],[159,-10.178],[161,-10.3394]]
       }
       upperarm_l_fk_ctrl.Rotation.Z {
        type=MovieSceneScriptingFloatChannel
        default=5.0959
        thinned_from=162
        keys [[0,-5.2261],[9,-5.2302],[10,-5.2575],[11,-5.3575],[12,-5.6054],[13,-6.0181],[14,-6.3996],[15,-6.5711],[16,-6.533],[17,-6.2754],[19,-5.4453],[20,-5.1451],[21,-4.9956],[22,-4.9417],[23,-4.9343],[24,-4.9455],[25,-4.968],[26,-5.0271],[27,-5.1835],[28,-5.4966],[29,-5.9318],[30,-6.2539],[31,-6.2919],[32,-6.1278],[33,-5.8638],[34,-5.55],[35,-5.2124],[36,-4.8412],[37,-4.4241],[39,-3.5335],[40,-3.1216],[41,-2.7891],[42,-2.6023],[43,-2.5812],[44,-2.724],[45,-3.0858],[46,-3.6285],[47,-4.2154],[48,-4.7268],[49,-5.084],[50,-5.2642],[51,-5.3194],[52,-5.3875],[53,-5.6064],[55,-6.3457],[56,-6.5425],[57,-6.5249],[58,-6.3607],[59,-6.115],[60,-5.8058],[61,-5.4488],[62,-5.1524],[63,-4.9971],[64,-4.9418],[65,-4.9343],[66,-4.9455],[67,-4.968],[68,-5.0271],[69,-5.1835],[70,-5.4966],[71,-5.9318],[72,-6.2539],[73,-6.2919],[74,-6.1278],[75,-5.8638],[76,-5.55],[77,-5.2124],[78,-4.8412],[79,-4.4241],[81,-3.5335],[82,-3.1216],[83,-2.7891],[84,-2.6023],[85,-2.5812],[86,-2.724],[87,-3.0858],[88,-3.6285],[89,-4.2154],[90,-4.7268],[91,-5.0843],[92,-5.265],[93,-5.309],[94,-5.2945],[98,-5.1888],[102,-4.9814],[103,-4.946],[104,-4.9305],[105,-4.9328],[107,-4.968],[108,-5.0271],[109,-5.1835],[110,-5.4966],[111,-5.9318],[112,-6.2539],[113,-6.2919],[114,-6.1278],[115,-5.8638],[116,-5.55],[117,-5.2124],[118,-4.8412],[119,-4.4241],[121,-3.5335],[122,-3.1216],[123,-2.7891],[124,-2.6023],[125,-2.5812],[126,-2.724],[127,-3.0858],[128,-3.6285],[129,-4.2154],[130,-4.7268],[131,-5.0843],[132,-5.265],[133,-5.309],[134,-5.2945],[137,-5.2179],[138,-5.1814],[139,-5.1269],[141,-5.0001],[142,-4.9542],[143,-4.9326],[144,-4.9331],[145,-4.9455],[146,-4.968],[147,-5.0271],[148,-5.1835],[149,-5.4966],[150,-5.9318],[151,-6.2539],[152,-6.2914],[153,-6.1254],[154,-5.8602],[155,-5.5584],[156,-5.2808],[157,-5.0884],[158,-5.0013],[159,-5.0056],[161,-5.1746]]
       }
       upperarm_l_fk_ctrl.Scale.X {type=MovieSceneScriptingFloatChannel, default=1, const [1.0005,162]}
       upperarm_l_fk_ctrl.Scale.Y {type=MovieSceneScriptingFloatChannel, default=1, const [1.0005,162]}
       upperarm_l_fk_ctrl.Scale.Z {type=MovieSceneScriptingFloatChannel, default=1, const [1.0005,162]}
       lowerarm_l_fk_ctrl.Location.X {type=MovieSceneScriptingFloatChannel, default=0, const [0,162]}
       lowerarm_l_fk_ctrl.Location.Y {type=MovieSceneScriptingFloatChannel, default=0, const [0,162]}
       lowerarm_l_fk_ctrl.Location.Z {type=MovieSceneScriptingFloatChannel, default=0, const [0,162]}
       lowerarm_l_fk_ctrl.Rotation.X {
        type=MovieSceneScriptingFloatChannel
        default=3.7112
        thinned_from=162
        keys [[0,1.0501],[2,1.0544],[8,1.0887],[9,1.093],[10,1.0949],[11,1.0931],[12,1.0861],[14,1.0614],[15,1.0549],[16,1.0539],[17,1.0573],[19,1.0695],[20,1.0735],[21,1.0746],[22,1.0735],[26,1.0529],[56,1.0519],[59,1.0597],[62,1.0734],[63,1.0746],[64,1.0735],[68,1.0529],[97,1.0501],[99,1.0552],[102,1.0694],[103,1.0726],[104,1.0729],[105,1.0694],[107,1.0573],[108,1.0529],[137,1.0503],[139,1.0568],[142,1.0719],[143,1.0727],[144,1.0694],[146,1.0573],[147,1.0529],[161,1.0495]]
       }
       lowerarm_l_fk_ctrl.Rotation.Y {
        type=MovieSceneScriptingFloatChannel
        default=-1.8544
        thinned_from=162
        keys [[0,0.3956],[1,0.3904],[2,0.3797],[3,0.3636],[4,0.3432],[5,0.3203],[6,0.2968],[7,0.2739],[8,0.2533],[9,0.2372],[10,0.2302],[11,0.2372],[12,0.2636],[13,0.3096],[14,0.3501],[15,0.3624],[16,0.3434],[17,0.2838],[19,0.1011],[20,0.0402],[21,0.0188],[22,0.0288],[23,0.0727],[24,0.1421],[25,0.2148],[26,0.2727],[27,0.3096],[28,0.3324],[29,0.3486],[30,0.3615],[31,0.3728],[32,0.3826],[33,0.3901],[34,0.3947],[35,0.3968],[51,0.3976],[52,0.3967],[53,0.3935],[54,0.387],[55,0.3772],[56,0.3624],[57,0.3371],[58,0.298],[59,0.2459],[60,0.1793],[61,0.1017],[62,0.0418],[63,0.0192],[64,0.0288],[65,0.0727],[66,0.1421],[67,0.2148],[68,0.2727],[69,0.3096],[70,0.3324],[71,0.3486],[72,0.3615],[73,0.3728],[74,0.3826],[75,0.3901],[76,0.3947],[77,0.3968],[95,0.3977],[96,0.3966],[97,0.3888],[98,0.3628],[99,0.3119],[100,0.2419],[101,0.1653],[102,0.0955],[103,0.0471],[104,0.0367],[105,0.0737],[106,0.1422],[107,0.2148],[108,0.2727],[109,0.3096],[110,0.3324],[111,0.3486],[112,0.3615],[113,0.3728],[114,0.3826],[115,0.3901],[116,0.3947],[117,0.3968],[135,0.3977],[136,0.3962],[137,0.3859],[138,0.352],[139,0.2877],[141,0.1197],[142,0.0578],[143,0.0395],[144,0.0741],[145,0.1422],[146,0.2148],[147,0.2727],[148,0.3096],[149,0.3324],[150,0.3486],[151,0.3615],[152,0.3728],[153,0.3826],[154,0.3901],[155,0.3947],[156,0.3968],[161,0.3978]]
       }
       lowerarm_l_fk_ctrl.Rotation.Z {
        type=MovieSceneScriptingFloatChannel
        default=1.4608
        thinned_from=162
        keys [[0,58.1905],[2,57.9932],[3,58.1334],[4,58.4961],[5,59.0404],[6,59.7929],[10,63.8424],[12,65.4698],[14,66.7671],[18,68.6865],[19,69.0288],[20,69.2169],[21,69.2202],[22,69.0257],[23,68.5657],[25,67.2372],[26,66.7954],[27,66.6424],[28,66.7251],[29,67.0785],[32,68.9897],[33,69.5195],[34,69.7943],[35,69.7472],[36,69.3675],[37,68.5755],[38,67.4266],[40,64.8307],[41,63.7412],[42,63.1063],[43,63.0084],[44,63.4139],[45,64.4515],[47,67.4903],[48,68.4848],[49,68.7908],[50,68.4933],[52,66.8035],[53,66.6324],[57,68.0401],[62,69.2101],[63,69.2188],[64,69.0256],[65,68.5657],[67,67.2372],[68,66.7954],[69,66.6424],[70,66.7251],[71,67.0785],[74,68.9897],[75,69.5195],[76,69.7943],[77,69.7472],[78,69.3675],[79,68.5755],[80,67.4266],[82,64.8307],[83,63.7412],[84,63.1063],[85,63.0084],[86,63.4139],[87,64.4515],[89,67.4903],[90,68.4857],[91,68.8071],[92,68.5727],[93,67.692],[95,64.7881],[96,63.4841],[97,62.7777],[98,62.9432],[99,63.8013],[102,67.6503],[103,68.5321],[104,68.8445],[105,68.5508],[107,67.2383],[108,66.7955],[109,66.6424],[110,66.7251],[111,67.0785],[114,68.9897],[115,69.5195],[116,69.7943],[117,69.7472],[118,69.3675],[119,68.5755],[120,67.4266],[122,64.8307],[123,63.7412],[124,63.1063],[125,63.0084],[126,63.4139],[127,64.4515],[129,67.4903],[130,68.4857],[131,68.8071],[132,68.5727],[133,67.692],[135,64.7873],[136,63.468],[137,62.6558],[138,62.4658],[142,63.3363],[143,63.2932],[144,62.8937],[146,61.5652],[147,61.1223],[148,60.9692],[149,61.0522],[150,61.4119],[152,62.6921],[153,63.1341],[154,63.2201],[155,63.065],[156,62.7174],[157,62.1012],[158,61.1673],[159,60.0713],[160,59.0984],[161,58.5282]]
       }
       lowerarm_l_fk_ctrl.Scale.X {type=MovieSceneScriptingFloatChannel, default=1, const [1.0005,162]}
       lowerarm_l_fk_ctrl.Scale.Y {type=MovieSceneScriptingFloatChannel, default=1, const [1.0005,162]}
       lowerarm_l_fk_ctrl.Scale.Z {type=MovieSceneScriptingFloatChannel, default=1, const [1.0005,162]}
       hand_l_fk_ctrl.Location.X {type=MovieSceneScriptingFloatChannel, default=0, thinned_from=162, keys [[0,0],[161,0]]}
       hand_l_fk_ctrl.Location.Y {
        type=MovieSceneScriptingFloatChannel
        default=0
        thinned_from=162
        keys [[0,0.4312],[53,0.4312],[54,0.4307],[55,0.427],[56,0.4143],[57,0.3875],[58,0.3461],[59,0.292],[60,0.2278],[61,0.1561],[62,0.0793],[64,-0.0793],[65,-0.1561],[66,-0.2278],[67,-0.292],[68,-0.346],[69,-0.3875],[70,-0.4143],[71,-0.427],[72,-0.4307],[142,-0.4308],[143,-0.4279],[144,-0.4177],[145,-0.3962],[146,-0.3626],[147,-0.3185],[148,-0.2654],[149,-0.2053],[150,-0.1398],[151,-0.0708],[153,0.0708],[154,0.1398],[155,0.2053],[156,0.2654],[157,0.3185],[158,0.3626],[159,0.3962],[160,0.4177],[161,0.4276]]
       }
       hand_l_fk_ctrl.Location.Z {
        type=MovieSceneScriptingFloatChannel
        default=0
        thinned_from=162
        keys [[0,-0.0668],[54,-0.0668],[55,-0.0662],[56,-0.0642],[57,-0.0601],[58,-0.0536],[59,-0.0453],[60,-0.0353],[61,-0.0242],[62,-0.0123],[64,0.0123],[65,0.0242],[66,0.0353],[67,0.0453],[68,0.0536],[69,0.0601],[70,0.0642],[71,0.0662],[72,0.0668],[142,0.0668],[143,0.0663],[144,0.0647],[145,0.0614],[146,0.0562],[147,0.0494],[148,0.0411],[149,0.0318],[150,0.0217],[151,0.011],[153,-0.011],[154,-0.0217],[155,-0.0318],[156,-0.0411],[157,-0.0494],[158,-0.0562],[159,-0.0614],[160,-0.0647],[161,-0.0663]]
       }
       hand_l_fk_ctrl.Rotation.X {
        type=MovieSceneScriptingFloatChannel
        default=0
        thinned_from=162
        keys [[0,-1.074],[24,-1.0742],[25,-1.0778],[26,-1.1028],[27,-1.1973],[28,-1.4554],[29,-1.9989],[30,-2.8231],[31,-3.8029],[32,-4.8181],[33,-5.7636],[34,-6.5428],[35,-7.1093],[36,-7.5302],[37,-7.8924],[40,-8.8986],[42,-9.5309],[44,-10.1145],[46,-10.6352],[48,-11.0794],[49,-11.2693],[50,-11.4337],[51,-11.4875],[52,-11.0056],[53,-9.2858],[55,-3.4783],[56,-1.7301],[57,-1.1639],[58,-1.0785],[59,-1.074],[66,-1.0742],[67,-1.0778],[68,-1.1028],[69,-1.1973],[70,-1.4554],[71,-1.9989],[72,-2.8231],[73,-3.8029],[74,-4.8181],[75,-5.7636],[76,-6.5428],[77,-7.1093],[78,-7.5302],[79,-7.8924],[82,-8.8986],[84,-9.5309],[86,-10.1145],[88,-10.6352],[90,-11.0793],[91,-11.2683],[92,-11.4328],[93,-11.5712],[94,-11.682],[95,-11.7617],[96,-11.7805],[97,-11.5874],[98,-10.8717],[99,-9.461],[100,-7.5218],[101,-5.396],[102,-3.4567],[103,-2.0451],[104,-1.3229],[105,-1.1069],[106,-1.0758],[107,-1.0778],[108,-1.1028],[109,-1.1973],[110,-1.4554],[111,-1.9989],[112,-2.8231],[113,-3.8029],[114,-4.8181],[115,-5.7636],[116,-6.5428],[117,-7.1093],[118,-7.5302],[119,-7.8924],[122,-8.8986],[124,-9.5309],[126,-10.1145],[128,-10.6352],[130,-11.0793],[131,-11.2683],[132,-11.4328],[133,-11.5712],[134,-11.682],[135,-11.7611],[136,-11.7695],[137,-11.5071],[138,-10.5739],[139,-8.7901],[141,-4.1276],[142,-2.3428],[143,-1.4032],[144,-1.1179],[145,-1.0763],[146,-1.0778],[147,-1.1028],[148,-1.1973],[149,-1.4554],[150,-1.9989],[151,-2.8235],[152,-3.8104],[153,-4.8533],[154,-5.7685],[155,-6.1338],[156,-5.6582],[157,-4.6727],[158,-3.6353],[159,-2.6667],[160,-1.8221],[161,-1.312]]
       }
       hand_l_fk_ctrl.Rotation.Y {
        type=MovieSceneScriptingFloatChannel
        default=0
        thinned_from=162
        keys [[0,9.4713],[25,9.4791],[26,9.53],[27,9.7219],[28,10.2462],[29,11.3504],[30,13.0248],[31,15.0155],[32,17.0779],[33,18.999],[34,20.5819],[35,21.733],[36,22.5881],[39,24.6985],[42,26.6528],[44,27.8383],[46,28.8963],[48,29.7987],[50,30.5184],[51,30.6279],[52,29.6487],[53,26.1547],[55,14.356],[56,10.8043],[57,9.654],[58,9.4805],[67,9.4791],[68,9.53],[69,9.7219],[70,10.2462],[71,11.3504],[72,13.0248],[73,15.0155],[74,17.0779],[75,18.999],[76,20.5819],[77,21.733],[78,22.5881],[81,24.6985],[84,26.6528],[86,27.8383],[88,28.8963],[90,29.7986],[92,30.5167],[94,31.0231],[95,31.1849],[96,31.223],[97,30.8308],[98,29.3767],[99,26.5107],[100,22.5709],[101,18.2521],[102,14.3121],[103,11.4443],[104,9.977],[105,9.5382],[106,9.475],[107,9.4791],[108,9.53],[109,9.7219],[110,10.2462],[111,11.3504],[112,13.0248],[113,15.0155],[114,17.0779],[115,18.999],[116,20.5819],[117,21.733],[118,22.5881],[121,24.6985],[124,26.6528],[126,27.8383],[128,28.8963],[130,29.7986],[132,30.5167],[134,31.0231],[135,31.1838],[136,31.2008],[137,30.6677],[138,28.7718],[139,25.1477],[141,15.6752],[142,12.0492],[143,10.1402],[144,9.5605],[145,9.4761],[146,9.4791],[147,9.53],[148,9.7219],[149,10.2462],[150,11.3504],[151,13.0257],[152,15.0307],[153,17.1496],[154,19.0089],[155,19.7511],[156,18.7848],[157,16.7826],[158,14.6749],[159,12.7072],[160,10.9912],[161,9.9548]]
       }
       hand_l_fk_ctrl.Rotation.Z {
        type=MovieSceneScriptingFloatChannel
        default=0
        thinned_from=162
        keys [[0,-3.4393],[24,-3.4396],[25,-3.4451],[26,-3.4834],[27,-3.6277],[28,-4.0219],[29,-4.8522],[30,-6.1111],[31,-7.6079],[32,-9.1587],[33,-10.6031],[34,-11.7933],[35,-12.6588],[36,-13.3017],[37,-13.855],[40,-15.3921],[42,-16.358],[44,-17.2494],[46,-18.0449],[48,-18.7234],[49,-19.0135],[50,-19.2646],[51,-19.3469],[52,-18.6106],[53,-15.9835],[55,-7.112],[56,-4.4415],[57,-3.5766],[58,-3.4462],[66,-3.4396],[67,-3.4451],[68,-3.4834],[69,-3.6277],[70,-4.0219],[71,-4.8522],[72,-6.1111],[73,-7.6079],[74,-9.1587],[75,-10.6031],[76,-11.7933],[77,-12.6588],[78,-13.3017],[79,-13.855],[82,-15.3921],[84,-16.358],[86,-17.2494],[88,-18.0449],[90,-18.7233],[92,-19.2632],[93,-19.4748],[94,-19.644],[95,-19.7657],[96,-19.7944],[97,-19.4994],[98,-18.4061],[99,-16.2512],[100,-13.2888],[101,-10.0415],[102,-7.0791],[103,-4.9228],[104,-3.8195],[105,-3.4896],[106,-3.442],[107,-3.4451],[108,-3.4834],[109,-3.6277],[110,-4.0219],[111,-4.8522],[112,-6.1111],[113,-7.6079],[114,-9.1587],[115,-10.6031],[116,-11.7933],[117,-12.6588],[118,-13.3017],[119,-13.855],[122,-15.3921],[124,-16.358],[126,-17.2494],[128,-18.0449],[130,-18.7233],[132,-19.2632],[133,-19.4748],[134,-19.644],[135,-19.7649],[136,-19.7777],[137,-19.3768],[138,-17.9513],[139,-15.2263],[141,-8.1039],[142,-5.3776],[143,-3.9422],[144,-3.5063],[145,-3.4429],[146,-3.4452],[147,-3.4834],[148,-3.6277],[149,-4.0219],[150,-4.8522],[151,-6.1118],[152,-7.6193],[153,-9.2125],[154,-10.6106],[155,-11.1686],[156,-10.4421],[157,-8.9366],[158,-7.3519],[159,-5.8723],[160,-4.5821],[161,-3.8028]]
       }
       hand_l_fk_ctrl.Scale.X {type=MovieSceneScriptingFloatChannel, default=1, const [1.0005,162]}
       hand_l_fk_ctrl.Scale.Y {type=MovieSceneScriptingFloatChannel, default=1, const [1.0005,162]}
       hand_l_fk_ctrl.Scale.Z {type=MovieSceneScriptingFloatChannel, default=1, const [1.0005,162]}
       thumb_01_l_ctrl.Location.X {type=MovieSceneScriptingFloatChannel, default=0, const [0,1]}
       thumb_01_l_ctrl.Location.Y {type=MovieSceneScriptingFloatChannel, default=0, const [0,1]}
       thumb_01_l_ctrl.Location.Z {type=MovieSceneScriptingFloatChannel, default=0, const [0,1]}
       thumb_01_l_ctrl.Rotation.X {type=MovieSceneScriptingFloatChannel, default=0, const [26.151,1]}
       thumb_01_l_ctrl.Rotation.Y {type=MovieSceneScriptingFloatChannel, default=0, const [10.6122,1]}
       thumb_01_l_ctrl.Rotation.Z {type=MovieSceneScriptingFloatChannel, default=0, const [-8.5664,1]}
       thumb_01_l_ctrl.Scale.X {type=MovieSceneScriptingFloatChannel, default=1, const [1,1]}
       thumb_01_l_ctrl.Scale.Y {type=MovieSceneScriptingFloatChannel, default=1, const [1,1]}
       thumb_01_l_ctrl.Scale.Z {type=MovieSceneScriptingFloatChannel, default=1, const [1,1]}
       thumb_02_l_ctrl.Location.X {type=MovieSceneScriptingFloatChannel, default=0, const [0,1]}
       thumb_02_l_ctrl.Location.Y {type=MovieSceneScriptingFloatChannel, default=0, const [0,1]}
       thumb_02_l_ctrl.Location.Z {type=MovieSceneScriptingFloatChannel, default=0, const [0,1]}
       thumb_02_l_ctrl.Rotation.X {type=MovieSceneScriptingFloatChannel, default=0, const [0,1]}
       thumb_02_l_ctrl.Rotation.Y {type=MovieSceneScriptingFloatChannel, default=0, const [0,1]}
       thumb_02_l_ctrl.Rotation.Z {type=MovieSceneScriptingFloatChannel, default=0, const [0.0073,1]}
       thumb_02_l_ctrl.Scale.X {type=MovieSceneScriptingFloatChannel, default=1, const [1,1]}
       thumb_02_l_ctrl.Scale.Y {type=MovieSceneScriptingFloatChannel, default=1, const [1,1]}
       thumb_02_l_ctrl.Scale.Z {type=MovieSceneScriptingFloatChannel, default=1, const [1,1]}
       thumb_03_l_ctrl.Location.X {type=MovieSceneScriptingFloatChannel, default=0, const [0,1]}
       thumb_03_l_ctrl.Location.Y {type=MovieSceneScriptingFloatChannel, default=0, const [0,1]}
       thumb_03_l_ctrl.Location.Z {type=MovieSceneScriptingFloatChannel, default=0, const [0,1]}
       thumb_03_l_ctrl.Rotation.X {type=MovieSceneScriptingFloatChannel, default=0, const [0,1]}
       thumb_03_l_ctrl.Rotation.Y {type=MovieSceneScriptingFloatChannel, default=0, const [0,1]}
       thumb_03_l_ctrl.Rotation.Z {type=MovieSceneScriptingFloatChannel, default=0, const [-0.0004,1]}
       thumb_03_l_ctrl.Scale.X {type=MovieSceneScriptingFloatChannel, default=1, const [1,1]}
       thumb_03_l_ctrl.Scale.Y {type=MovieSceneScriptingFloatChannel, default=1, const [1,1]}
       thumb_03_l_ctrl.Scale.Z {type=MovieSceneScriptingFloatChannel, default=1, const [1,1]}
       index_metacarpal_l_ctrl.Location.X {type=MovieSceneScriptingFloatChannel, default=0, const [0,1]}
       index_metacarpal_l_ctrl.Location.Y {type=MovieSceneScriptingFloatChannel, default=0, const [0,1]}
       index_metacarpal_l_ctrl.Location.Z {type=MovieSceneScriptingFloatChannel, default=0, const [0,1]}
       index_metacarpal_l_ctrl.Rotation.X {type=MovieSceneScriptingFloatChannel, default=0, const [0.955,1]}
       index_metacarpal_l_ctrl.Rotation.Y {type=MovieSceneScriptingFloatChannel, default=0, const [3.1321,1]}
       index_metacarpal_l_ctrl.Rotation.Z {type=MovieSceneScriptingFloatChannel, default=0, const [0.0216,1]}
       index_metacarpal_l_ctrl.Scale.X {type=MovieSceneScriptingFloatChannel, default=1, const [1,1]}
       index_metacarpal_l_ctrl.Scale.Y {type=MovieSceneScriptingFloatChannel, default=1, const [1,1]}
       index_metacarpal_l_ctrl.Scale.Z {type=MovieSceneScriptingFloatChannel, default=1, const [1,1]}
       index_01_l_ctrl.Location.X {type=MovieSceneScriptingFloatChannel, default=0, const [0,1]}
       index_01_l_ctrl.Location.Y {type=MovieSceneScriptingFloatChannel, default=0, const [0,1]}
       index_01_l_ctrl.Location.Z {type=MovieSceneScriptingFloatChannel, default=0, const [0,1]}
       index_01_l_ctrl.Rotation.X {type=MovieSceneScriptingFloatChannel, default=0, const [0,1]}
       index_01_l_ctrl.Rotation.Y {type=MovieSceneScriptingFloatChannel, default=0, const [0,1]}
       index_01_l_ctrl.Rotation.Z {type=MovieSceneScriptingFloatChannel, default=0, const [-59.5276,1]}
       index_01_l_ctrl.Scale.X {type=MovieSceneScriptingFloatChannel, default=1, const [1,1]}
       index_01_l_ctrl.Scale.Y {type=MovieSceneScriptingFloatChannel, default=1, const [1,1]}
       index_01_l_ctrl.Scale.Z {type=MovieSceneScriptingFloatChannel, default=1, const [1,1]}
       index_02_l_ctrl.Location.X {type=MovieSceneScriptingFloatChannel, default=0, const [0,1]}
       index_02_l_ctrl.Location.Y {type=MovieSceneScriptingFloatChannel, default=0, const [0,1]}
       index_02_l_ctrl.Location.Z {type=MovieSceneScriptingFloatChannel, default=0, const [0,1]}
       index_02_l_ctrl.Rotation.X {type=MovieSceneScriptingFloatChannel, default=0, const [0,1]}
       index_02_l_ctrl.Rotation.Y {type=MovieSceneScriptingFloatChannel, default=0, const [0,1]}
       index_02_l_ctrl.Rotation.Z {type=MovieSceneScriptingFloatChannel, default=0, const [-53.0237,1]}
       index_02_l_ctrl.Scale.X {type=MovieSceneScriptingFloatChannel, default=1, const [1,1]}
       index_02_l_ctrl.Scale.Y {type=MovieSceneScriptingFloatChannel, default=1, const [1,1]}
       index_02_l_ctrl.Scale.Z {type=MovieSceneScriptingFloatChannel, default=1, const [1,1]}
       index_03_l_ctrl.Location.X {type=MovieSceneScriptingFloatChannel, default=0, const [0,1]}
       index_03_l_ctrl.Location.Y {type=MovieSceneScriptingFloatChannel, default=0, const [0,1]}
       index_03_l_ctrl.Location.Z {type=MovieSceneScriptingFloatChannel, default=0, const [0,1]}
       index_03_l_ctrl.Rotation.X {type=MovieSceneScriptingFloatChannel, default=0, const [0,1]}
       index_03_l_ctrl.Rotation.Y {type=MovieSceneScriptingFloatChannel, default=0, const [0,1]}
       index_03_l_ctrl.Rotation.Z {type=MovieSceneScriptingFloatChannel, default=0, const [-56.0042,1]}
       index_03_l_ctrl.Scale.X {type=MovieSceneScriptingFloatChannel, default=1, const [1,1]}
       index_03_l_ctrl.Scale.Y {type=MovieSceneScriptingFloatChannel, default=1, const [1,1]}
       index_03_l_ctrl.Scale.Z {type=MovieSceneScriptingFloatChannel, default=1, const [1,1]}
       middle_metacarpal_l_ctrl.Location.X {type=MovieSceneScriptingFloatChannel, default=0, const [0,1]}
       middle_metacarpal_l_ctrl.Location.Y {type=MovieSceneScriptingFloatChannel, default=0, const [0,1]}
       middle_metacarpal_l_ctrl.Location.Z {type=MovieSceneScriptingFloatChannel, default=0, const [0,1]}
       middle_metacarpal_l_ctrl.Rotation.X {type=MovieSceneScriptingFloatChannel, default=0, const [0,1]}
       middle_metacarpal_l_ctrl.Rotation.Y {type=MovieSceneScriptingFloatChannel, default=0, const [0,1]}
       middle_metacarpal_l_ctrl.Rotation.Z {type=MovieSceneScriptingFloatChannel, default=0, const [-4.0425,1]}
       middle_metacarpal_l_ctrl.Scale.X {type=MovieSceneScriptingFloatChannel, default=1, const [1,1]}
       middle_metacarpal_l_ctrl.Scale.Y {type=MovieSceneScriptingFloatChannel, default=1, const [1,1]}
       middle_metacarpal_l_ctrl.Scale.Z {type=MovieSceneScriptingFloatChannel, default=1, const [1,1]}
       middle_01_l_ctrl.Location.X {type=MovieSceneScriptingFloatChannel, default=0, const [0,1]}
       middle_01_l_ctrl.Location.Y {type=MovieSceneScriptingFloatChannel, default=0, const [0,1]}
       middle_01_l_ctrl.Location.Z {type=MovieSceneScriptingFloatChannel, default=0, const [0,1]}
       middle_01_l_ctrl.Rotation.X {type=MovieSceneScriptingFloatChannel, default=0, const [0,1]}
       middle_01_l_ctrl.Rotation.Y {type=MovieSceneScriptingFloatChannel, default=0, const [0,1]}
       middle_01_l_ctrl.Rotation.Z {type=MovieSceneScriptingFloatChannel, default=0, const [-52.9974,1]}
       middle_01_l_ctrl.Scale.X {type=MovieSceneScriptingFloatChannel, default=1, const [1,1]}
       middle_01_l_ctrl.Scale.Y {type=MovieSceneScriptingFloatChannel, default=1, const [1,1]}
       middle_01_l_ctrl.Scale.Z {type=MovieSceneScriptingFloatChannel, default=1, const [1,1]}
       middle_02_l_ctrl.Location.X {type=MovieSceneScriptingFloatChannel, default=0, const [0,1]}
       middle_02_l_ctrl.Location.Y {type=MovieSceneScriptingFloatChannel, default=0, const [0,1]}
       middle_02_l_ctrl.Location.Z {type=MovieSceneScriptingFloatChannel, default=0, const [0,1]}
       middle_02_l_ctrl.Rotation.X {type=MovieSceneScriptingFloatChannel, default=0, const [0,1]}
       middle_02_l_ctrl.Rotation.Y {type=MovieSceneScriptingFloatChannel, default=0, const [0,1]}
       middle_02_l_ctrl.Rotation.Z {type=MovieSceneScriptingFloatChannel, default=0, const [-65.0474,1]}
       middle_02_l_ctrl.Scale.X {type=MovieSceneScriptingFloatChannel, default=1, const [1,1]}
       middle_02_l_ctrl.Scale.Y {type=MovieSceneScriptingFloatChannel, default=1, const [1,1]}
       middle_02_l_ctrl.Scale.Z {type=MovieSceneScriptingFloatChannel, default=1, const [1,1]}
       middle_03_l_ctrl.Location.X {type=MovieSceneScriptingFloatChannel, default=0, const [0,1]}
       middle_03_l_ctrl.Location.Y {type=MovieSceneScriptingFloatChannel, default=0, const [0,1]}
       middle_03_l_ctrl.Location.Z {type=MovieSceneScriptingFloatChannel, default=0, const [0,1]}
       middle_03_l_ctrl.Rotation.X {type=MovieSceneScriptingFloatChannel, default=0, const [0,1]}
       middle_03_l_ctrl.Rotation.Y {type=MovieSceneScriptingFloatChannel, default=0, const [0,1]}
       middle_03_l_ctrl.Rotation.Z {type=MovieSceneScriptingFloatChannel, default=0, const [-69.6807,1]}
       middle_03_l_ctrl.Scale.X {type=MovieSceneScriptingFloatChannel, default=1, const [1,1]}
       middle_03_l_ctrl.Scale.Y {type=MovieSceneScriptingFloatChannel, default=1, const [1,1]}
       middle_03_l_ctrl.Scale.Z {type=MovieSceneScriptingFloatChannel, default=1, const [1,1]}
       ring_metacarpal_l_ctrl.Location.X {type=MovieSceneScriptingFloatChannel, default=0, const [0,1]}
       ring_metacarpal_l_ctrl.Location.Y {type=MovieSceneScriptingFloatChannel, default=0, const [0,1]}
       ring_metacarpal_l_ctrl.Location.Z {type=MovieSceneScriptingFloatChannel, default=0, const [0,1]}
       ring_metacarpal_l_ctrl.Rotation.X {type=MovieSceneScriptingFloatChannel, default=0, const [0,1]}
       ring_metacarpal_l_ctrl.Rotation.Y {type=MovieSceneScriptingFloatChannel, default=0, const [0,1]}
       ring_metacarpal_l_ctrl.Rotation.Z {type=MovieSceneScriptingFloatChannel, default=0, const [-4.0425,1]}
       ring_metacarpal_l_ctrl.Scale.X {type=MovieSceneScriptingFloatChannel, default=1, const [1,1]}
       ring_metacarpal_l_ctrl.Scale.Y {type=MovieSceneScriptingFloatChannel, default=1, const [1,1]}
       ring_metacarpal_l_ctrl.Scale.Z {type=MovieSceneScriptingFloatChannel, default=1, const [1,1]}
       ring_01_l_ctrl.Location.X {type=MovieSceneScriptingFloatChannel, default=0, const [0,1]}
       ring_01_l_ctrl.Location.Y {type=MovieSceneScriptingFloatChannel, default=0, const [0,1]}
       ring_01_l_ctrl.Location.Z {type=MovieSceneScriptingFloatChannel, default=0, const [0,1]}
       ring_01_l_ctrl.Rotation.X {type=MovieSceneScriptingFloatChannel, default=0, const [0,1]}
       ring_01_l_ctrl.Rotation.Y {type=MovieSceneScriptingFloatChannel, default=0, const [0,1]}
       ring_01_l_ctrl.Rotation.Z {type=MovieSceneScriptingFloatChannel, default=0, const [-47.3267,1]}
       ring_01_l_ctrl.Scale.X {type=MovieSceneScriptingFloatChannel, default=1, const [1,1]}
       ring_01_l_ctrl.Scale.Y {type=MovieSceneScriptingFloatChannel, default=1, const [1,1]}
       ring_01_l_ctrl.Scale.Z {type=MovieSceneScriptingFloatChannel, default=1, const [1,1]}
       ring_02_l_ctrl.Location.X {type=MovieSceneScriptingFloatChannel, default=0, const [0,1]}
       ring_02_l_ctrl.Location.Y {type=MovieSceneScriptingFloatChannel, default=0, const [0,1]}
       ring_02_l_ctrl.Location.Z {type=MovieSceneScriptingFloatChannel, default=0, const [0,1]}
       ring_02_l_ctrl.Rotation.X {type=MovieSceneScriptingFloatChannel, default=0, const [0,1]}
       ring_02_l_ctrl.Rotation.Y {type=MovieSceneScriptingFloatChannel, default=0, const [0,1]}
       ring_02_l_ctrl.Rotation.Z {type=MovieSceneScriptingFloatChannel, default=0, const [-63.5055,1]}
       ring_02_l_ctrl.Scale.X {type=MovieSceneScriptingFloatChannel, default=1, const [1,1]}
       ring_02_l_ctrl.Scale.Y {type=MovieSceneScriptingFloatChannel, default=1, const [1,1]}
       ring_02_l_ctrl.Scale.Z {type=MovieSceneScriptingFloatChannel, default=1, const [1,1]}
       ring_03_l_ctrl.Location.X {type=MovieSceneScriptingFloatChannel, default=0, const [0,1]}
       ring_03_l_ctrl.Location.Y {type=MovieSceneScriptingFloatChannel, default=0, const [0,1]}
       ring_03_l_ctrl.Location.Z {type=MovieSceneScriptingFloatChannel, default=0, const [0,1]}
       ring_03_l_ctrl.Rotation.X {type=MovieSceneScriptingFloatChannel, default=0, const [0,1]}
       ring_03_l_ctrl.Rotation.Y {type=MovieSceneScriptingFloatChannel, default=0, const [0,1]}
       ring_03_l_ctrl.Rotation.Z {type=MovieSceneScriptingFloatChannel, default=0, const [-50.5238,1]}
       ring_03_l_ctrl.Scale.X {type=MovieSceneScriptingFloatChannel, default=1, const [1,1]}
       ring_03_l_ctrl.Scale.Y {type=MovieSceneScriptingFloatChannel, default=1, const [1,1]}
       ring_03_l_ctrl.Scale.Z {type=MovieSceneScriptingFloatChannel, default=1, const [1,1]}
       pinky_metacarpal_l_ctrl.Location.X {type=MovieSceneScriptingFloatChannel, default=0, const [0,1]}
       pinky_metacarpal_l_ctrl.Location.Y {type=MovieSceneScriptingFloatChannel, default=0, const [0,1]}
       pinky_metacarpal_l_ctrl.Location.Z {type=MovieSceneScriptingFloatChannel, default=0, const [0,1]}
       pinky_metacarpal_l_ctrl.Rotation.X {type=MovieSceneScriptingFloatChannel, default=0, const [0,1]}
       pinky_metacarpal_l_ctrl.Rotation.Y {type=MovieSceneScriptingFloatChannel, default=0, const [0,1]}
       pinky_metacarpal_l_ctrl.Rotation.Z {type=MovieSceneScriptingFloatChannel, default=0, const [-4.0425,1]}
       pinky_metacarpal_l_ctrl.Scale.X {type=MovieSceneScriptingFloatChannel, default=1, const [1,1]}
       pinky_metacarpal_l_ctrl.Scale.Y {type=MovieSceneScriptingFloatChannel, default=1, const [1,1]}
       pinky_metacarpal_l_ctrl.Scale.Z {type=MovieSceneScriptingFloatChannel, default=1, const [1,1]}
       pinky_01_l_ctrl.Location.X {type=MovieSceneScriptingFloatChannel, default=0, const [0,1]}
       pinky_01_l_ctrl.Location.Y {type=MovieSceneScriptingFloatChannel, default=0, const [0,1]}
       pinky_01_l_ctrl.Location.Z {type=MovieSceneScriptingFloatChannel, default=0, const [0,1]}
       pinky_01_l_ctrl.Rotation.X {type=MovieSceneScriptingFloatChannel, default=0, const [0.0001,1]}
       pinky_01_l_ctrl.Rotation.Y {type=MovieSceneScriptingFloatChannel, default=0, const [0,1]}
       pinky_01_l_ctrl.Rotation.Z {type=MovieSceneScriptingFloatChannel, default=0, const [-48.5057,1]}
       pinky_01_l_ctrl.Scale.X {type=MovieSceneScriptingFloatChannel, default=1, const [1,1]}
       pinky_01_l_ctrl.Scale.Y {type=MovieSceneScriptingFloatChannel, default=1, const [1,1]}
       pinky_01_l_ctrl.Scale.Z {type=MovieSceneScriptingFloatChannel, default=1, const [1,1]}
       pinky_02_l_ctrl.Location.X {type=MovieSceneScriptingFloatChannel, default=0, const [0,1]}
       pinky_02_l_ctrl.Location.Y {type=MovieSceneScriptingFloatChannel, default=0, const [0,1]}
       pinky_02_l_ctrl.Location.Z {type=MovieSceneScriptingFloatChannel, default=0, const [0,1]}
       pinky_02_l_ctrl.Rotation.X {type=MovieSceneScriptingFloatChannel, default=0, const [0,1]}
       pinky_02_l_ctrl.Rotation.Y {type=MovieSceneScriptingFloatChannel, default=0, const [0,1]}
       pinky_02_l_ctrl.Rotation.Z {type=MovieSceneScriptingFloatChannel, default=0, const [-57.7751,1]}
       pinky_02_l_ctrl.Scale.X {type=MovieSceneScriptingFloatChannel, default=1, const [1,1]}
       pinky_02_l_ctrl.Scale.Y {type=MovieSceneScriptingFloatChannel, default=1, const [1,1]}
       pinky_02_l_ctrl.Scale.Z {type=MovieSceneScriptingFloatChannel, default=1, const [1,1]}
       pinky_03_l_ctrl.Location.X {type=MovieSceneScriptingFloatChannel, default=0, const [0,1]}
       pinky_03_l_ctrl.Location.Y {type=MovieSceneScriptingFloatChannel, default=0, const [0,1]}
       pinky_03_l_ctrl.Location.Z {type=MovieSceneScriptingFloatChannel, default=0, const [0,1]}
       pinky_03_l_ctrl.Rotation.X {type=MovieSceneScriptingFloatChannel, default=0, const [0,1]}
       pinky_03_l_ctrl.Rotation.Y {type=MovieSceneScriptingFloatChannel, default=0, const [0,1]}
       pinky_03_l_ctrl.Rotation.Z {type=MovieSceneScriptingFloatChannel, default=0, const [-61.3898,1]}
       pinky_03_l_ctrl.Scale.X {type=MovieSceneScriptingFloatChannel, default=1, const [1,1]}
       pinky_03_l_ctrl.Scale.Y {type=MovieSceneScriptingFloatChannel, default=1, const [1,1]}
       pinky_03_l_ctrl.Scale.Z {type=MovieSceneScriptingFloatChannel, default=1, const [1,1]}
       hand_l_ik_ctrl.Location.X {
        type=MovieSceneScriptingFloatChannel
        default=4.1588
        thinned_from=12
        keys [[0,-11.9158],[13,-13.1293],[17,-11.9158],[23,-13.7156],[35,5.0164],[55,-16.0981],[161,-16.0981]]
       }
       hand_l_ik_ctrl.Location.Y {
        type=MovieSceneScriptingFloatChannel
        default=8.6278
        thinned_from=12
        keys [[0,10.3568],[13,9.3098],[17,10.3568],[23,14.8396],[35,2.6883],[55,19.3138],[161,19.3138]]
       }
       hand_l_ik_ctrl.Location.Z {
        type=MovieSceneScriptingFloatChannel
        default=11.0864
        thinned_from=12
        keys [[0,18.9589],[13,18.8493],[17,18.9589],[23,13.8235],[35,3.0765],[55,14.4366],[161,14.4366]]
       }
       hand_l_ik_ctrl.Rotation.X {
        type=MovieSceneScriptingFloatChannel
        default=41.8572
        thinned_from=12
        keys [[0,51.0117],[13,32.5422],[17,51.0117],[23,-62.5934],[35,9.3843],[55,-67.2042],[161,-67.2042]]
       }
       hand_l_ik_ctrl.Rotation.Y {
        type=MovieSceneScriptingFloatChannel
        default=2.0272
        thinned_from=12
        keys [[0,77.7969],[13,81.5903],[17,77.7969],[23,76.0071],[35,12.6639],[55,74.3337],[161,74.3337]]
       }
       hand_l_ik_ctrl.Rotation.Z {
        type=MovieSceneScriptingFloatChannel
        default=34.5602
        thinned_from=12
        keys [[0,80.6345],[13,59.8524],[17,80.6345],[23,-41.4963],[35,9.7712],[55,-45.0588],[161,-45.0588]]
       }
       hand_l_ik_ctrl.Scale.X {type=MovieSceneScriptingFloatChannel, default=1, const [1,12]}
       hand_l_ik_ctrl.Scale.Y {type=MovieSceneScriptingFloatChannel, default=1, const [1,12]}
       hand_l_ik_ctrl.Scale.Z {type=MovieSceneScriptingFloatChannel, default=1, const [1,12]}
       Softness {type=MovieSceneScriptingFloatChannel, default=0, const [0,12]}
       hand_l_ikorient_ctrl.Rotation.Roll {type=MovieSceneScriptingFloatChannel, default=0, const [0,10]}
       hand_l_ikorient_ctrl.Rotation.Pitch {type=MovieSceneScriptingFloatChannel, default=0, const [0,10]}
       hand_l_ikorient_ctrl.Rotation.Yaw {type=MovieSceneScriptingFloatChannel, default=0, thinned_from=10, keys [[0,0],[161,0]]}
       arm_l_pv_ik_ctrl.Location.X {
        type=MovieSceneScriptingFloatChannel
        default=17.3769
        thinned_from=12
        keys [[0,18.4181],[13,15.9114],[17,18.4181],[23,-0.8338],[35,20.0617],[55,-7.0841],[161,-7.0841]]
       }
       arm_l_pv_ik_ctrl.Location.Y {
        type=MovieSceneScriptingFloatChannel
        default=21.5796
        thinned_from=12
        keys [[0,11.8896],[13,8.2626],[17,11.8896],[23,-4.0863],[35,13.5743],[55,-5.0655],[161,-5.0655]]
       }
       arm_l_pv_ik_ctrl.Location.Z {
        type=MovieSceneScriptingFloatChannel
        default=-2.1953
        thinned_from=12
        keys [[0,-0.3428],[13,1.1053],[17,-0.3428],[23,-2.4716],[35,0.0759],[55,-2.2844],[161,-2.2844]]
       }
       arm_l_pv_ik_ctrl.Rotation.X {
        type=MovieSceneScriptingFloatChannel
        default=-0.7312
        thinned_from=12
        keys [[0,7.3186],[13,7.4357],[17,7.3186],[23,7.3099],[35,7.2282],[55,6.7137],[161,6.7137]]
       }
       arm_l_pv_ik_ctrl.Rotation.Y {
        type=MovieSceneScriptingFloatChannel
        default=-11.7468
        thinned_from=12
        keys [[0,-12.2972],[13,-11.9846],[17,-12.2972],[23,-12.3205],[35,-8.8266],[55,-13.8994],[161,-13.8994]]
       }
       arm_l_pv_ik_ctrl.Rotation.Z {
        type=MovieSceneScriptingFloatChannel
        default=47.7759
        thinned_from=12
        keys [[0,42.2389],[13,36.6038],[17,42.2389],[23,18.2414],[35,42.8974],[55,16.907],[161,16.907]]
       }
       arm_l_pv_ik_ctrl.Scale.X {type=MovieSceneScriptingFloatChannel, default=1, const [1,12]}
       arm_l_pv_ik_ctrl.Scale.Y {type=MovieSceneScriptingFloatChannel, default=1, const [1,12]}
       arm_l_pv_ik_ctrl.Scale.Z {type=MovieSceneScriptingFloatChannel, default=1, const [1,12]}
       upperarm_r_fk_ctrl.Location.X {type=MovieSceneScriptingFloatChannel, default=0, thinned_from=162, keys [[0,0],[161,0]]}
       upperarm_r_fk_ctrl.Location.Y {type=MovieSceneScriptingFloatChannel, default=0, thinned_from=162, keys [[0,0],[161,0]]}
       upperarm_r_fk_ctrl.Location.Z {type=MovieSceneScriptingFloatChannel, default=0, thinned_from=162, keys [[0,0],[161,0]]}
       upperarm_r_fk_ctrl.Rotation.X {
        type=MovieSceneScriptingFloatChannel
        default=24.195
        thinned_from=162
        keys [[0,15.6458],[9,15.6486],[10,15.6674],[11,15.737],[12,15.9221],[13,16.2978],[14,16.8423],[15,17.4735],[16,18.1747],[18,19.7367],[19,20.4846],[20,21.1224],[21,21.6135],[22,21.9948],[23,22.3172],[24,22.5937],[25,22.8091],[26,22.9249],[27,22.9257],[28,22.8059],[29,22.4879],[30,21.8527],[31,20.7993],[32,19.3709],[33,17.8252],[34,16.6119],[35,15.9535],[36,15.6692],[37,15.5581],[41,15.4091],[43,15.398],[45,15.46],[48,15.6511],[51,15.7233],[52,15.8156],[53,16.1338],[54,16.7156],[55,17.4153],[56,18.0471],[57,18.5385],[59,19.3495],[60,19.8496],[62,21.1084],[63,21.6106],[64,21.9946],[65,22.3172],[66,22.5937],[67,22.8091],[68,22.9249],[69,22.9257],[70,22.8059],[71,22.4879],[72,21.8527],[73,20.7993],[74,19.3709],[75,17.8252],[76,16.6119],[77,15.9535],[78,15.6692],[79,15.5581],[83,15.4091],[85,15.398],[87,15.46],[90,15.6511],[92,15.711],[96,15.6746],[97,15.7739],[98,16.1367],[99,16.8622],[100,17.8853],[102,20.2072],[103,21.1821],[104,21.8763],[105,22.3103],[106,22.5986],[107,22.8102],[108,22.925],[109,22.9257],[110,22.8059],[111,22.4879],[112,21.8527],[113,20.7993],[114,19.3709],[115,17.8252],[116,16.6119],[117,15.9535],[118,15.6692],[119,15.5581],[123,15.4091],[125,15.398],[127,15.46],[130,15.6511],[132,15.711],[136,15.6806],[137,15.8189],[138,16.3055],[139,17.2505],[140,18.5235],[141,19.8712],[142,21.0329],[143,21.8361],[144,22.3048],[145,22.5983],[146,22.8102],[147,22.925],[148,22.9257],[149,22.8059],[150,22.4879],[151,21.8527],[152,20.7991],[153,19.3699],[154,17.8234],[155,16.6118],[156,15.9617],[157,15.7009],[158,15.6283],[161,15.64]]
       }
       upperarm_r_fk_ctrl.Rotation.Y {
        type=MovieSceneScriptingFloatChannel
        default=-5.5277
        thinned_from=162
        keys [[0,-10.3949],[9,-10.3886],[10,-10.3479],[11,-10.1959],[12,-9.7864],[13,-8.9232],[14,-7.5834],[15,-5.8534],[16,-3.6406],[17,-0.8951],[18,2.0836],[19,4.7849],[20,6.6211],[21,7.4479],[22,7.6054],[23,7.2927],[24,6.678],[25,5.9126],[26,4.9364],[27,3.5568],[28,1.4617],[29,-1.423],[30,-4.3193],[31,-6.4582],[32,-7.9661],[33,-9.1194],[34,-9.8362],[35,-10.0985],[36,-10.0393],[37,-9.661],[38,-9.0181],[40,-7.5096],[41,-6.866],[42,-6.4831],[43,-6.4054],[44,-6.6071],[45,-7.1679],[46,-8.025],[47,-8.953],[48,-9.7607],[49,-10.3243],[50,-10.6022],[51,-10.6176],[52,-10.3163],[53,-9.4337],[54,-7.8754],[55,-5.9632],[56,-4.1196],[57,-2.5076],[58,-1.0104],[59,0.5787],[60,2.5097],[61,4.7575],[62,6.5703],[63,7.4373],[64,7.6048],[65,7.2926],[66,6.678],[67,5.9126],[68,4.9364],[69,3.5568],[70,1.4617],[71,-1.423],[72,-4.3192],[73,-6.4582],[74,-7.9661],[75,-9.1194],[76,-9.8362],[77,-10.0985],[78,-10.0393],[79,-9.661],[80,-9.0181],[82,-7.5096],[83,-6.866],[84,-6.4831],[85,-6.4054],[86,-6.6071],[87,-7.1679],[88,-8.025],[89,-8.953],[90,-9.7607],[91,-10.3242],[92,-10.6033],[93,-10.6542],[94,-10.5992],[95,-10.5151],[96,-10.3902],[97,-9.9759],[98,-8.7213],[99,-6.3151],[100,-3.0306],[101,0.5594],[102,3.8283],[103,6.1857],[104,7.2703],[105,7.2482],[106,6.6758],[107,5.9125],[108,4.9364],[109,3.5568],[110,1.4617],[111,-1.423],[112,-4.3192],[113,-6.4582],[114,-7.9661],[115,-9.1194],[116,-9.8362],[117,-10.0985],[118,-10.0393],[119,-9.661],[120,-9.0181],[122,-7.5096],[123,-6.866],[124,-6.4831],[125,-6.4054],[126,-6.6071],[127,-7.1679],[128,-8.025],[129,-8.953],[130,-9.7607],[131,-10.3242],[132,-10.6033],[133,-10.6542],[134,-10.5992],[135,-10.5142],[136,-10.3718],[137,-9.8408],[138,-8.2201],[139,-5.1859],[140,-1.2416],[141,2.6931],[142,5.6824],[143,7.1346],[144,7.2297],[145,6.6749],[146,5.9125],[147,4.9364],[148,3.5568],[149,1.4617],[150,-1.423],[151,-4.3192],[152,-6.4575],[153,-7.9626],[154,-9.1155],[155,-9.8508],[156,-10.163],[157,-10.2349],[159,-10.178],[161,-10.3394]]
       }
       upperarm_r_fk_ctrl.Rotation.Z {
        type=MovieSceneScriptingFloatChannel
        default=13.1497
        thinned_from=162
        keys [[0,-5.2261],[9,-5.2302],[10,-5.2575],[11,-5.3575],[12,-5.6054],[13,-6.0181],[14,-6.3996],[15,-6.5711],[16,-6.533],[17,-6.2754],[19,-5.4453],[20,-5.1451],[21,-4.9956],[22,-4.9417],[23,-4.9343],[24,-4.9455],[25,-4.968],[26,-5.0271],[27,-5.1835],[28,-5.4966],[29,-5.9318],[30,-6.2539],[31,-6.2919],[32,-6.1278],[33,-5.8638],[34,-5.55],[35,-5.2124],[36,-4.8412],[37,-4.4241],[39,-3.5335],[40,-3.1216],[41,-2.7891],[42,-2.6023],[43,-2.5812],[44,-2.724],[45,-3.0858],[46,-3.6285],[47,-4.2154],[48,-4.7268],[49,-5.084],[50,-5.2642],[51,-5.3194],[52,-5.3875],[53,-5.6064],[55,-6.3457],[56,-6.5425],[57,-6.5249],[58,-6.3607],[59,-6.115],[60,-5.8058],[61,-5.4488],[62,-5.1524],[63,-4.9971],[64,-4.9418],[65,-4.9343],[66,-4.9455],[67,-4.968],[68,-5.0271],[69,-5.1835],[70,-5.4966],[71,-5.9318],[72,-6.2539],[73,-6.2919],[74,-6.1278],[75,-5.8638],[76,-5.55],[77,-5.2124],[78,-4.8412],[79,-4.4241],[81,-3.5335],[82,-3.1216],[83,-2.7891],[84,-2.6023],[85,-2.5812],[86,-2.724],[87,-3.0858],[88,-3.6285],[89,-4.2154],[90,-4.7268],[91,-5.0843],[92,-5.265],[93,-5.309],[94,-5.2945],[98,-5.1888],[102,-4.9814],[103,-4.946],[104,-4.9305],[105,-4.9328],[107,-4.968],[108,-5.0271],[109,-5.1835],[110,-5.4966],[111,-5.9318],[112,-6.2539],[113,-6.2919],[114,-6.1278],[115,-5.8638],[116,-5.55],[117,-5.2124],[118,-4.8412],[119,-4.4241],[121,-3.5335],[122,-3.1216],[123,-2.7891],[124,-2.6023],[125,-2.5812],[126,-2.724],[127,-3.0858],[128,-3.6285],[129,-4.2154],[130,-4.7268],[131,-5.0843],[132,-5.265],[133,-5.309],[134,-5.2945],[137,-5.2179],[138,-5.1814],[139,-5.1269],[141,-5.0001],[142,-4.9542],[143,-4.9326],[144,-4.9331],[145,-4.9455],[146,-4.968],[147,-5.0271],[148,-5.1835],[149,-5.4966],[150,-5.9318],[151,-6.2539],[152,-6.2914],[153,-6.1254],[154,-5.8602],[155,-5.5584],[156,-5.2808],[157,-5.0884],[158,-5.0013],[159,-5.0056],[161,-5.1746]]
       }
       upperarm_r_fk_ctrl.Scale.X {type=MovieSceneScriptingFloatChannel, default=1, const [1.0005,162]}
       upperarm_r_fk_ctrl.Scale.Y {type=MovieSceneScriptingFloatChannel, default=1, const [1.0005,162]}
       upperarm_r_fk_ctrl.Scale.Z {type=MovieSceneScriptingFloatChannel, default=1, const [1.0005,162]}
       lowerarm_r_fk_ctrl.Location.X {type=MovieSceneScriptingFloatChannel, default=0, const [0,162]}
       lowerarm_r_fk_ctrl.Location.Y {type=MovieSceneScriptingFloatChannel, default=0, const [0,162]}
       lowerarm_r_fk_ctrl.Location.Z {type=MovieSceneScriptingFloatChannel, default=0, const [0,162]}
       lowerarm_r_fk_ctrl.Rotation.X {
        type=MovieSceneScriptingFloatChannel
        default=-3.8332
        thinned_from=162
        keys [[0,1.0501],[2,1.0544],[8,1.0887],[9,1.093],[10,1.0949],[11,1.0931],[12,1.0861],[14,1.0614],[15,1.0549],[16,1.0539],[17,1.0573],[19,1.0695],[20,1.0735],[21,1.0746],[22,1.0735],[26,1.0529],[56,1.0519],[59,1.0597],[62,1.0734],[63,1.0746],[64,1.0735],[68,1.0529],[97,1.0501],[99,1.0552],[102,1.0694],[103,1.0726],[104,1.0729],[105,1.0694],[107,1.0573],[108,1.0529],[137,1.0503],[139,1.0568],[142,1.0719],[143,1.0727],[144,1.0694],[146,1.0573],[147,1.0529],[161,1.0495]]
       }
       lowerarm_r_fk_ctrl.Rotation.Y {
        type=MovieSceneScriptingFloatChannel
        default=1.5865
        thinned_from=162
        keys [[0,0.3956],[1,0.3904],[2,0.3797],[3,0.3636],[4,0.3432],[5,0.3203],[6,0.2968],[7,0.2739],[8,0.2533],[9,0.2372],[10,0.2302],[11,0.2372],[12,0.2636],[13,0.3096],[14,0.3501],[15,0.3624],[16,0.3434],[17,0.2838],[19,0.1011],[20,0.0402],[21,0.0188],[22,0.0288],[23,0.0727],[24,0.1421],[25,0.2148],[26,0.2727],[27,0.3096],[28,0.3324],[29,0.3486],[30,0.3615],[31,0.3728],[32,0.3826],[33,0.3901],[34,0.3947],[35,0.3968],[51,0.3976],[52,0.3967],[53,0.3935],[54,0.387],[55,0.3772],[56,0.3624],[57,0.3371],[58,0.298],[59,0.2459],[60,0.1793],[61,0.1017],[62,0.0418],[63,0.0192],[64,0.0288],[65,0.0727],[66,0.1421],[67,0.2148],[68,0.2727],[69,0.3096],[70,0.3324],[71,0.3486],[72,0.3615],[73,0.3728],[74,0.3826],[75,0.3901],[76,0.3947],[77,0.3968],[95,0.3977],[96,0.3966],[97,0.3888],[98,0.3628],[99,0.3119],[100,0.2419],[101,0.1653],[102,0.0955],[103,0.0471],[104,0.0367],[105,0.0737],[106,0.1422],[107,0.2148],[108,0.2727],[109,0.3096],[110,0.3324],[111,0.3486],[112,0.3615],[113,0.3728],[114,0.3826],[115,0.3901],[116,0.3947],[117,0.3968],[135,0.3977],[136,0.3962],[137,0.3859],[138,0.352],[139,0.2877],[141,0.1197],[142,0.0578],[143,0.0395],[144,0.0741],[145,0.1422],[146,0.2148],[147,0.2727],[148,0.3096],[149,0.3324],[150,0.3486],[151,0.3615],[152,0.3728],[153,0.3826],[154,0.3901],[155,0.3947],[156,0.3968],[161,0.3978]]
       }
       lowerarm_r_fk_ctrl.Rotation.Z {
        type=MovieSceneScriptingFloatChannel
        default=-4.1227
        thinned_from=162
        keys [[0,59.9078],[3,59.2203],[4,59.1713],[5,59.3651],[6,59.8708],[7,60.7319],[9,62.8733],[11,64.7033],[13,66.157],[17,68.2614],[19,69.0288],[20,69.2169],[21,69.2202],[22,69.0257],[23,68.5657],[25,67.2372],[26,66.7954],[27,66.6424],[28,66.7251],[29,67.0785],[32,68.9897],[33,69.5195],[34,69.7943],[35,69.7472],[36,69.3675],[37,68.5755],[38,67.4266],[40,64.8307],[41,63.7412],[42,63.1063],[43,63.0084],[44,63.4139],[45,64.4515],[47,67.4903],[48,68.4848],[49,68.7908],[50,68.4933],[52,66.8035],[53,66.6324],[57,68.0401],[62,69.2101],[63,69.2188],[64,69.0256],[65,68.5657],[67,67.2372],[68,66.7954],[69,66.6424],[70,66.7251],[71,67.0785],[74,68.9897],[75,69.5195],[76,69.7943],[77,69.7472],[78,69.3675],[79,68.5755],[80,67.4266],[82,64.8307],[83,63.7412],[84,63.1063],[85,63.0084],[86,63.4139],[87,64.4515],[89,67.4903],[90,68.4857],[91,68.8071],[92,68.5727],[93,67.692],[95,64.7881],[96,63.4841],[97,62.7777],[98,62.9432],[99,63.8013],[102,67.6503],[103,68.5321],[104,68.8445],[105,68.5508],[107,67.2383],[108,66.7955],[109,66.6424],[110,66.7251],[111,67.0785],[114,68.9897],[115,69.5195],[116,69.7943],[117,69.7472],[118,69.3675],[119,68.5755],[120,67.4266],[122,64.8307],[123,63.7412],[124,63.1063],[125,63.0084],[126,63.4139],[127,64.4515],[129,67.4903],[130,68.4857],[131,68.8071],[132,68.5727],[133,67.692],[135,64.7875],[136,63.4722],[137,62.6873],[138,62.5875],[139,62.9162],[141,63.8955],[142,64.2479],[143,64.295],[144,63.9229],[146,62.5985],[147,62.1558],[148,62.0041],[149,62.0926],[150,62.4422],[152,63.4329],[153,63.7845],[154,63.9911],[155,64.0395],[156,63.8783],[157,63.425],[158,62.6104],[159,61.6037],[160,60.7289],[161,60.2317]]
       }
       lowerarm_r_fk_ctrl.Scale.X {type=MovieSceneScriptingFloatChannel, default=1, const [1.0005,162]}
       lowerarm_r_fk_ctrl.Scale.Y {type=MovieSceneScriptingFloatChannel, default=1, const [1.0005,162]}
       lowerarm_r_fk_ctrl.Scale.Z {type=MovieSceneScriptingFloatChannel, default=1, const [1.0005,162]}
       hand_r_fk_ctrl.Location.X {type=MovieSceneScriptingFloatChannel, default=0, const [0,162]}
       hand_r_fk_ctrl.Location.Y {type=MovieSceneScriptingFloatChannel, default=0, const [-0.4312,162]}
       hand_r_fk_ctrl.Location.Z {type=MovieSceneScriptingFloatChannel, default=0, const [0.0668,162]}
       hand_r_fk_ctrl.Rotation.X {
        type=MovieSceneScriptingFloatChannel
        default=-12.0844
        thinned_from=162
        keys [[0,-1.074],[24,-1.0742],[25,-1.0778],[26,-1.1028],[27,-1.1973],[28,-1.4554],[29,-1.9989],[30,-2.8231],[31,-3.8029],[32,-4.8181],[33,-5.7636],[34,-6.5428],[35,-7.1093],[36,-7.5302],[37,-7.8924],[40,-8.8986],[42,-9.5309],[44,-10.1145],[46,-10.6352],[48,-11.0794],[49,-11.2693],[50,-11.4337],[51,-11.4875],[52,-11.0056],[53,-9.2858],[55,-3.4783],[56,-1.7301],[57,-1.1639],[58,-1.0785],[59,-1.074],[66,-1.0742],[67,-1.0778],[68,-1.1028],[69,-1.1973],[70,-1.4554],[71,-1.9989],[72,-2.8231],[73,-3.8029],[74,-4.8181],[75,-5.7636],[76,-6.5428],[77,-7.1093],[78,-7.5302],[79,-7.8924],[82,-8.8986],[84,-9.5309],[86,-10.1145],[88,-10.6352],[90,-11.0793],[91,-11.2683],[92,-11.4328],[93,-11.5712],[94,-11.682],[95,-11.7617],[96,-11.7805],[97,-11.5874],[98,-10.8717],[99,-9.461],[100,-7.5218],[101,-5.396],[102,-3.4567],[103,-2.0451],[104,-1.3229],[105,-1.1069],[106,-1.0758],[107,-1.0778],[108,-1.1028],[109,-1.1973],[110,-1.4554],[111,-1.9989],[112,-2.8231],[113,-3.8029],[114,-4.8181],[115,-5.7636],[116,-6.5428],[117,-7.1093],[118,-7.5302],[119,-7.8924],[122,-8.8986],[124,-9.5309],[126,-10.1145],[128,-10.6352],[130,-11.0793],[131,-11.2683],[132,-11.4328],[133,-11.5712],[134,-11.682],[135,-11.7611],[136,-11.7695],[137,-11.5071],[138,-10.5739],[139,-8.7901],[141,-4.1276],[142,-2.3428],[143,-1.4032],[144,-1.1179],[145,-1.0763],[146,-1.0778],[147,-1.1028],[148,-1.1973],[149,-1.4554],[150,-1.9989],[151,-2.8235],[152,-3.8104],[153,-4.8533],[154,-5.7685],[155,-6.1338],[156,-5.6582],[157,-4.6727],[158,-3.6353],[159,-2.6667],[160,-1.8221],[161,-1.312]]
       }
       hand_r_fk_ctrl.Rotation.Y {
        type=MovieSceneScriptingFloatChannel
        default=4.2557
        thinned_from=162
        keys [[0,9.4713],[25,9.4791],[26,9.53],[27,9.7219],[28,10.2462],[29,11.3504],[30,13.0248],[31,15.0155],[32,17.0779],[33,18.999],[34,20.5819],[35,21.733],[36,22.5881],[39,24.6985],[42,26.6528],[44,27.8383],[46,28.8963],[48,29.7987],[50,30.5184],[51,30.6279],[52,29.6487],[53,26.1547],[55,14.356],[56,10.8043],[57,9.654],[58,9.4805],[67,9.4791],[68,9.53],[69,9.7219],[70,10.2462],[71,11.3504],[72,13.0248],[73,15.0155],[74,17.0779],[75,18.999],[76,20.5819],[77,21.733],[78,22.5881],[81,24.6985],[84,26.6528],[86,27.8383],[88,28.8963],[90,29.7986],[92,30.5167],[94,31.0231],[95,31.1849],[96,31.223],[97,30.8308],[98,29.3767],[99,26.5107],[100,22.5709],[101,18.2521],[102,14.3121],[103,11.4443],[104,9.977],[105,9.5382],[106,9.475],[107,9.4791],[108,9.53],[109,9.7219],[110,10.2462],[111,11.3504],[112,13.0248],[113,15.0155],[114,17.0779],[115,18.999],[116,20.5819],[117,21.733],[118,22.5881],[121,24.6985],[124,26.6528],[126,27.8383],[128,28.8963],[130,29.7986],[132,30.5167],[134,31.0231],[135,31.1838],[136,31.2008],[137,30.6677],[138,28.7718],[139,25.1477],[141,15.6752],[142,12.0492],[143,10.1402],[144,9.5605],[145,9.4761],[146,9.4791],[147,9.53],[148,9.7219],[149,10.2462],[150,11.3504],[151,13.0257],[152,15.0307],[153,17.1496],[154,19.0089],[155,19.7511],[156,18.7848],[157,16.7826],[158,14.6749],[159,12.7072],[160,10.9912],[161,9.9548]]
       }
       hand_r_fk_ctrl.Rotation.Z {
        type=MovieSceneScriptingFloatChannel
        default=-5.7448
        thinned_from=162
        keys [[0,-3.4393],[24,-3.4396],[25,-3.4451],[26,-3.4834],[27,-3.6277],[28,-4.0219],[29,-4.8522],[30,-6.1111],[31,-7.6079],[32,-9.1587],[33,-10.6031],[34,-11.7933],[35,-12.6588],[36,-13.3017],[37,-13.855],[40,-15.3921],[42,-16.358],[44,-17.2494],[46,-18.0449],[48,-18.7234],[49,-19.0135],[50,-19.2646],[51,-19.3469],[52,-18.6106],[53,-15.9835],[55,-7.112],[56,-4.4415],[57,-3.5766],[58,-3.4462],[66,-3.4396],[67,-3.4451],[68,-3.4834],[69,-3.6277],[70,-4.0219],[71,-4.8522],[72,-6.1111],[73,-7.6079],[74,-9.1587],[75,-10.6031],[76,-11.7933],[77,-12.6588],[78,-13.3017],[79,-13.855],[82,-15.3921],[84,-16.358],[86,-17.2494],[88,-18.0449],[90,-18.7233],[92,-19.2632],[93,-19.4748],[94,-19.644],[95,-19.7657],[96,-19.7944],[97,-19.4994],[98,-18.4061],[99,-16.2512],[100,-13.2888],[101,-10.0415],[102,-7.0791],[103,-4.9228],[104,-3.8195],[105,-3.4896],[106,-3.442],[107,-3.4451],[108,-3.4834],[109,-3.6277],[110,-4.0219],[111,-4.8522],[112,-6.1111],[113,-7.6079],[114,-9.1587],[115,-10.6031],[116,-11.7933],[117,-12.6588],[118,-13.3017],[119,-13.855],[122,-15.3921],[124,-16.358],[126,-17.2494],[128,-18.0449],[130,-18.7233],[132,-19.2632],[133,-19.4748],[134,-19.644],[135,-19.7649],[136,-19.7777],[137,-19.3768],[138,-17.9513],[139,-15.2263],[141,-8.1039],[142,-5.3776],[143,-3.9422],[144,-3.5063],[145,-3.4429],[146,-3.4452],[147,-3.4834],[148,-3.6277],[149,-4.0219],[150,-4.8522],[151,-6.1118],[152,-7.6193],[153,-9.2125],[154,-10.6106],[155,-11.1686],[156,-10.4421],[157,-8.9366],[158,-7.3519],[159,-5.8723],[160,-4.5821],[161,-3.8028]]
       }
       hand_r_fk_ctrl.Scale.X {type=MovieSceneScriptingFloatChannel, default=1, const [1.0005,162]}
       hand_r_fk_ctrl.Scale.Y {type=MovieSceneScriptingFloatChannel, default=1, const [1.0005,162]}
       hand_r_fk_ctrl.Scale.Z {type=MovieSceneScriptingFloatChannel, default=1, const [1.0005,162]}
       thumb_01_r_ctrl.Location.X {type=MovieSceneScriptingFloatChannel, default=0, const [0,1]}
       thumb_01_r_ctrl.Location.Y {type=MovieSceneScriptingFloatChannel, default=0, const [0,1]}
       thumb_01_r_ctrl.Location.Z {type=MovieSceneScriptingFloatChannel, default=0, const [0,1]}
       thumb_01_r_ctrl.Rotation.X {type=MovieSceneScriptingFloatChannel, default=0, const [26.151,1]}
       thumb_01_r_ctrl.Rotation.Y {type=MovieSceneScriptingFloatChannel, default=0, const [10.6122,1]}
       thumb_01_r_ctrl.Rotation.Z {type=MovieSceneScriptingFloatChannel, default=0, const [-8.5664,1]}
       thumb_01_r_ctrl.Scale.X {type=MovieSceneScriptingFloatChannel, default=1, const [1,1]}
       thumb_01_r_ctrl.Scale.Y {type=MovieSceneScriptingFloatChannel, default=1, const [1,1]}
       thumb_01_r_ctrl.Scale.Z {type=MovieSceneScriptingFloatChannel, default=1, const [1,1]}
       thumb_02_r_ctrl.Location.X {type=MovieSceneScriptingFloatChannel, default=0, const [0,1]}
       thumb_02_r_ctrl.Location.Y {type=MovieSceneScriptingFloatChannel, default=0, const [0,1]}
       thumb_02_r_ctrl.Location.Z {type=MovieSceneScriptingFloatChannel, default=0, const [0,1]}
       thumb_02_r_ctrl.Rotation.X {type=MovieSceneScriptingFloatChannel, default=0, const [0,1]}
       thumb_02_r_ctrl.Rotation.Y {type=MovieSceneScriptingFloatChannel, default=0, const [0,1]}
       thumb_02_r_ctrl.Rotation.Z {type=MovieSceneScriptingFloatChannel, default=0, const [0.0073,1]}
       thumb_02_r_ctrl.Scale.X {type=MovieSceneScriptingFloatChannel, default=1, const [1,1]}
       thumb_02_r_ctrl.Scale.Y {type=MovieSceneScriptingFloatChannel, default=1, const [1,1]}
       thumb_02_r_ctrl.Scale.Z {type=MovieSceneScriptingFloatChannel, default=1, const [1,1]}
       thumb_03_r_ctrl.Location.X {type=MovieSceneScriptingFloatChannel, default=0, const [0,1]}
       thumb_03_r_ctrl.Location.Y {type=MovieSceneScriptingFloatChannel, default=0, const [0,1]}
       thumb_03_r_ctrl.Location.Z {type=MovieSceneScriptingFloatChannel, default=0, const [0,1]}
       thumb_03_r_ctrl.Rotation.X {type=MovieSceneScriptingFloatChannel, default=0, const [0,1]}
       thumb_03_r_ctrl.Rotation.Y {type=MovieSceneScriptingFloatChannel, default=0, const [0,1]}
       thumb_03_r_ctrl.Rotation.Z {type=MovieSceneScriptingFloatChannel, default=0, const [-0.0004,1]}
       thumb_03_r_ctrl.Scale.X {type=MovieSceneScriptingFloatChannel, default=1, const [1,1]}
       thumb_03_r_ctrl.Scale.Y {type=MovieSceneScriptingFloatChannel, default=1, const [1,1]}
       thumb_03_r_ctrl.Scale.Z {type=MovieSceneScriptingFloatChannel, default=1, const [1,1]}
       index_metacarpal_r_ctrl.Location.X {type=MovieSceneScriptingFloatChannel, default=0, const [0,1]}
       index_metacarpal_r_ctrl.Location.Y {type=MovieSceneScriptingFloatChannel, default=0, const [0,1]}
       index_metacarpal_r_ctrl.Location.Z {type=MovieSceneScriptingFloatChannel, default=0, const [0,1]}
       index_metacarpal_r_ctrl.Rotation.X {type=MovieSceneScriptingFloatChannel, default=0, const [6.1766,1]}
       index_metacarpal_r_ctrl.Rotation.Y {type=MovieSceneScriptingFloatChannel, default=0, const [3.1321,1]}
       index_metacarpal_r_ctrl.Rotation.Z {type=MovieSceneScriptingFloatChannel, default=0, const [0.0216,1]}
       index_metacarpal_r_ctrl.Scale.X {type=MovieSceneScriptingFloatChannel, default=1, const [1,1]}
       index_metacarpal_r_ctrl.Scale.Y {type=MovieSceneScriptingFloatChannel, default=1, const [1,1]}
       index_metacarpal_r_ctrl.Scale.Z {type=MovieSceneScriptingFloatChannel, default=1, const [1,1]}
       index_01_r_ctrl.Location.X {type=MovieSceneScriptingFloatChannel, default=0, const [0,1]}
       index_01_r_ctrl.Location.Y {type=MovieSceneScriptingFloatChannel, default=0, const [0,1]}
       index_01_r_ctrl.Location.Z {type=MovieSceneScriptingFloatChannel, default=0, const [0,1]}
       index_01_r_ctrl.Rotation.X {type=MovieSceneScriptingFloatChannel, default=0, const [0,1]}
       index_01_r_ctrl.Rotation.Y {type=MovieSceneScriptingFloatChannel, default=0, const [0,1]}
       index_01_r_ctrl.Rotation.Z {type=MovieSceneScriptingFloatChannel, default=0, const [-59.5276,1]}
       index_01_r_ctrl.Scale.X {type=MovieSceneScriptingFloatChannel, default=1, const [1,1]}
       index_01_r_ctrl.Scale.Y {type=MovieSceneScriptingFloatChannel, default=1, const [1,1]}
       index_01_r_ctrl.Scale.Z {type=MovieSceneScriptingFloatChannel, default=1, const [1,1]}
       index_02_r_ctrl.Location.X {type=MovieSceneScriptingFloatChannel, default=0, const [0,1]}
       index_02_r_ctrl.Location.Y {type=MovieSceneScriptingFloatChannel, default=0, const [0,1]}
       index_02_r_ctrl.Location.Z {type=MovieSceneScriptingFloatChannel, default=0, const [0,1]}
       index_02_r_ctrl.Rotation.X {type=MovieSceneScriptingFloatChannel, default=0, const [0,1]}
       index_02_r_ctrl.Rotation.Y {type=MovieSceneScriptingFloatChannel, default=0, const [0,1]}
       index_02_r_ctrl.Rotation.Z {type=MovieSceneScriptingFloatChannel, default=0, const [-53.0237,1]}
       index_02_r_ctrl.Scale.X {type=MovieSceneScriptingFloatChannel, default=1, const [1,1]}
       index_02_r_ctrl.Scale.Y {type=MovieSceneScriptingFloatChannel, default=1, const [1,1]}
       index_02_r_ctrl.Scale.Z {type=MovieSceneScriptingFloatChannel, default=1, const [1,1]}
       index_03_r_ctrl.Location.X {type=MovieSceneScriptingFloatChannel, default=0, const [0,1]}
       index_03_r_ctrl.Location.Y {type=MovieSceneScriptingFloatChannel, default=0, const [0,1]}
       index_03_r_ctrl.Location.Z {type=MovieSceneScriptingFloatChannel, default=0, const [0,1]}
       index_03_r_ctrl.Rotation.X {type=MovieSceneScriptingFloatChannel, default=0, const [0,1]}
       index_03_r_ctrl.Rotation.Y {type=MovieSceneScriptingFloatChannel, default=0, const [0,1]}
       index_03_r_ctrl.Rotation.Z {type=MovieSceneScriptingFloatChannel, default=0, const [-56.0042,1]}
       index_03_r_ctrl.Scale.X {type=MovieSceneScriptingFloatChannel, default=1, const [1,1]}
       index_03_r_ctrl.Scale.Y {type=MovieSceneScriptingFloatChannel, default=1, const [1,1]}
       index_03_r_ctrl.Scale.Z {type=MovieSceneScriptingFloatChannel, default=1, const [1,1]}
       middle_metacarpal_r_ctrl.Location.X {type=MovieSceneScriptingFloatChannel, default=0, const [0,1]}
       middle_metacarpal_r_ctrl.Location.Y {type=MovieSceneScriptingFloatChannel, default=0, const [0,1]}
       middle_metacarpal_r_ctrl.Location.Z {type=MovieSceneScriptingFloatChannel, default=0, const [0,1]}
       middle_metacarpal_r_ctrl.Rotation.X {type=MovieSceneScriptingFloatChannel, default=0, const [0,1]}
       middle_metacarpal_r_ctrl.Rotation.Y {type=MovieSceneScriptingFloatChannel, default=0, const [0,1]}
       middle_metacarpal_r_ctrl.Rotation.Z {type=MovieSceneScriptingFloatChannel, default=0, const [-4.0425,1]}
       middle_metacarpal_r_ctrl.Scale.X {type=MovieSceneScriptingFloatChannel, default=1, const [1,1]}
       middle_metacarpal_r_ctrl.Scale.Y {type=MovieSceneScriptingFloatChannel, default=1, const [1,1]}
       middle_metacarpal_r_ctrl.Scale.Z {type=MovieSceneScriptingFloatChannel, default=1, const [1,1]}
       middle_01_r_ctrl.Location.X {type=MovieSceneScriptingFloatChannel, default=0, const [0,1]}
       middle_01_r_ctrl.Location.Y {type=MovieSceneScriptingFloatChannel, default=0, const [0,1]}
       middle_01_r_ctrl.Location.Z {type=MovieSceneScriptingFloatChannel, default=0, const [0,1]}
       middle_01_r_ctrl.Rotation.X {type=MovieSceneScriptingFloatChannel, default=0, const [0,1]}
       middle_01_r_ctrl.Rotation.Y {type=MovieSceneScriptingFloatChannel, default=0, const [0,1]}
       middle_01_r_ctrl.Rotation.Z {type=MovieSceneScriptingFloatChannel, default=0, const [-52.9974,1]}
       middle_01_r_ctrl.Scale.X {type=MovieSceneScriptingFloatChannel, default=1, const [1,1]}
       middle_01_r_ctrl.Scale.Y {type=MovieSceneScriptingFloatChannel, default=1, const [1,1]}
       middle_01_r_ctrl.Scale.Z {type=MovieSceneScriptingFloatChannel, default=1, const [1,1]}
       middle_02_r_ctrl.Location.X {type=MovieSceneScriptingFloatChannel, default=0, const [0,1]}
       middle_02_r_ctrl.Location.Y {type=MovieSceneScriptingFloatChannel, default=0, const [0,1]}
       middle_02_r_ctrl.Location.Z {type=MovieSceneScriptingFloatChannel, default=0, const [0,1]}
       middle_02_r_ctrl.Rotation.X {type=MovieSceneScriptingFloatChannel, default=0, const [0,1]}
       middle_02_r_ctrl.Rotation.Y {type=MovieSceneScriptingFloatChannel, default=0, const [0,1]}
       middle_02_r_ctrl.Rotation.Z {type=MovieSceneScriptingFloatChannel, default=0, const [-65.0474,1]}
       middle_02_r_ctrl.Scale.X {type=MovieSceneScriptingFloatChannel, default=1, const [1,1]}
       middle_02_r_ctrl.Scale.Y {type=MovieSceneScriptingFloatChannel, default=1, const [1,1]}
       middle_02_r_ctrl.Scale.Z {type=MovieSceneScriptingFloatChannel, default=1, const [1,1]}
       middle_03_r_ctrl.Location.X {type=MovieSceneScriptingFloatChannel, default=0, const [0,1]}
       middle_03_r_ctrl.Location.Y {type=MovieSceneScriptingFloatChannel, default=0, const [0,1]}
       middle_03_r_ctrl.Location.Z {type=MovieSceneScriptingFloatChannel, default=0, const [0,1]}
       middle_03_r_ctrl.Rotation.X {type=MovieSceneScriptingFloatChannel, default=0, const [0,1]}
       middle_03_r_ctrl.Rotation.Y {type=MovieSceneScriptingFloatChannel, default=0, const [0,1]}
       middle_03_r_ctrl.Rotation.Z {type=MovieSceneScriptingFloatChannel, default=0, const [-69.6807,1]}
       middle_03_r_ctrl.Scale.X {type=MovieSceneScriptingFloatChannel, default=1, const [1,1]}
       middle_03_r_ctrl.Scale.Y {type=MovieSceneScriptingFloatChannel, default=1, const [1,1]}
       middle_03_r_ctrl.Scale.Z {type=MovieSceneScriptingFloatChannel, default=1, const [1,1]}
       ring_metacarpal_r_ctrl.Location.X {type=MovieSceneScriptingFloatChannel, default=0, const [0,1]}
       ring_metacarpal_r_ctrl.Location.Y {type=MovieSceneScriptingFloatChannel, default=0, const [0,1]}
       ring_metacarpal_r_ctrl.Location.Z {type=MovieSceneScriptingFloatChannel, default=0, const [0,1]}
       ring_metacarpal_r_ctrl.Rotation.X {type=MovieSceneScriptingFloatChannel, default=0, const [0,1]}
       ring_metacarpal_r_ctrl.Rotation.Y {type=MovieSceneScriptingFloatChannel, default=0, const [0,1]}
       ring_metacarpal_r_ctrl.Rotation.Z {type=MovieSceneScriptingFloatChannel, default=0, const [-4.0425,1]}
       ring_metacarpal_r_ctrl.Scale.X {type=MovieSceneScriptingFloatChannel, default=1, const [1,1]}
       ring_metacarpal_r_ctrl.Scale.Y {type=MovieSceneScriptingFloatChannel, default=1, const [1,1]}
       ring_metacarpal_r_ctrl.Scale.Z {type=MovieSceneScriptingFloatChannel, default=1, const [1,1]}
       ring_01_r_ctrl.Location.X {type=MovieSceneScriptingFloatChannel, default=0, const [0,1]}
       ring_01_r_ctrl.Location.Y {type=MovieSceneScriptingFloatChannel, default=0, const [0,1]}
       ring_01_r_ctrl.Location.Z {type=MovieSceneScriptingFloatChannel, default=0, const [0,1]}
       ring_01_r_ctrl.Rotation.X {type=MovieSceneScriptingFloatChannel, default=0, const [0,1]}
       ring_01_r_ctrl.Rotation.Y {type=MovieSceneScriptingFloatChannel, default=0, const [0,1]}
       ring_01_r_ctrl.Rotation.Z {type=MovieSceneScriptingFloatChannel, default=0, const [-47.3267,1]}
       ring_01_r_ctrl.Scale.X {type=MovieSceneScriptingFloatChannel, default=1, const [1,1]}
       ring_01_r_ctrl.Scale.Y {type=MovieSceneScriptingFloatChannel, default=1, const [1,1]}
       ring_01_r_ctrl.Scale.Z {type=MovieSceneScriptingFloatChannel, default=1, const [1,1]}
       ring_02_r_ctrl.Location.X {type=MovieSceneScriptingFloatChannel, default=0, const [0,1]}
       ring_02_r_ctrl.Location.Y {type=MovieSceneScriptingFloatChannel, default=0, const [0,1]}
       ring_02_r_ctrl.Location.Z {type=MovieSceneScriptingFloatChannel, default=0, const [0,1]}
       ring_02_r_ctrl.Rotation.X {type=MovieSceneScriptingFloatChannel, default=0, const [0,1]}
       ring_02_r_ctrl.Rotation.Y {type=MovieSceneScriptingFloatChannel, default=0, const [0,1]}
       ring_02_r_ctrl.Rotation.Z {type=MovieSceneScriptingFloatChannel, default=0, const [-63.5055,1]}
       ring_02_r_ctrl.Scale.X {type=MovieSceneScriptingFloatChannel, default=1, const [1,1]}
       ring_02_r_ctrl.Scale.Y {type=MovieSceneScriptingFloatChannel, default=1, const [1,1]}
       ring_02_r_ctrl.Scale.Z {type=MovieSceneScriptingFloatChannel, default=1, const [1,1]}
       ring_03_r_ctrl.Location.X {type=MovieSceneScriptingFloatChannel, default=0, const [0,1]}
       ring_03_r_ctrl.Location.Y {type=MovieSceneScriptingFloatChannel, default=0, const [0,1]}
       ring_03_r_ctrl.Location.Z {type=MovieSceneScriptingFloatChannel, default=0, const [0,1]}
       ring_03_r_ctrl.Rotation.X {type=MovieSceneScriptingFloatChannel, default=0, const [0,1]}
       ring_03_r_ctrl.Rotation.Y {type=MovieSceneScriptingFloatChannel, default=0, const [0,1]}
       ring_03_r_ctrl.Rotation.Z {type=MovieSceneScriptingFloatChannel, default=0, const [-50.5238,1]}
       ring_03_r_ctrl.Scale.X {type=MovieSceneScriptingFloatChannel, default=1, const [1,1]}
       ring_03_r_ctrl.Scale.Y {type=MovieSceneScriptingFloatChannel, default=1, const [1,1]}
       ring_03_r_ctrl.Scale.Z {type=MovieSceneScriptingFloatChannel, default=1, const [1,1]}
       pinky_metacarpal_r_ctrl.Location.X {type=MovieSceneScriptingFloatChannel, default=0, const [0,1]}
       pinky_metacarpal_r_ctrl.Location.Y {type=MovieSceneScriptingFloatChannel, default=0, const [0,1]}
       pinky_metacarpal_r_ctrl.Location.Z {type=MovieSceneScriptingFloatChannel, default=0, const [0,1]}
       pinky_metacarpal_r_ctrl.Rotation.X {type=MovieSceneScriptingFloatChannel, default=0, const [0,1]}
       pinky_metacarpal_r_ctrl.Rotation.Y {type=MovieSceneScriptingFloatChannel, default=0, const [0,1]}
       pinky_metacarpal_r_ctrl.Rotation.Z {type=MovieSceneScriptingFloatChannel, default=0, const [-4.0425,1]}
       pinky_metacarpal_r_ctrl.Scale.X {type=MovieSceneScriptingFloatChannel, default=1, const [1,1]}
       pinky_metacarpal_r_ctrl.Scale.Y {type=MovieSceneScriptingFloatChannel, default=1, const [1,1]}
       pinky_metacarpal_r_ctrl.Scale.Z {type=MovieSceneScriptingFloatChannel, default=1, const [1,1]}
       pinky_01_r_ctrl.Location.X {type=MovieSceneScriptingFloatChannel, default=0, const [0,1]}
       pinky_01_r_ctrl.Location.Y {type=MovieSceneScriptingFloatChannel, default=0, const [0,1]}
       pinky_01_r_ctrl.Location.Z {type=MovieSceneScriptingFloatChannel, default=0, const [0,1]}
       pinky_01_r_ctrl.Rotation.X {type=MovieSceneScriptingFloatChannel, default=0, const [0.0001,1]}
       pinky_01_r_ctrl.Rotation.Y {type=MovieSceneScriptingFloatChannel, default=0, const [0,1]}
       pinky_01_r_ctrl.Rotation.Z {type=MovieSceneScriptingFloatChannel, default=0, const [-48.5057,1]}
       pinky_01_r_ctrl.Scale.X {type=MovieSceneScriptingFloatChannel, default=1, const [1,1]}
       pinky_01_r_ctrl.Scale.Y {type=MovieSceneScriptingFloatChannel, default=1, const [1,1]}
       pinky_01_r_ctrl.Scale.Z {type=MovieSceneScriptingFloatChannel, default=1, const [1,1]}
       pinky_02_r_ctrl.Location.X {type=MovieSceneScriptingFloatChannel, default=0, const [0,1]}
       pinky_02_r_ctrl.Location.Y {type=MovieSceneScriptingFloatChannel, default=0, const [0,1]}
       pinky_02_r_ctrl.Location.Z {type=MovieSceneScriptingFloatChannel, default=0, const [0,1]}
       pinky_02_r_ctrl.Rotation.X {type=MovieSceneScriptingFloatChannel, default=0, const [0,1]}
       pinky_02_r_ctrl.Rotation.Y {type=MovieSceneScriptingFloatChannel, default=0, const [0,1]}
       pinky_02_r_ctrl.Rotation.Z {type=MovieSceneScriptingFloatChannel, default=0, const [-57.7751,1]}
       pinky_02_r_ctrl.Scale.X {type=MovieSceneScriptingFloatChannel, default=1, const [1,1]}
       pinky_02_r_ctrl.Scale.Y {type=MovieSceneScriptingFloatChannel, default=1, const [1,1]}
       pinky_02_r_ctrl.Scale.Z {type=MovieSceneScriptingFloatChannel, default=1, const [1,1]}
       pinky_03_r_ctrl.Location.X {type=MovieSceneScriptingFloatChannel, default=0, const [0,1]}
       pinky_03_r_ctrl.Location.Y {type=MovieSceneScriptingFloatChannel, default=0, const [0,1]}
       pinky_03_r_ctrl.Location.Z {type=MovieSceneScriptingFloatChannel, default=0, const [0,1]}
       pinky_03_r_ctrl.Rotation.X {type=MovieSceneScriptingFloatChannel, default=0, const [0,1]}
       pinky_03_r_ctrl.Rotation.Y {type=MovieSceneScriptingFloatChannel, default=0, const [0,1]}
       pinky_03_r_ctrl.Rotation.Z {type=MovieSceneScriptingFloatChannel, default=0, const [-61.3898,1]}
       pinky_03_r_ctrl.Scale.X {type=MovieSceneScriptingFloatChannel, default=1, const [1,1]}
       pinky_03_r_ctrl.Scale.Y {type=MovieSceneScriptingFloatChannel, default=1, const [1,1]}
       pinky_03_r_ctrl.Scale.Z {type=MovieSceneScriptingFloatChannel, default=1, const [1,1]}
       hand_r_ik_ctrl.Location.X {
        type=MovieSceneScriptingFloatChannel
        default=-11.1389
        thinned_from=13
        keys [[0,-12.7316],[11,-12.2733],[15,-12.7316],[21,-5.7134],[27,-6.5704],[35,-6.8211],[55,-5.8125],[161,-5.8125]]
       }
       hand_r_ik_ctrl.Location.Y {
        type=MovieSceneScriptingFloatChannel
        default=19.728
        thinned_from=13
        keys [[0,14.0948],[11,13.8651],[15,14.0948],[21,10.8541],[27,5.4324],[35,16.4562],[55,5.4966],[161,5.4966]]
       }
       hand_r_ik_ctrl.Location.Z {
        type=MovieSceneScriptingFloatChannel
        default=-7.4919
        thinned_from=13
        keys [[0,10.9256],[11,13.5406],[15,10.9256],[21,16.6665],[27,19.6099],[35,-10.1234],[55,16.9637],[161,16.9637]]
       }
       hand_r_ik_ctrl.Rotation.X {
        type=MovieSceneScriptingFloatChannel
        default=-16.9227
        thinned_from=13
        keys [[0,-70.4677],[11,-69.7451],[15,-70.4677],[21,-49.4322],[27,92.359],[35,-35.5474],[55,-40.0387],[161,-40.0387]]
       }
       hand_r_ik_ctrl.Rotation.Y {
        type=MovieSceneScriptingFloatChannel
        default=4.2254
        thinned_from=13
        keys [[0,59.4785],[11,64.4711],[15,59.4785],[21,80.4861],[27,80.622],[35,4.6132],[55,81.9355],[161,81.9355]]
       }
       hand_r_ik_ctrl.Rotation.Z {
        type=MovieSceneScriptingFloatChannel
        default=-5.7911
        thinned_from=13
        keys [[0,-63.6685],[11,-60.4705],[15,-63.6685],[21,-33.6357],[27,105.2854],[35,-14.9176],[55,-25.2719],[161,-25.2719]]
       }
       hand_r_ik_ctrl.Scale.X {type=MovieSceneScriptingFloatChannel, default=1, const [1,13]}
       hand_r_ik_ctrl.Scale.Y {type=MovieSceneScriptingFloatChannel, default=1, const [1,13]}
       hand_r_ik_ctrl.Scale.Z {type=MovieSceneScriptingFloatChannel, default=1, const [1,13]}
       Softness_2 {type=MovieSceneScriptingFloatChannel, default=0, const [0,13]}
       hand_r_ikorient_ctrl.Rotation.Roll {type=MovieSceneScriptingFloatChannel, default=0, const [0,10]}
       hand_r_ikorient_ctrl.Rotation.Pitch {type=MovieSceneScriptingFloatChannel, default=0, const [0,10]}
       hand_r_ikorient_ctrl.Rotation.Yaw {type=MovieSceneScriptingFloatChannel, default=0, thinned_from=10, keys [[0,0],[161,0]]}
       arm_r_pv_ik_ctrl.Location.X {
        type=MovieSceneScriptingFloatChannel
        default=-20.3856
        thinned_from=13
        keys [[0,-21.8245],[11,-17.6078],[15,-21.8245],[21,0.6427],[27,15.436],[35,-26.2841],[55,6.6981],[161,6.6981]]
       }
       arm_r_pv_ik_ctrl.Location.Y {
        type=MovieSceneScriptingFloatChannel
        default=-11.2154
        thinned_from=13
        keys [[0,-8.6042],[11,-6.37],[15,-8.6042],[21,1.2611],[27,7.8023],[35,-8.0019],[55,1.9083],[161,1.9083]]
       }
       arm_r_pv_ik_ctrl.Location.Z {
        type=MovieSceneScriptingFloatChannel
        default=-2.2167
        thinned_from=13
        keys [[0,-0.1135],[11,0.5809],[15,-0.1135],[21,-2.5783],[27,1.3234],[35,-0.5891],[55,-0.1935],[161,-0.1935]]
       }
       arm_r_pv_ik_ctrl.Rotation.X {
        type=MovieSceneScriptingFloatChannel
        default=-4.3577
        thinned_from=13
        keys [[0,7.967],[11,7.8996],[15,7.967],[21,7.972],[27,9.3561],[35,6.1694],[55,8.3099],[161,8.3099]]
       }
       arm_r_pv_ik_ctrl.Rotation.Y {
        type=MovieSceneScriptingFloatChannel
        default=-2.123
        thinned_from=13
        keys [[0,-11.5314],[11,-11.8577],[15,-11.5314],[21,-11.5071],[27,-1.4286],[35,-8.9775],[55,-9.858],[161,-9.858]]
       }
       arm_r_pv_ik_ctrl.Rotation.Z {
        type=MovieSceneScriptingFloatChannel
        default=1.5482
        thinned_from=13
        keys [[0,-10.6989],[11,-5.0748],[15,-10.6989],[21,13.2994],[27,26.9957],[35,-23.146],[55,14.7064],[161,14.7064]]
       }
       arm_r_pv_ik_ctrl.Scale.X {type=MovieSceneScriptingFloatChannel, default=1, const [1,13]}
       arm_r_pv_ik_ctrl.Scale.Y {type=MovieSceneScriptingFloatChannel, default=1, const [1,13]}
       arm_r_pv_ik_ctrl.Scale.Z {type=MovieSceneScriptingFloatChannel, default=1, const [1,13]}
       body_ctrl.Location.X {
        type=MovieSceneScriptingFloatChannel
        default=0
        thinned_from=84
        keys [[0,8.4887],[2,8.4367],[4,8.4099],[6,8.4187],[8,8.3999],[10,8.1526],[12,7.8672],[14,7.3312],[16,6.2128],[18,4.3609],[20,1.9519],[22,-0.6286],[24,-2.6522],[26,-4.5092],[28,-5.6126],[30,-6.3855],[32,-6.8548],[34,-7.1969],[36,-7.3611],[38,-7.6478],[40,-7.6934],[42,-7.8026],[44,-7.8452],[46,-7.8485],[48,-7.922],[51,-7.8022],[53,-7.5609],[55,-7.1686],[57,-6.5776],[59,-5.3513],[60,-4.5633],[61,-3.5631],[63,-0.6395],[65,2.4386],[67,4.8567],[69,6.3042],[71,7.0987],[73,7.4636],[77,7.9303],[79,8.1783],[81,8.3587],[83,8.4837],[85,8.5609],[87,8.5422],[89,8.4075],[91,8.1785],[95,7.6524],[97,7.3063],[99,6.6068],[101,5.679],[103,3.6989],[105,0.7764],[106,-0.763],[107,-2.1386],[109,-3.7498],[111,-4.8157],[113,-5.4912],[115,-5.8557],[117,-6.0067],[119,-6.2048],[123,-6.4531],[125,-6.5589],[127,-6.5817],[129,-6.4695],[131,-6.4078],[135,-6.1992],[136,-6.0755],[137,-5.876],[139,-5.2703],[141,-3.8902],[143,-1.6048],[145,1.4713],[147,4.4822],[149,6.7549],[151,7.9121],[153,8.4912],[155,8.5624],[157,8.5147],[159,8.4467],[161,8.4534]]
       }
       body_ctrl.Location.Y {
        type=MovieSceneScriptingFloatChannel
        default=0
        keys [[0,-0.695],[12,-0.695],[26,-1.506],[42,-0.695],[60,-1.3273],[71,-0.695],[97,-0.695],[106,-1.3069],[127,-0.695],[136,-0.695],[145,-1.4701],[151,-1.1845],[161,-0.695]]
       }
       body_ctrl.Location.Z {
        type=MovieSceneScriptingFloatChannel
        default=0
        thinned_from=162
        keys [[0,-1.6888],[1,-1.7422],[2,-1.7883],[3,-1.7496],[4,-1.6486],[5,-1.433],[6,-1.141],[7,-0.8588],[8,-0.6137],[9,-0.4288],[10,-0.3419],[11,-0.3635],[12,-0.4607],[13,-0.6639],[14,-1.042],[15,-1.6535],[16,-2.3439],[17,-2.91],[18,-3.3435],[19,-3.6532],[20,-3.8351],[21,-3.922],[22,-3.9751],[23,-4.0186],[24,-4.0363],[25,-4.0067],[26,-3.9169],[27,-3.7081],[28,-3.3218],[29,-2.7486],[30,-2.1542],[31,-1.722],[32,-1.4518],[33,-1.2951],[34,-1.2288],[35,-1.2452],[36,-1.3401],[37,-1.5972],[38,-2.0476],[39,-2.624],[40,-3.087],[41,-3.3015],[42,-3.3698],[43,-3.3257],[44,-3.1372],[45,-2.6764],[46,-2.1493],[47,-1.8338],[48,-1.6765],[49,-1.5996],[50,-1.5549],[51,-1.5517],[52,-1.6379],[53,-1.8283],[54,-2.1407],[55,-2.6448],[56,-3.3575],[57,-4.015],[58,-4.4111],[59,-4.6497],[60,-4.8111],[61,-4.8993],[62,-4.9311],[63,-4.9239],[64,-4.8361],[65,-4.6552],[66,-4.4022],[67,-4.0791],[68,-3.6909],[69,-3.2578],[70,-2.8067],[71,-2.424],[72,-2.1746],[73,-2.0192],[74,-1.9192],[75,-1.8857],[76,-1.9536],[77,-2.1239],[78,-2.4081],[79,-2.7877],[80,-3.2053],[81,-3.5396],[82,-3.7215],[83,-3.7702],[84,-3.6406],[85,-3.3498],[86,-2.9492],[87,-2.4971],[88,-2.1231],[89,-1.8697],[90,-1.6917],[91,-1.5814],[92,-1.5233],[93,-1.5056],[94,-1.5414],[95,-1.6375],[96,-1.8301],[97,-2.1477],[98,-2.5833],[99,-3.1696],[100,-3.9199],[101,-4.5985],[102,-5.0201],[103,-5.1959],[104,-5.1531],[105,-4.9826],[106,-4.6977],[107,-4.246],[108,-3.6092],[109,-2.8394],[110,-2.0076],[111,-1.3193],[112,-0.908],[113,-0.6552],[114,-0.4809],[115,-0.4022],[116,-0.4277],[117,-0.5155],[118,-0.6533],[119,-0.8892],[120,-1.2952],[121,-1.8011],[122,-2.281],[123,-2.6173],[124,-2.7743],[125,-2.8209],[126,-2.7612],[127,-2.6033],[128,-2.3302],[129,-1.9404],[130,-1.4721],[131,-1.0175],[132,-0.6712],[133,-0.4966],[134,-0.4769],[135,-0.5261],[136,-0.664],[137,-0.9153],[138,-1.2794],[139,-1.847],[141,-3.5367],[142,-4.1],[143,-4.4603],[144,-4.7079],[145,-4.811],[146,-4.732],[147,-4.5196],[148,-4.1907],[149,-3.6715],[150,-2.9016],[151,-2.0018],[152,-1.1373],[153,-0.5392],[154,-0.2812],[155,-0.2189],[156,-0.3531],[157,-0.6492],[158,-1.0341],[159,-1.3749],[160,-1.5626],[161,-1.6888]]
       }
       body_ctrl.Rotation.X {type=MovieSceneScriptingFloatChannel, default=0, const [0.0472,94]}
       body_ctrl.Rotation.Y {
        type=MovieSceneScriptingFloatChannel
        default=0
        thinned_from=51
        keys [[0,-0.918],[4,-0.8281],[8,-0.6957],[12,-0.594],[16,-0.5872],[20,-0.7271],[24,-1.0391],[26,-1.2457],[28,-1.4714],[32,-1.9103],[36,-2.275],[38,-2.3071],[44,-2.386],[48,-2.4514],[53,-2.5475],[57,-2.7325],[60,-2.8214],[61,-2.8325],[65,-2.8424],[69,-2.7523],[71,-2.6673],[73,-2.5614],[77,-2.2773],[81,-1.9104],[85,-1.4757],[89,-0.9915],[93,-0.4765],[97,0.0256],[101,0.4362],[105,0.7206],[106,0.7735],[109,0.9072],[113,1.0231],[117,1.0769],[121,1.0665],[125,0.9928],[127,0.9325],[129,0.8616],[133,0.6793],[136,0.5117],[137,0.4514],[141,0.1855],[145,-0.1051],[149,-0.3952],[151,-0.5311],[153,-0.648],[157,-0.83],[161,-0.9244]]
       }
       body_ctrl.Rotation.Z {
        type=MovieSceneScriptingFloatChannel
        default=0
        thinned_from=51
        keys [[0,22.174],[4,22.052],[8,21.1712],[12,17.8307],[16,10.8177],[20,2.5883],[24,-2.816],[26,-4.0944],[28,-4.8504],[32,-5.2779],[36,-5.1916],[40,-5.0413],[44,-4.8571],[48,-4.6425],[53,-4.3399],[57,-1.3277],[60,3.1502],[61,4.9349],[65,12.7516],[69,18.6771],[71,20.3275],[73,21.3388],[77,22.0232],[81,22.1191],[89,22.1148],[93,21.9584],[97,20.9972],[101,17.6718],[105,10.9891],[106,8.9686],[109,3.1154],[113,-2.392],[117,-4.7074],[121,-5.3181],[125,-5.4497],[129,-5.4548],[133,-5.1384],[136,-4.0587],[137,-3.52],[141,1.1673],[145,9.0651],[149,16.6009],[151,19.0654],[153,20.649],[157,21.8937],[161,22.1173]]
       }
       body_ctrl.Scale.X {type=MovieSceneScriptingFloatChannel, default=1, const [1,94]}
       body_ctrl.Scale.Y {type=MovieSceneScriptingFloatChannel, default=1, const [1,94]}
       body_ctrl.Scale.Z {type=MovieSceneScriptingFloatChannel, default=1, const [1,94]}
       hips_ctrl.Location.X {type=MovieSceneScriptingFloatChannel, default=0, const [0,10]}
       hips_ctrl.Location.Y {type=MovieSceneScriptingFloatChannel, default=0, const [0,10]}
       hips_ctrl.Location.Z {type=MovieSceneScriptingFloatChannel, default=0, const [0,10]}
       hips_ctrl.Rotation.X {type=MovieSceneScriptingFloatChannel, default=0, const [0,10]}
       hips_ctrl.Rotation.Y {type=MovieSceneScriptingFloatChannel, default=0, const [0,10]}
       hips_ctrl.Rotation.Z {type=MovieSceneScriptingFloatChannel, default=0, const [4.418,10]}
       hips_ctrl.Scale.X {type=MovieSceneScriptingFloatChannel, default=1, const [1,10]}
       hips_ctrl.Scale.Y {type=MovieSceneScriptingFloatChannel, default=1, const [1,10]}
       hips_ctrl.Scale.Z {type=MovieSceneScriptingFloatChannel, default=1, const [1,10]}
       hips_tan_ctrl.Location.X {
        type=MovieSceneScriptingFloatChannel
        default=0
        thinned_from=10
        keys [[0,-0.0165],[0.0013,0],[53,-0.0019],[64,-0.0025],[88,-0.0054],[93,-0.0054],[154,-0.0125],[161,-0.0165]]
       }
       hips_tan_ctrl.Location.Y {
        type=MovieSceneScriptingFloatChannel
        default=0.008
        thinned_from=10
        keys [[0,0.4038],[0.0013,0.008],[15.0013,0.0204],[53,0.0524],[64,0.0678],[88,0.1371],[93,0.1371],[154,0.4038],[161,0.4038]]
       }
       hips_tan_ctrl.Location.Z {
        type=MovieSceneScriptingFloatChannel
        default=-0.0073
        thinned_from=10
        keys [[0,0.0351],[0.0013,-0.0073],[53,-0.0026],[64,-0.0009],[88,0.0065],[93,0.0065],[154,0.0351],[161,0.0351]]
       }
       hips_tan_ctrl.Rotation.X {type=MovieSceneScriptingFloatChannel, default=0, const [0,10]}
       hips_tan_ctrl.Rotation.Y {type=MovieSceneScriptingFloatChannel, default=0, const [0,10]}
       hips_tan_ctrl.Rotation.Z {type=MovieSceneScriptingFloatChannel, default=0, const [0,10]}
       hips_tan_ctrl.Scale.X {type=MovieSceneScriptingFloatChannel, default=1, const [1,10]}
       hips_tan_ctrl.Scale.Y {type=MovieSceneScriptingFloatChannel, default=1, const [1,10]}
       hips_tan_ctrl.Scale.Z {type=MovieSceneScriptingFloatChannel, default=1, const [1,10]}
       spine_01_ctrl.Location.X {type=MovieSceneScriptingFloatChannel, default=0, const [0,10]}
       spine_01_ctrl.Location.Y {type=MovieSceneScriptingFloatChannel, default=0, const [0,10]}
       spine_01_ctrl.Location.Z {type=MovieSceneScriptingFloatChannel, default=0, const [0,10]}
       spine_01_ctrl.Rotation.X {type=MovieSceneScriptingFloatChannel, default=0, const [-0.4849,9]}
       spine_01_ctrl.Rotation.Y {type=MovieSceneScriptingFloatChannel, default=0, thinned_from=10, keys [[0,0],[161,0]]}
       spine_01_ctrl.Rotation.Z {type=MovieSceneScriptingFloatChannel, default=0, thinned_from=10, keys [[0,4.4086],[0.0013,4.418],[93,4.418],[154,4.4085],[161,4.418]]}
       spine_01_ctrl.Scale.X {type=MovieSceneScriptingFloatChannel, default=1, const [1,10]}
       spine_01_ctrl.Scale.Y {type=MovieSceneScriptingFloatChannel, default=1, const [1,10]}
       spine_01_ctrl.Scale.Z {type=MovieSceneScriptingFloatChannel, default=1, const [1,10]}
       spine_02_ctrl.Location.X {type=MovieSceneScriptingFloatChannel, default=0, const [0,10]}
       spine_02_ctrl.Location.Y {type=MovieSceneScriptingFloatChannel, default=0, const [0,10]}
       spine_02_ctrl.Location.Z {type=MovieSceneScriptingFloatChannel, default=0, const [0,10]}
       spine_02_ctrl.Rotation.X {type=MovieSceneScriptingFloatChannel, default=0, const [4.5541,9]}
       spine_02_ctrl.Rotation.Y {type=MovieSceneScriptingFloatChannel, default=0, thinned_from=10, keys [[0,0],[161,0]]}
       spine_02_ctrl.Rotation.Z {type=MovieSceneScriptingFloatChannel, default=0, thinned_from=10, keys [[0,4.4086],[0.0013,4.418],[93,4.418],[154,4.4085],[161,4.418]]}
       spine_02_ctrl.Scale.X {type=MovieSceneScriptingFloatChannel, default=1, const [1,10]}
       spine_02_ctrl.Scale.Y {type=MovieSceneScriptingFloatChannel, default=1, const [1,10]}
       spine_02_ctrl.Scale.Z {type=MovieSceneScriptingFloatChannel, default=1, const [1,10]}
       spine_03_ctrl.Location.X {type=MovieSceneScriptingFloatChannel, default=0, const [0,10]}
       spine_03_ctrl.Location.Y {type=MovieSceneScriptingFloatChannel, default=0, const [0,10]}
       spine_03_ctrl.Location.Z {type=MovieSceneScriptingFloatChannel, default=0, const [0,10]}
       spine_03_ctrl.Rotation.X {
        type=MovieSceneScriptingFloatChannel
        default=0
        thinned_from=10
        keys [[0,0],[0.0013,0],[15.0013,1.7422],[53,0],[64,0.2959],[88,1.263],[93,1.263],[154,0.6451],[161,0]]
       }
       spine_03_ctrl.Rotation.Y {type=MovieSceneScriptingFloatChannel, default=0, thinned_from=10, keys [[0,0],[161,0]]}
       spine_03_ctrl.Rotation.Z {type=MovieSceneScriptingFloatChannel, default=0, thinned_from=10, keys [[0,4.418],[93,4.418],[154,3.328],[161,4.418]]}
       spine_03_ctrl.Scale.X {type=MovieSceneScriptingFloatChannel, default=1, const [1,10]}
       spine_03_ctrl.Scale.Y {type=MovieSceneScriptingFloatChannel, default=1, const [1,10]}
       spine_03_ctrl.Scale.Z {type=MovieSceneScriptingFloatChannel, default=1, const [1,10]}
       spine_04_ctrl.Location.X {type=MovieSceneScriptingFloatChannel, default=0, thinned_from=15, keys [[0,0],[148,0],[161,0.0025]]}
       spine_04_ctrl.Location.Y {type=MovieSceneScriptingFloatChannel, default=0, thinned_from=15, keys [[0,0],[161,0]]}
       spine_04_ctrl.Location.Z {type=MovieSceneScriptingFloatChannel, default=0, thinned_from=15, keys [[0,0],[161,0]]}
       spine_04_ctrl.Rotation.X {type=MovieSceneScriptingFloatChannel, default=0, const [3.9108,15]}
       spine_04_ctrl.Rotation.Y {type=MovieSceneScriptingFloatChannel, default=0, thinned_from=15, keys [[0,0],[161,0]]}
       spine_04_ctrl.Rotation.Z {type=MovieSceneScriptingFloatChannel, default=-17.0961, const [-24.3268,15]}
       spine_04_ctrl.Scale.X {type=MovieSceneScriptingFloatChannel, default=1, const [1,15]}
       spine_04_ctrl.Scale.Y {type=MovieSceneScriptingFloatChannel, default=1, const [1,15]}
       spine_04_ctrl.Scale.Z {type=MovieSceneScriptingFloatChannel, default=1, const [1,15]}
       head_ctrl.Location.X {type=MovieSceneScriptingFloatChannel, default=0, const [0,10]}
       head_ctrl.Location.Y {type=MovieSceneScriptingFloatChannel, default=0, const [0,10]}
       head_ctrl.Location.Z {type=MovieSceneScriptingFloatChannel, default=0, const [0,10]}
       head_ctrl.Rotation.X {type=MovieSceneScriptingFloatChannel, default=0, thinned_from=10, keys [[0,0],[161,0]]}
       head_ctrl.Rotation.Y {type=MovieSceneScriptingFloatChannel, default=0, thinned_from=10, keys [[0,0],[161,0]]}
       head_ctrl.Rotation.Z {type=MovieSceneScriptingFloatChannel, default=0, thinned_from=10, keys [[0,0],[161,0]]}
       head_ctrl.Scale.X {type=MovieSceneScriptingFloatChannel, default=1, const [1,10]}
       head_ctrl.Scale.Y {type=MovieSceneScriptingFloatChannel, default=1, const [1,10]}
       head_ctrl.Scale.Z {type=MovieSceneScriptingFloatChannel, default=1, const [1,10]}
       clavicle_l_ctrl.Location.X {type=MovieSceneScriptingFloatChannel, default=0, const [0,12]}
       clavicle_l_ctrl.Location.Y {type=MovieSceneScriptingFloatChannel, default=0, const [0,12]}
       clavicle_l_ctrl.Location.Z {type=MovieSceneScriptingFloatChannel, default=0, const [0,12]}
       clavicle_l_ctrl.Rotation.X {type=MovieSceneScriptingFloatChannel, default=0, thinned_from=12, keys [[0,0],[161,0]]}
       clavicle_l_ctrl.Rotation.Y {type=MovieSceneScriptingFloatChannel, default=0, const [0,12]}
       clavicle_l_ctrl.Rotation.Z {type=MovieSceneScriptingFloatChannel, default=0, thinned_from=12, keys [[0,0],[161,0]]}
       clavicle_l_ctrl.Scale.X {type=MovieSceneScriptingFloatChannel, default=1, const [1,12]}
       clavicle_l_ctrl.Scale.Y {type=MovieSceneScriptingFloatChannel, default=1, const [1,12]}
       clavicle_l_ctrl.Scale.Z {type=MovieSceneScriptingFloatChannel, default=1, const [1,12]}
       clavicle_r_ctrl.Location.X {type=MovieSceneScriptingFloatChannel, default=0, thinned_from=13, keys [[0,0],[161,0]]}
       clavicle_r_ctrl.Location.Y {type=MovieSceneScriptingFloatChannel, default=0, thinned_from=13, keys [[0,0],[161,0]]}
       clavicle_r_ctrl.Location.Z {type=MovieSceneScriptingFloatChannel, default=0, thinned_from=13, keys [[0,0],[161,0]]}
       clavicle_r_ctrl.Rotation.X {type=MovieSceneScriptingFloatChannel, default=-0.2961, const [-0.2961,13]}
       clavicle_r_ctrl.Rotation.Y {type=MovieSceneScriptingFloatChannel, default=-0.0114, const [-0.0114,13]}
       clavicle_r_ctrl.Rotation.Z {type=MovieSceneScriptingFloatChannel, default=-4.4081, const [-4.4081,13]}
       clavicle_r_ctrl.Scale.X {type=MovieSceneScriptingFloatChannel, default=1, const [1,13]}
       clavicle_r_ctrl.Scale.Y {type=MovieSceneScriptingFloatChannel, default=1, const [1,13]}
       clavicle_r_ctrl.Scale.Z {type=MovieSceneScriptingFloatChannel, default=1, const [1,13]}
       neck_01_ctrl.Location.X {type=MovieSceneScriptingFloatChannel, default=0, const [0,6]}
       neck_01_ctrl.Location.Y {type=MovieSceneScriptingFloatChannel, default=0, const [0,6]}
       neck_01_ctrl.Location.Z {type=MovieSceneScriptingFloatChannel, default=0, const [0,6]}
       neck_01_ctrl.Rotation.X {type=MovieSceneScriptingFloatChannel, default=0, const [0,6]}
       neck_01_ctrl.Rotation.Y {type=MovieSceneScriptingFloatChannel, default=0, const [0,6]}
       neck_01_ctrl.Rotation.Z {type=MovieSceneScriptingFloatChannel, default=0, const [0,6]}
       neck_01_ctrl.Scale.X {type=MovieSceneScriptingFloatChannel, default=1, const [1,6]}
       neck_01_ctrl.Scale.Y {type=MovieSceneScriptingFloatChannel, default=1, const [1,6]}
       neck_01_ctrl.Scale.Z {type=MovieSceneScriptingFloatChannel, default=1, const [1,6]}
       neck_02_ctrl.Location.X {type=MovieSceneScriptingFloatChannel, default=0, const [0,5]}
       neck_02_ctrl.Location.Y {type=MovieSceneScriptingFloatChannel, default=0, const [0,5]}
       neck_02_ctrl.Location.Z {type=MovieSceneScriptingFloatChannel, default=0, const [0,5]}
       neck_02_ctrl.Rotation.X {type=MovieSceneScriptingFloatChannel, default=0, const [0,5]}
       neck_02_ctrl.Rotation.Y {type=MovieSceneScriptingFloatChannel, default=0, const [0,5]}
       neck_02_ctrl.Rotation.Z {type=MovieSceneScriptingFloatChannel, default=0, const [0,5]}
       neck_02_ctrl.Scale.X {type=MovieSceneScriptingFloatChannel, default=1, const [1,5]}
       neck_02_ctrl.Scale.Y {type=MovieSceneScriptingFloatChannel, default=1, const [1,5]}
       neck_02_ctrl.Scale.Z {type=MovieSceneScriptingFloatChannel, default=1, const [1,5]}
       head_ik_ctrl.Location.X {type=MovieSceneScriptingFloatChannel, default=0, const [0,10]}
       head_ik_ctrl.Location.Y {type=MovieSceneScriptingFloatChannel, default=0, const [0,10]}
       head_ik_ctrl.Location.Z {type=MovieSceneScriptingFloatChannel, default=0, const [0,10]}
       head_ik_ctrl.Rotation.X {type=MovieSceneScriptingFloatChannel, default=0, thinned_from=10, keys [[0,0],[161,0]]}
       head_ik_ctrl.Rotation.Y {type=MovieSceneScriptingFloatChannel, default=0, thinned_from=10, keys [[0,0],[161,0]]}
       head_ik_ctrl.Rotation.Z {type=MovieSceneScriptingFloatChannel, default=0, thinned_from=10, keys [[0,0],[161,0]]}
       head_ik_ctrl.Scale.X {type=MovieSceneScriptingFloatChannel, default=1, const [1,10]}
       head_ik_ctrl.Scale.Y {type=MovieSceneScriptingFloatChannel, default=1, const [1,10]}
       head_ik_ctrl.Scale.Z {type=MovieSceneScriptingFloatChannel, default=1, const [1,10]}
       chest_ctrl.Location.X {
        type=MovieSceneScriptingFloatChannel
        default=0
        thinned_from=10
        keys [[0,0.269],[0.0013,0],[53,0.0302],[64,0.0407],[88,0.0877],[93,0.0877],[154,0.2027],[161,0.269]]
       }
       chest_ctrl.Location.Y {
        type=MovieSceneScriptingFloatChannel
        default=0
        thinned_from=10
        keys [[0,-1.2318],[0.0013,0],[15.0013,-0.0387],[53,-0.1381],[64,-0.1862],[88,-0.4017],[93,-0.4017],[154,-1.2318],[161,-1.2318]]
       }
       chest_ctrl.Location.Z {
        type=MovieSceneScriptingFloatChannel
        default=0
        thinned_from=10
        keys [[0,-0.1775],[0.0013,0],[53,-0.0199],[64,-0.0268],[88,-0.0579],[93,-0.0579],[154,-0.1775],[161,-0.1775]]
       }
       chest_ctrl.Rotation.X {
        type=MovieSceneScriptingFloatChannel
        default=0
        thinned_from=10
        keys [[0,-5.0575],[0.0013,0],[15.0013,-0.1588],[53,-0.5671],[64,-0.7645],[88,-1.6493],[93,-1.6493],[154,-3.7944],[161,-5.0575]]
       }
       chest_ctrl.Rotation.Y {
        type=MovieSceneScriptingFloatChannel
        default=0
        thinned_from=10
        keys [[0,0.7851],[0.0013,0],[15.0013,0.0247],[53,0.088],[64,0.1187],[88,0.256],[93,0.256],[154,0.5914],[161,0.7851]]
       }
       chest_ctrl.Rotation.Z {
        type=MovieSceneScriptingFloatChannel
        default=0
        thinned_from=10
        keys [[0,13.2195],[0.0013,0],[15.0013,0.4151],[53,1.4822],[64,1.9984],[88,4.3109],[93,4.3109],[154,9.958],[161,13.2195]]
       }
       chest_ctrl.Scale.X {type=MovieSceneScriptingFloatChannel, default=1, const [1,10]}
       chest_ctrl.Scale.Y {type=MovieSceneScriptingFloatChannel, default=1, const [1,10]}
       chest_ctrl.Scale.Z {type=MovieSceneScriptingFloatChannel, default=1, const [1,10]}
       chest_tan_ctrl.Location.X {
        type=MovieSceneScriptingFloatChannel
        default=0
        thinned_from=10
        keys [[0,0.0217],[0.0013,0],[53,0.0024],[64,0.0033],[88,0.0071],[93,0.0071],[154,0.0164],[161,0.0217]]
       }
       chest_tan_ctrl.Location.Y {
        type=MovieSceneScriptingFloatChannel
        default=-0.0303
        thinned_from=10
        keys [[0,-0.2538],[0.0013,-0.0303],[53,-0.0553],[64,-0.0641],[88,-0.1031],[93,-0.1031],[154,-0.0359],[161,-0.2538]]
       }
       chest_tan_ctrl.Location.Z {type=MovieSceneScriptingFloatChannel, default=0.265, thinned_from=10, keys [[0,0.2667],[0.0013,0.265],[93,0.2655],[154,0.2889],[161,0.2667]]}
       chest_tan_ctrl.Rotation.X {
        type=MovieSceneScriptingFloatChannel
        default=0
        thinned_from=10
        keys [[0,0],[15.0013,0.1602],[53,0.5721],[64,0.7714],[88,1.664],[93,1.664],[154,1.7293],[161,0]]
       }
       chest_tan_ctrl.Rotation.Y {type=MovieSceneScriptingFloatChannel, default=0, thinned_from=10, keys [[0,0],[161,0]]}
       chest_tan_ctrl.Rotation.Z {type=MovieSceneScriptingFloatChannel, default=0, thinned_from=10, keys [[0,0],[161,0]]}
       chest_tan_ctrl.Scale.X {type=MovieSceneScriptingFloatChannel, default=1, const [1,10]}
       chest_tan_ctrl.Scale.Y {type=MovieSceneScriptingFloatChannel, default=1, const [1,10]}
       chest_tan_ctrl.Scale.Z {type=MovieSceneScriptingFloatChannel, default=1, const [1,10]}
      ]
     }]
   }
   {
    type=MovieSceneControlRigParameterTrack
    display_name=Fortnite_Legacy_Face_CtrlRig
    sections [{
      type=MovieSceneControlRigParameterSection
      channels [
       Flipbook {type=MovieSceneScriptingBoolChannel, default=false, const [false,3]}
       Face_Ctrl_Master.Location.X {type=MovieSceneScriptingFloatChannel, default=47.3466, keys [[0,47.3466],[15,44.827],[133,47.3466],[148,47.3466]]}
       Face_Ctrl_Master.Location.Y {type=MovieSceneScriptingFloatChannel, default=0.0202, keys [[0,0.0202],[15,0.2195],[133,0.0202],[148,0.0202]]}
       Face_Ctrl_Master.Location.Z {type=MovieSceneScriptingFloatChannel, default=2.4497, keys [[0,2.4497],[15,-0.2385],[133,2.4497],[148,2.4497]]}
       Face_Ctrl_Master.Rotation.X {type=MovieSceneScriptingFloatChannel, default=0, const [0,4]}
       Face_Ctrl_Master.Rotation.Y {type=MovieSceneScriptingFloatChannel, default=0, const [0,4]}
       Face_Ctrl_Master.Rotation.Z {type=MovieSceneScriptingFloatChannel, default=0, const [0,4]}
       Face_Ctrl_Master.Scale.X {type=MovieSceneScriptingFloatChannel, default=3, const [3,4]}
       Face_Ctrl_Master.Scale.Y {type=MovieSceneScriptingFloatChannel, default=3, const [3,4]}
       Face_Ctrl_Master.Scale.Z {type=MovieSceneScriptingFloatChannel, default=3, const [3,4]}
       gui_follow_head {type=MovieSceneScriptingFloatChannel, default=1, const [1,3]}
      ]
     }]
   }
   {
    type=MovieScene3DTransformTrack
    sections [{
      type=MovieScene3DTransformSection
      channels [
       Location.X {type=MovieSceneScriptingDoubleChannel, default=0, const [0,75]}
       Location.Y {type=MovieSceneScriptingDoubleChannel, default=0, const [0,75]}
       Location.Z {type=MovieSceneScriptingDoubleChannel, default=0, const [0,75]}
       Rotation.X {type=MovieSceneScriptingDoubleChannel, default=0, const [0,75]}
       Rotation.Y {type=MovieSceneScriptingDoubleChannel, default=0, const [0,75]}
       Rotation.Z {type=MovieSceneScriptingDoubleChannel, default=0, const [0,75]}
       Scale.X {type=MovieSceneScriptingDoubleChannel, default=1, const [1,75]}
       Scale.Y {type=MovieSceneScriptingDoubleChannel, default=1, const [1,75]}
       Scale.Z {type=MovieSceneScriptingDoubleChannel, default=1, const [1,75]}
      ]
     }]
   }
   {
    type=MovieScene3DTransformTrack
    sections [{
      type=MovieScene3DTransformSection
      channels [
       Location.X {type=MovieSceneScriptingDoubleChannel, default=0, const [0,2]}
       Location.Y {type=MovieSceneScriptingDoubleChannel, default=0, const [0,2]}
       Location.Z {type=MovieSceneScriptingDoubleChannel, default=0, const [0,2]}
       Rotation.X {type=MovieSceneScriptingDoubleChannel, default=0, const [0,2]}
       Rotation.Y {type=MovieSceneScriptingDoubleChannel, default=0, const [0,2]}
       Rotation.Z {type=MovieSceneScriptingDoubleChannel, default=0, const [0,2]}
       Scale.X {type=MovieSceneScriptingDoubleChannel, default=1, const [1,2]}
       Scale.Y {type=MovieSceneScriptingDoubleChannel, default=1, const [1,2]}
       Scale.Z {type=MovieSceneScriptingDoubleChannel, default=1, const [1,2]}
      ]
     }]
   }
  ]
 }
 Body {id=#2, parent_id=#1, possessed_class=SkeletalMeshComponentBudgeted, tracks []}
 Head {id=#3, parent_id=#1, possessed_class=SkeletalMeshComponentBudgeted, tracks []}
]
display_rate {denominator=1, numerator=30}
tick_resolution {denominator=1, numerator=24000}
tracks []
